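# Material function for the GnatGala character that blends fur hair cards into the body mesh's incoming material attributes.
# A transition gradient from vertex color green or UV2, switchable and also output, is sharpened into a mask for blending base color, normals and roughness.
# Card textures add color variation, two-sided normals and roughness.
# Subsurface shading combines anisotropy, adjusted hair tangents and fresnel-based emissive that fakes light transmission, partly scaled by time of day.
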MaterialFunction /BRCosmetics/Characters/Player/Male/Medium/Bodies/M_MED_GnatGala/Materials/MF_GnatGala_HairBlend
graph {nodes [
  n1 = VertexColor()
  n2 = ComponentMask(None=TextureCoordinate() {coordinateIndex=2}) {g=true}
  FunctionOutput(None=NamedRerouteDeclaration(None=StaticSwitchParameter(True=@n1.G, False=@n2) {defaultValue=true, group=HairCardsBlend, parameterName=Use vertex colors}) {name=FurTransitionGradation, nodeColor=<rgba 1 0.3516 0 1>}) {outputName=FurTransitionMask}
  n3 = Constant()
  n4 = ScalarParameter() {defaultValue=0.3, group=FN_Hair, parameterName=SpeculaiMIN}
  n5 = ScalarParameter() {defaultValue=0.5, group=FN_Hair, parameterName=SpeculaiMAX}
  n6 = NamedRerouteUsage()
  n7 = NamedRerouteUsage()
  n8 = SetMaterialAttributes!(
   MaterialAttributes=NamedRerouteUsage()
   Roughness=NamedRerouteUsage()
   Specular=Saturate(None=LinearInterpolate(A=@n4, B=@n5, Alpha=@n6))
   Metallic=Constant() {r=0.05}
   Anisotropy=LinearInterpolate(
    A=ScalarParameter() {defaultValue=0.5, group=FN_Hair, parameterName=Hair_anisotropy_min}
    B=ScalarParameter() {defaultValue=0.75, group=FN_Hair, parameterName=Hair_anisotropy_max}
    Alpha=Saturate(None=@n7)
   )
   BaseColor=NamedRerouteUsage()
   Normal=NamedRerouteUsage()
   ShadingModel=ShadingModel() {shadingModel=MSM_Subsurface}
   PixelDepthOffset
   EmissiveColor=NamedRerouteUsage()
   Opacity
   OpacityMask
   SubsurfaceColor
   Tangent=NamedRerouteUsage()
   CustomData0=@n3
   CustomData1=@n3
  ) {attributeSetTypes [
    #1
    #2
    #3
    #4
    #5
    #6
    #7
    #8
    #9
    #10
    #11
    #12
    #13
    #14
    #15
   ]}
  FunctionOutput(None=@n8)
  n9 = GetMaterialAttributes(None=FunctionInput() {bUsePreviewValueAsDefault=true, inputType=FunctionInput_MaterialAttributes}) {attributeGetTypes [
    #1
    #2
    #5
    #6
    #16
    #13
    #3
    #4
    #11
   ]}
  NamedRerouteDeclaration(None=@n9.Tangent) {name=Input_Tangent, nodeColor=<rgba 0.2656 0 1 1>}
  NamedRerouteDeclaration(None=@n9.Normal) {name=Input_BaseNormal, nodeColor=<rgba 0 1 0.5313 1>}
  NamedRerouteDeclaration(None=@n9.Specular) {name=Inputspecular, nodeColor=<rgba 0 0.2266 1 1>}
  n10 = TextureSampleParameter2D(UVs=TextureCoordinate() {coordinateIndex=1}) {
   desc="R = SpecMask G = Metallic B = Roughness"
   group=HairCardsBlend
   parameterName=Hair_SpecularMasks
   samplerSource=SSM_Wrap_WorldGroupSettings
   samplerType=SAMPLERTYPE_Masks
   texture=&T_Fortnite_Default_S
  }
  NamedRerouteDeclaration(None=LinearInterpolate(A=Multiply(A=@n10.B) {constB=0.93}, B=@n9.Roughness, Alpha=NamedRerouteUsage())) {name=Output_Roughness, nodeColor=<rgba 1 0 0.6797 1>}
  NamedRerouteDeclaration(None=@n9.Anisotropy) {name=OutPutAnisotropy, nodeColor=<rgba 0.1016 0 1 1>}
  n11 = Multiply(
   A=ScalarParameter() {defaultValue=10, group=FN_Hair, parameterName=Emissive_TIP_Brightness}
   B=MaterialFunctionCall(Value_Morning=Constant() {r=0.6}, Value_Day=Constant() {r=0.8}, Value_Evening=Constant() {r=0.4}, Value_Night=Constant() {r=0.2}) {materialFunction=&Scalar_Time_of_Day}.Result
  )
  n12 = ScalarParameter() {defaultValue=0.5, group=FN_Hair, parameterName=Emissive_Root_Brightness}
  n13 = NamedRerouteUsage()
  n14 = NamedRerouteDeclaration(None=MaterialFunctionCall(Brightness=Constant() {r=0.5}, Exponent=Constant() {r=3.5}, Vertex / Pixel Normal=VertexNormalWS()) {materialFunction=&MF_Fresnel_FX}.Fresnel Inverted Double Sided) {name=fresnel, nodeColor=<rgba 0 1 0.2969 1>}
  n15 = NamedRerouteUsage()
  NamedRerouteDeclaration(None=Multiply(A=Multiply(A=@n14, B=@n15), B=LinearInterpolate(A=@n11, B=@n12, Alpha=@n13) {constA=100})) {name=simulate light transmission, nodeColor=<rgba 0 1 0.8594 1>}
  NamedRerouteDeclaration(None=@n9.MaterialAttributes) {name=MA, nodeColor=<rgba 0 1 0.7656 1>}
  n16 = VertexNormalWS()
  n17 = TwoSidedSign()
  n18 = TextureSampleParameter2D(UVs=TextureCoordinate() {coordinateIndex=1}) {
   group=HairCardsBlend
   parameterName=Hair_Normals
   samplerSource=SSM_Wrap_WorldGroupSettings
   samplerType=SAMPLERTYPE_Normal
   texture=/Engine/EngineMaterials/DefaultNormal
  }
  n19 = Multiply(A=@n18.RGB, B=TwoSidedSign())
  n20 = ScalarParameter() {defaultValue=0.5, group=HairCardsBlend, parameterName=Hair cards Normal Flatness}
  n21 = Transform(None=MaterialFunctionCall(
    BaseNormal=Multiply(A=@n16, B=@n17)
    AdditionalNormal=MaterialFunctionCall(Normal=@n19, Flatness=@n20) {materialFunction=&FlattenNormal}.Result
   ) {materialFunction=&BlendAngleCorrectedNormals}.Result) {transformSourceType=TRANSFORMSOURCE_World, transformType=TRANSFORM_Tangent}
  n22 = MaterialFunctionCall(
   Normal=Multiply(A=NamedRerouteUsage(), B=TwoSidedSign())
   Flatness=LinearInterpolate(B=ScalarParameter() {defaultValue=0.5, group=HairCardsBlend, parameterName=body Mesh Normal Flatness}, Alpha=NamedRerouteUsage()) {constA=1}
  ) {materialFunction=&FlattenNormal}
  n23 = NamedRerouteUsage()
  NamedRerouteDeclaration(None=NamedRerouteDeclaration(None=LinearInterpolate(A=@n21, B=@n22.Result, Alpha=@n23)) {name=OutputNM, nodeColor=<rgba 1 0.5625 0 1>}) {name=Normal, nodeColor=<rgba 0.5 1 0 1>}
  n24 = Constant()
  n25 = Subtract(
   A=MaterialFunctionCall(X=@n24, Y=Constant(), Z=@n24) {materialFunction=&MakeFloat3}.Result
   B=Constant3Vector() {constant=<rgba 0.5 0.5 0 1>}
  )
  n26 = ScalarParameter() {defaultValue=0.3, group=FN_Hair, parameterName=Scraggle}
  n27 = NamedRerouteUsage()
  NamedRerouteDeclaration(None=Add(
    A=MaterialFunctionCall!(Tangent input=@n27) {materialFunction=/Game/Global/Materials/Cosmetics/Hair/MF_FN_HairAttributes}.Tangent
    B=Multiply(A=@n25, B=@n26)
   )) {name=OutPutTangent, nodeColor=<rgba 0 1 0.8359 1>}
  n28 = Add(A=NamedRerouteUsage(), B=ScalarParameter() {defaultValue=0.35, group=HairCardsBlend, parameterName=HairTransitionMask}) {constB=0.25}
  NamedRerouteDeclaration(None=Power(Base=Saturate(None=@n28)) {constExponent=10}) {name=FurTransitionMask, nodeColor=<rgba 0 1 0.8125 1>}
  NamedRerouteDeclaration(None=@n9.BaseColor) {name=Input_BaseColor, nodeColor=<rgba 1 0 0.0234 1>}
  n29 = TextureSampleParameter2D(UVs=TextureCoordinate() {coordinateIndex=1}) {
   group=HairCardsBlend
   parameterName=Hair cards Color variable
   samplerType=SAMPLERTYPE_Masks
   texture=/Game/Characters/Player/Male/Medium/Bodies/M_MED_GuineaPig/Materials/FurCards/T_GuineaPig_Fur_Alpha2_v3_randomGrey
  }
  n30 = TextureSampleParameter2D() {group=HairCardsBlend, parameterName=Hair_Diffuse, samplerSource=SSM_Wrap_WorldGroupSettings, texture=/Engine/EngineMaterials/DefaultDiffuse}
  n31 = LinearInterpolate(A=@n30.RGB, B=NamedRerouteUsage(), Alpha=ScalarParameter() {defaultValue=0.35, group=HairCardsBlend, parameterName=Body_blur_diffuse_alpha}) {constAlpha=0.8}
  n32 = Multiply(
   A=LinearInterpolate(
    A=ScalarParameter() {defaultValue=0.8, group=HairCardsBlend, parameterName=Hair cards Color variable dark}
    B=ScalarParameter() {defaultValue=1.5, group=HairCardsBlend, parameterName=Hair cards Color variable Brightness}
    Alpha=@n29.R
   ) {constA=0.9, constB=1.1}
   B=@n31
  )
  n33 = NamedRerouteUsage()
  NamedRerouteDeclaration(None=LinearInterpolate(A=LinearInterpolate(A=@n32, B=@n31, Alpha=@n33), B=NamedRerouteUsage(), Alpha=NamedRerouteUsage())) {name=OutPutBaseColor, nodeColor=<rgba 0.2891 1 0 1>}
 ]}
