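# Hair card material function for the GnatGala player body, overriding an incoming material attribute set, optionally passed through MF_GnatGala_HairBlend first.
# A mask texture runs through MF_MaskedGeoHoloControl and a vertex-green-weighted threshold to form the opacity mask, zeroed in shadow passes and at Low quality.
# Temporal dither drives pixel depth offset, weighted by vertex green or third-UV V.
# A scalar parameter pushes vertices along normals, weighted by vertex green.
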MaterialFunction /BRCosmetics/Characters/Player/Male/Medium/Bodies/M_MED_GnatGala/Materials/MF_GnatGala_Haircard
graph {nodes [
  n1 = TextureObjectParameter() {
   group=HairCardsBlend
   parameterName=SkinFX_Mask
   samplerType=SAMPLERTYPE_Masks
   texture=/BRCosmetics/Characters/Player/Male/Medium/Bodies/M_MED_GnatGala/Textures/T_GnatGalaHair_Alpha
  }
  n2 = TextureSample(Tex=@n1) {
   constCoordinate=1
   samplerType=SAMPLERTYPE_Masks
   texture=/Game/Characters/Player/Male/Medium/Bodies/M_MED_GuineaPig/Textures/T_GuineaPig_Body_M
  }
  n3 = MaterialFunctionCall!(R=Power(Base=@n2.R)) {materialFunction=/Game/Global/MaterialFunctions/TechArt/MF_MaskedGeoHoloControl}
  n4 = Multiply(A=@n3.R Out, B=Constant() {r=1}) {constB=2}
  n5 = Constant() {r=2}
  n6 = VertexColor()
  n7 = Constant() {r=1}
  n8 = If(
   A=Power(Base=Saturate(None=@n4), Exp=LinearInterpolate(B=@n5, Alpha=@n6.G) {constA=1}) {constExponent=10}
   B=Constant() {r=0.5}
   "A > B"=@n7
   "A == B"=@n7
   "A < B"=@n3.R Out
  )
  n9 = Multiply(A=@n8, B=Constant())
  n10 = TextureCoordinate() {coordinateIndex=2}
  n11 = Power(Base=StaticSwitchParameter(True=@n6.G, False=ComponentMask(None=@n10) {g=true}) {defaultValue=true, group=GuineaPig_furblend, parameterName=Use vertex colors}) {constExponent=1.5}
  n12 = LinearInterpolate(B=Constant() {r=0.5}, Alpha=@n11)
  n13 = Constant() {r=1}
  n14 = Constant() {r=1.5}
  n15 = ScalarParameter() {group=HairCardsBlend, parameterName=World Position Offset}
  n16 = VertexColor()
  n17 = FunctionInput() {bUsePreviewValueAsDefault=true, inputType=FunctionInput_MaterialAttributes}
  n18 = MaterialFunctionCall!(In=@n17) {materialFunction=&/MF_GnatGala_HairBlend}
  n19 = SetMaterialAttributes!(
   MaterialAttributes=GetMaterialAttributes(None=StaticSwitchParameter(True=@n18.Result, False=@n17) {defaultValue=true, group=Skin FX, parameterName=UseFN_GuineaPig_furblend}) {attributeGetTypes [#1, #2]}.MaterialAttributes
   OpacityMask=QualitySwitch(Default=ShadowReplace(Default=@n8, Shadow=@n9), Low=Constant())
   PixelDepthOffset=Multiply(
    A=MaterialFunctionCall(Alpha Threshold=@n12, Random=@n13) {materialFunction=&DitherTemporalAA}.Result
    B=LinearInterpolate(B=@n14, Alpha=@n11)
   )
   Opacity=LinearInterpolate(A=Constant() {r=1}, Alpha=TwoSidedSign())
   SubsurfaceColor=Constant()
   WorldPositionOffset=Multiply(A=VertexNormalWS(), B=Multiply(A=@n15, B=@n16.G))
  ) {attributeSetTypes [
    #1
    #3
    #4
    #5
    #6
   ]}
  n20 = GetMaterialAttributes(None=@n19) {attributeGetTypes [#1]}
  n21 = Constant() {r=1}
  FunctionOutput(None=SetMaterialAttributes!(MaterialAttributes=@n20.MaterialAttributes, OpacityMask=Multiply(A=@n20.OpacityMask, B=@n21)) {attributeSetTypes [#1]})
 ]}
