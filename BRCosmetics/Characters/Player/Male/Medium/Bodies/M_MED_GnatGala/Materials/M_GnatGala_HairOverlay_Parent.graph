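# Parent material for the hair overlay on the male medium GnatGala player body cosmetic: masked, two-sided subsurface, with ray-traced shadows off.
# It uses material attributes, wrapping the outfit-specific MF_GnatGala_MessHairOverlay function between the shared 2023 character-parent PreFX and PostFX functions.
# PreFX outputs feed both the overlay and PostFX, which also receives the overlay result as its material input.
# Usage flags cover skeletal meshes, clothing, morph targets, mesh deformers, Nanite and Niagara mesh particles.
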Material /BRCosmetics/Characters/Player/Male/Medium/Bodies/M_MED_GnatGala/Materials/M_GnatGala_HairOverlay_Parent
graph {
 properties {
  bCastRayTracedShadows=false
  bUseMaterialAttributes=true
  bUsedWithClothing=true
  bUsedWithMeshDeformer=true
  bUsedWithMorphTargets=true
  bUsedWithNanite=true
  bUsedWithNiagaraMeshParticles=true
  bUsedWithSkeletalMesh=true
  blendMode=BLEND_Masked
  pixelDepthOffsetMode=PDOM_Legacy
  shadingModel=MSM_Subsurface
  twoSided=true
  usedShadingModels=Subsurface
 }
 nodes [
  n1 = MaterialFunctionCall() {materialFunction=&MF_CharacterParent_2023_PreFX}
  MaterialFunctionCall(
   Material=MaterialFunctionCall!(PreSkinLOC_1=@n1.Pre skinned Local Position, MA_1=@n1.Pre skinned Local Position) {materialFunction=&/MF_GnatGala_MessHairOverlay}.Result
   M=@n1.M
   Pre Skinned Local Position=@n1.M
   Distance Blend=@n1.M
   Fresnel=@n1.M
  ) {materialFunction=&MF_CharacterParent_2023_PostFX}
 ]
}
