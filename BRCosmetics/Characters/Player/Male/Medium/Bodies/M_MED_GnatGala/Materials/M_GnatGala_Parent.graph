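# Parent character material for the GnatGala body cosmetic, adding a distance-driven "mess hair" effect between the shared 2023 character PreFX and PostFX functions.
# The hair mask combines a flat sample and a bump-offset parallax sample of the head FX texture, with world-aligned noise for breakup.
# Pixel-depth fades control blending toward the hair color and higher roughness, and flatten the normal.
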Material /BRCosmetics/Characters/Player/Male/Medium/Bodies/M_MED_GnatGala/Materials/M_GnatGala_Parent
graph {
 properties {
  bUseMaterialAttributes=true
  bUsedWithClothing=true
  bUsedWithMeshDeformer=true
  bUsedWithMorphTargets=true
  bUsedWithNanite=true
  bUsedWithNiagaraMeshParticles=true
  bUsedWithSkeletalMesh=true
  pixelDepthOffsetMode=PDOM_Legacy
  shadingModel=MSM_FromMaterialExpression
  usedShadingModels=Default Lit | Subsurface | Clear Coat | Subsurface Profile | Cloth
 }
 nodes [
  n1 = MaterialFunctionCall() {materialFunction=&MF_CharacterParent_2023_PreFX}
  n2 = GetMaterialAttributes(None=@n1.Result) {attributeGetTypes [#1, #2, #3]}
  n3 = NamedRerouteUsage()
  n4 = VectorParameter() {group=GnatGala_MessHair, parameterName=MessHairColorA, defaultValue=<rgba 0.0365 0.0208 0.0144 1>}
  n5 = Power(
   Base=MaterialFunctionCall() {materialFunction=&/MF_PixelDepthFadeIn}.Result
   Exp=ScalarParameter() {defaultValue=1, group=GnatGala_MessHair, parameterName=DistanceExp}
  )
  n6 = Saturate(None=Multiply(A=NamedRerouteUsage(), B=Saturate(None=@n5)))
  n7 = NamedRerouteUsage()
  n8 = MaterialFunctionCall() {materialFunction=&MF_PixelDepthFadeIn}
  n9 = NamedRerouteUsage()
  MaterialFunctionCall(
   Material=SetMaterialAttributes!(
    MaterialAttributes=@n2.MaterialAttributes
    Base Color=LinearInterpolate(A=@n3, B=@n4.RGB, Alpha=@n6)
    Normal=MaterialFunctionCall(Normal=@n7, Flatness=@n8.Result) {materialFunction=&FlattenNormal}.Result
    Roughness=LinearInterpolate(A=@n9, Alpha=@n6)
   ) {attributeSetTypes [#1, #3, #2]}
   M=@n1.M
   Pre Skinned Local Position=@n1.M
   Distance Blend=@n1.M
   Fresnel=@n1.M
  ) {materialFunction=&MF_CharacterParent_2023_PostFX}
  n10 = TextureObjectParameter() {
   group=GnatGala_MessHair
   parameterName=WPONoise
   samplerType=SAMPLERTYPE_Grayscale
   texture=/Game/Global/Textures/Default/Noise/LowResBlurredNoiseNormalized
  }
  n11 = Constant() {r=21}
  n12 = Add(A=NamedRerouteUsage(), B=Constant3Vector() {constant=<rgba 0.0227 1.1167 0 1>})
  NamedRerouteDeclaration(None=LinearInterpolate(
    A=Constant() {r=0.2}
    B=Constant() {r=0.6}
    Alpha=MaterialFunctionCall(TextureObject=@n10, TextureSize=@n11, WorldPosition=@n12) {materialFunction=&WorldAlignedTexture}.XYZ Texture
   )) {name=Noise, nodeColor=<rgba 0.0547 0 1 1>}
  NamedRerouteDeclaration(None=@n1.Pre skinned Local Position) {name=BasePreSkinned, nodeColor=<rgba 0 1 0.5547 1>}
  n13 = TextureSampleParameter2D(UVs=MaterialFunctionCall!(UV Coord=TextureCoordinate(), Height=Constant() {r=0.05}, ReferencePlane=Constant(), HeightRatio=Constant() {r=0.1}, Normal) {materialFunction=/Game/Global/MaterialFunctions/MF_BumpOffset_Accurate}.Result) {
   group=GnatGala_MessHair
   parameterName=MessHairMask
   samplerSource=SSM_Wrap_WorldGroupSettings
   samplerType=SAMPLERTYPE_Masks
   texture=/BRCosmetics/Characters/Player/Male/Medium/Bodies/M_MED_GnatGala/Textures/T_GnatGala_Head_FX
  }
  n14 = NamedRerouteDeclaration(None=Power(Base=@n13.G, Exp=Constant() {r=1}) {constExponent=1.8}) {name=MessHairMaskB, nodeColor=<rgba 1 0 0.9141 1>}
  n15 = TextureSampleParameter2D() {
   group=GnatGala_MessHair
   parameterName=MessHairMask
   samplerSource=SSM_Wrap_WorldGroupSettings
   samplerType=SAMPLERTYPE_Masks
   texture=&T_GnatGala_Head_FX
  }
  n16 = NamedRerouteDeclaration(None=Power(Base=@n15.R, Exp=Constant() {r=1.0042}) {constExponent=3}) {name=MessHairMaskA, nodeColor=<rgba 0 1 0.1094 1>}
  n17 = Multiply(
   A=Multiply(A=Max(A=@n14, B=@n16), B=@n13.B)
   B=ScalarParameter() {defaultValue=4, group=GnatGala_MessHair, parameterName=MessHairMask_Brightness}
  ) {constB=10}
  n18 = Constant() {r=1}
  NamedRerouteDeclaration(None=Multiply(A=Power(Base=@n17, Exp=@n18), B=NamedRerouteUsage())) {name=MessHairMask, nodeColor=<rgba 1 0.2344 0 1>}
  NamedRerouteDeclaration(None=@n2.Base Color) {name=InputColor, nodeColor=<rgba 0 0.0859 1 1>}
  NamedRerouteDeclaration(None=@n2.Roughness) {name=InputRoughness, nodeColor=<rgba 0.4062 1 0 1>}
  NamedRerouteDeclaration(None=@n2.Normal) {name=InputNormal, nodeColor=<rgba 1 0 0.7266 1>}
 ]
}
