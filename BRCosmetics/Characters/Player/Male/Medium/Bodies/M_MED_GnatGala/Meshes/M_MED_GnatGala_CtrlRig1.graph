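# Control Rig for the M_MED_GnatGala male cosmetic body, correcting costume and body deformation bones from the animated pose.
# A shared DeformRig function runs first, then sequenced passes adjust sleeves, hems, collar, shoulders, feet and garment folds.
# Aim constraints and spherical pose readers drive weighted bone offsets, with curve remapping and interpolated smoothing on front thigh folds.
# Skydiving and Emotes flags select shoulder fixes, while Pancake_Alpha gates extra front-fold overrides.
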ControlRigBlueprint /BRCosmetics/Characters/Player/Male/Medium/Bodies/M_MED_GnatGala/Meshes/M_MED_GnatGala_CtrlRig1
graphs [
 RigVMFunctionLibrary {links [], nodes []}
 RigVMFunction_Sequence_2_ContainedGraph {nodes [
   n1 = RigVMFunctionEntryNode()
   RigVMFunction_Sequence = &RigVMFunction_Sequence(ExecuteContext=@n1.ExecuteContext)
   n2 = &RigVMFunction_Sequence(ExecuteContext=@RigVMFunction_Sequence.B)
   n3 = &RigVMFunction_Sequence(ExecuteContext=@n2.B)
   RigVMFunctionReturnNode(A=@RigVMFunction_Sequence.A, B=@n2.A, C=@n3.A, D=@n3.B)
  ]}
 RigVMFunction_Sequence_2_ContainedGraph {nodes [
   n1 = RigVMFunctionEntryNode()
   n2 = &RigVMFunction_Sequence(ExecuteContext=@n1.ExecuteContext)
   n3 = &RigVMFunction_Sequence(ExecuteContext=@n2.B)
   RigVMFunctionReturnNode(A=@n2.A, B=@n3.A, C=@n3.B)
  ]}
 RigVMModel {nodes [
   RigUnit_BeginExecution = &RigUnit_BeginExecution()
   FN_DeformRig_FNC = RigVMFunctionReferenceNode(
    DeformRig_ Curve="(deformRig_l_upperarm_up_alpha_72_#1=False,deformRig_r_upperarm_up_alpha_73_#2=False,deformRig_l_upperarm_fr_alpha_74_#3=False,deformRig_r_upperarm_fr_alpha_75_CB51B8 ...(truncated)"
    DeformRig_Bool="(upperarm_1_#4=True,elbow_3_#5=True,wrist_7_#6=True,groin_8_#7=True,glute_17_#8=True,knee_9_40C408324CA261 ...(truncated)"
    DeformRig_Debug=(upperarm_debug_18_#4=False,wrist_debug_20_#6=False,lat_debug_26_#9=False,groinGlute_debug_27_#7=False)
    DeformRig_Value="(upperarm_up_18_#10=(Rotation=(X=0,Y=0,Z=0,W=1),Translation=(X=0,Y=0,Z=8),Scale3D=(X=0.7,Y=1,Z=1)),upperarm_fwd_fr_19_#11=(Rotation=(X=0,Y=0,Z=0,W=1),Translation=(X=0,Y=10,Z=0),Scale3D=(X=0.5,Y=1,Z= ...(truncated)"
    ExecuteContext=@RigUnit_BeginExecution.ExecutePin
   )
   > RigVMFunction_Sequence = RigVMAggregateNode()
   &RigUnit_AimConstraintLocalSpaceOffset(
    AimAxis=(X=-1,Y=0,Z=0)
    Child=(Type=Bone,Name="deform_sleeve_r")
    Filter=(bX=True,bY=True,bZ=True)
    Parents=((Item=(Type=Bone,Name="hand_r"),Weight=1))
    UpAxis=(X=0,Y=0,Z=-1)
    Weight=1
    WorldUp=(Target=(X=0,Y=0,Z=1),Kind=Location,Space=(Type=Bone,Name="deform_sleeve_r"))
    ExecutePin=@RigVMFunction_Sequence.A
   )
   > &RigUnit_AimConstraintLocalSpaceOffset(
    Child=(Type=Bone,Name="deform_sleeve_l")
    Filter=(bX=True,bY=True,bZ=True)
    Parents=((Item=(Type=Bone,Name="hand_l"),Weight=1))
    Weight=1
    WorldUp=(Target=(X=0,Y=0,Z=1),Kind=Location,Space=(Type=Bone,Name="deform_sleeve_l"))
   )
   > Get Delta Ref = RigVMFunctionReferenceNode(Item=(Type=Bone,Name="lowerarm_twist_01_l"))
   > n1 = &RigUnit_SphericalPoseReader(
    ActiveRegionScaleFactors=(PositiveWidth=1,NegativeWidth=1,PositiveHeight=1,NegativeHeight=1)
    ActiveRegionSize=0.01
    Debug=(bDrawDebug=false,bDraw2D=False,bDrawLocalAxes=False,DebugScale=25,DebugSegments=20,DebugThickness=0.25)
    DriverAxis=(X=1,Y=0,Z=0)
    DriverItem=(Type=Bone,Name="hand_l")
    FalloffRegionScaleFactors=(PositiveWidth=1,NegativeWidth=1,PositiveHeight=1,NegativeHeight=1)
    FalloffSize=0.3
    FlipHeightScaling=False
    FlipWidthScaling=False
    OptionalParentItem=(Type=Bone,Name="None")
    RotationOffset=(X=0,Y=-90,Z=0)
   )
   Subtract = &RigVMFunction_MathDoubleSub(A=1, B=@n1.OutputParam)
   > &RigUnit_ModifyTransforms(
    ItemToModify=((Item=(Type=Bone,Name="dyn_sleeve_l"),Transform=(Rotation=(X=0,Y=0,Z=0,W=1),Translation=(X=-6,Y=0,Z=2.5),Scale3D=(X=1,Y=1,Z=1))))
    WeightMinimum=0
    Weight=@Subtract.Result
   )
   Remap = &RigVMFunction_MathDoubleRemap(SourceMaximum=-110, SourceMinimum=-60, TargetMaximum=1, TargetMinimum=0, bClamp=False, Value=@"Get Delta Ref".Result.Rotation.Roll)
   > &RigUnit_ModifyTransforms(
    ItemToModify=((Item=(Type=Bone,Name="dyn_sleeve_l"),Transform=(Rotation=(X=0,Y=0,Z=0,W=1),Translation=(X=-2.5,Y=0,Z=0.5),Scale3D=(X=1,Y=1.1,Z=1.1))))
    WeightMinimum=0
    Weight=@Remap.Result
   )
   > Get Delta Ref_1 = RigVMFunctionReferenceNode(Item=(Type=Bone,Name="lowerarm_twist_01_r"))
   > n2 = &RigUnit_SphericalPoseReader(
    ActiveRegionScaleFactors=(PositiveWidth=1,NegativeWidth=1,PositiveHeight=1,NegativeHeight=1)
    ActiveRegionSize=0.01
    Debug=(bDrawDebug=false,bDraw2D=False,bDrawLocalAxes=False,DebugScale=25,DebugSegments=20,DebugThickness=0.25)
    DriverAxis=(X=1,Y=0,Z=0)
    DriverItem=(Type=Bone,Name="hand_r")
    FalloffRegionScaleFactors=(PositiveWidth=1,NegativeWidth=1,PositiveHeight=1,NegativeHeight=1)
    FalloffSize=0.3
    FlipHeightScaling=False
    FlipWidthScaling=False
    OptionalParentItem=(Type=Bone,Name="None")
    RotationOffset=(X=0,Y=-90,Z=0)
   )
   > &RigUnit_ModifyTransforms(
    ItemToModify=((Item=(Type=Bone,Name="dyn_sleeve_r"),Transform=(Rotation=(X=0,Y=0,Z=0,W=1),Translation=(X=6,Y=-1,Z=-2.5),Scale3D=(X=1,Y=1.05,Z=1.05))))
    WeightMinimum=0
    Weight=&RigVMFunction_MathDoubleSub(A=1, B=@n2.OutputParam).Result
   )
   > &RigUnit_ModifyTransforms(
    ItemToModify=((Item=(Type=Bone,Name="dyn_sleeve_r"),Transform=(Rotation=(X=0,Y=0,Z=0,W=1),Translation=(X=2.5,Y=0,Z=0),Scale3D=(X=1,Y=1.1,Z=1.1))))
    WeightMinimum=0
    Weight=&RigVMFunction_MathDoubleRemap(SourceMaximum=-110, SourceMinimum=-60, TargetMaximum=1, TargetMinimum=0, bClamp=False, Value=@"Get Delta Ref_1".Result.Rotation.Roll).Result
   )
   > n3 = &RigUnit_SphericalPoseReader(
    ActiveRegionScaleFactors=(PositiveWidth=1,NegativeWidth=1,PositiveHeight=1,NegativeHeight=1)
    ActiveRegionSize=0.01
    Debug=(bDrawDebug=true,bDraw2D=False,bDrawLocalAxes=False,DebugScale=25,DebugSegments=20,DebugThickness=0.25)
    DriverAxis=(X=1,Y=0,Z=0)
    DriverItem=(Type=Bone,Name="foot_l")
    FalloffRegionScaleFactors=(PositiveWidth=1,NegativeWidth=1,PositiveHeight=1,NegativeHeight=1)
    FalloffSize=0.3
    FlipHeightScaling=False
    FlipWidthScaling=False
    OptionalParentItem=(Type=Bone,Name="None")
    RotationOffset=(X=10,Y=-90,Z=0)
   )
   n4 = &RigVMFunction_AnimEvalRichCurve(
    Curve=(EditorCurveData=(Keys=((InterpMode=RCIM_Cubic),(InterpMode=RCIM_Cubic,TangentMode=RCTM_User,Time=1,Value=1,ArriveTangent=0.9542,LeaveTangent=0.9542))))
    SourceMinimum=0
    TargetMinimum=0
    Value=&RigVMFunction_MathDoubleSub(A=1, B=@n3.OutputParam).Result
   )
   > &RigUnit_ModifyTransforms(
    ItemToModify=((Item=(Type=Bone,Name="hem_l"),Transform=(Rotation=(X=0,Y=0,Z=0,W=1),Translation=(X=10,Y=-1.5,Z=0),Scale3D=(X=1,Y=1.1,Z=1.1))))
    WeightMinimum=0
    Weight=@n4.Result
   )
   > n5 = &RigUnit_SphericalPoseReader(
    ActiveRegionScaleFactors=(PositiveWidth=1,NegativeWidth=1,PositiveHeight=1,NegativeHeight=1)
    ActiveRegionSize=0.01
    Debug=(bDrawDebug=false,bDraw2D=False,bDrawLocalAxes=False,DebugScale=25,DebugSegments=20,DebugThickness=0.25)
    DriverAxis=(X=1,Y=0,Z=0)
    DriverItem=(Type=Bone,Name="foot_r")
    FalloffRegionScaleFactors=(PositiveWidth=1,NegativeWidth=1,PositiveHeight=1,NegativeHeight=1)
    FalloffSize=0.3
    FlipHeightScaling=False
    FlipWidthScaling=False
    OptionalParentItem=(Type=Bone,Name="None")
    RotationOffset=(X=7,Y=-90,Z=0)
   )
   n6 = &RigVMFunction_AnimEvalRichCurve(
    Curve=(EditorCurveData=(Keys=((InterpMode=RCIM_Cubic),(InterpMode=RCIM_Cubic,TangentMode=RCTM_User,Time=1,Value=1,ArriveTangent=0.9542,LeaveTangent=0.9542))))
    SourceMinimum=0
    TargetMinimum=0
    Value=&RigVMFunction_MathDoubleSub(A=1, B=@n5.OutputParam).Result
   )
   > &RigUnit_ModifyTransforms(
    ItemToModify=((Item=(Type=Bone,Name="hem_r"),Transform=(Rotation=(X=0,Y=0,Z=0,W=1),Translation=(X=-10,Y=1.5,Z=0),Scale3D=(X=1,Y=1.1,Z=1.1))))
    WeightMinimum=0
    Weight=@n6.Result
   )
   > &RigUnit_AimConstraintLocalSpaceOffset(
    AimAxis=(X=0,Y=-1,Z=0)
    Child=(Type=Bone,Name="deform_collar_l")
    Filter=(bX=True,bY=True,bZ=True)
    Parents=((Item=(Type=Bone,Name="deform_upperarm_up_base_l"),Weight=0.6),(Item=(Type=Bone,Name="deform_upperarm_up_l"),Weight=0.4))
    UpAxis=(X=1,Y=0,Z=0)
    Weight=1
    WorldUp=(Target=(X=0,Y=0,Z=1),Kind=Location,Space=(Type=Bone,Name="deform_collar_l"))
    bMaintainOffset=true
   )
   > n7 = &RigUnit_SphericalPoseReader(
    ActiveRegionScaleFactors=(PositiveWidth=1,NegativeWidth=1,PositiveHeight=1,NegativeHeight=1)
    ActiveRegionSize=0.15
    Debug=(bDrawDebug=false,bDraw2D=False,bDrawLocalAxes=False,DebugScale=25,DebugSegments=20,DebugThickness=0.25)
    DriverAxis=(X=1,Y=0,Z=0)
    DriverItem=(Type=Bone,Name="upperarm_l")
    FalloffRegionScaleFactors=(PositiveWidth=1,NegativeWidth=1,PositiveHeight=1,NegativeHeight=1)
    FalloffSize=0.4
    FlipHeightScaling=False
    FlipWidthScaling=False
    OptionalParentItem=(Type=Bone,Name="spine_05")
    RotationOffset=(X=70,Y=0,Z=60)
   )
   > &RigUnit_ModifyTransforms(
    ItemToModify=((Item=(Type=Bone,Name="deform_collar_l"),Transform=(Rotation=(X=0.3688,Y=0.1021,Z=-0.2464,W=-0.8904),Translation=(X=4.974,Y=1.4399,Z=0),Scale3D=(X=1,Y=0.3,Z=1))))
    WeightMinimum=0
    Weight=@n7.OutputParam
   )
   > n8 = &RigUnit_SphericalPoseReader(
    ActiveRegionScaleFactors=(PositiveWidth=1,NegativeWidth=1,PositiveHeight=1,NegativeHeight=1)
    ActiveRegionSize=0.15
    Debug=(bDrawDebug=false,bDraw2D=False,bDrawLocalAxes=False,DebugScale=25,DebugSegments=20,DebugThickness=0.25)
    DriverAxis=(X=1,Y=0,Z=0)
    DriverItem=(Type=Bone,Name="clavicle_l")
    FalloffRegionScaleFactors=(PositiveWidth=1,NegativeWidth=1,PositiveHeight=1,NegativeHeight=1)
    FalloffSize=0.4
    FlipHeightScaling=False
    FlipWidthScaling=False
    OptionalParentItem=(Type=Bone,Name="spine_05")
    RotationOffset=(X=0,Y=-25,Z=0)
   )
   > &RigUnit_ModifyTransforms(
    ItemToModify=((Item=(Type=Bone,Name="deform_collar_l"),Transform=(Rotation=(X=-0.3246,Y=-0.1397,Z=-0.0214,W=-0.9352),Translation=(X=3.3216,Y=1.1507,Z=1.2599),Scale3D=(X=1,Y=1,Z=1))))
    WeightMinimum=0
    Weight=&RigVMFunction_MathDoubleSub(B=@n7.OutputParam, A=@n8.OutputParam).Result
   )
   > &RigUnit_AimConstraintLocalSpaceOffset(
    Child=(Type=Bone,Name="collar_ba")
    Filter=(bX=True,bY=True,bZ=True)
    Parents=((Item=(Type=Bone,Name="head"),Weight=1))
    UpAxis=(X=0,Y=-1,Z=0)
    Weight=1
    WorldUp=(Target=(X=0,Y=0,Z=1),Kind=Location,Space=(Type=Bone,Name="collar_ba"))
   )
   > n9 = &RigUnit_SphericalPoseReader(
    ActiveRegionScaleFactors=(PositiveWidth=1,NegativeWidth=1,PositiveHeight=2,NegativeHeight=2)
    ActiveRegionSize=0.15
    Debug=(bDrawDebug=false,bDraw2D=False,bDrawLocalAxes=False,DebugScale=25,DebugSegments=20,DebugThickness=0.25)
    DriverAxis=(X=1,Y=0,Z=0)
    DriverItem=(Type=Bone,Name="head")
    FalloffRegionScaleFactors=(PositiveWidth=1,NegativeWidth=1,PositiveHeight=1,NegativeHeight=1)
    FalloffSize=0.8
    FlipHeightScaling=False
    FlipWidthScaling=False
    OptionalParentItem=(Type=Bone,Name="spine_05")
    RotationOffset=(X=-90,Y=0,Z=-40)
   )
   n10 = &RigVMFunction_AnimEvalRichCurve(
    Curve=(EditorCurveData=(Keys=((),(Time=0.2804,Value=0.3443),(Time=0.6903,Value=0.5036),(Time=1,Value=1,ArriveTangent=0.8754,LeaveTangent=0.8754))))
    SourceMinimum=0
    TargetMinimum=0
    Value=@n9.OutputParam
   )
   > n11 = &RigUnit_ModifyTransforms(
    ItemToModify=((Item=(Type=Bone,Name="collar_ba"),Transform=(Rotation=(X=0,Y=0,Z=0.2585,W=0.966),Translation=(X=-4.2969,Y=-4.8558,Z=0),Scale3D=(X=1.6012,Y=1,Z=1.3499))))
    WeightMinimum=0
    Weight=@n10.Result
   )
   RigVMFunction_ControlFlowBranch = &RigVMFunction_ControlFlowBranch(ExecuteContext=@n11.ExecutePin, Condition=RigVMVariableNode() {variable=Skydiving}.Value)
   n12 = &RigVMFunction_ControlFlowBranch(ExecuteContext=@RigVMFunction_ControlFlowBranch.False, Condition=RigVMVariableNode() {variable=Emotes}.Value)
   n13 = /Script/RigVM.RigVMFunction_MathQuaternionGetAxis(Axis=X, Quaternion=&RigUnit_GetTransform(Item=(Type=Bone,Name="spine_05"), bInitial=False).Transform.Rotation) {is=struct RigVMUnitNode, template="Axis::Execute(in Quaternion,in Axis,out Result)"}
   Dot___1 = /Script/RigVM.RigVMFunction_MathVectorDot(
    A=@n13.Result
    B=&RigVMFunction_MathQuaternionGetAxis(Axis=X, Quaternion=&RigUnit_GetTransform(Item=(Type=Bone,Name="upperarm_r"), bInitial=False).Transform.Rotation) {template="Axis::Execute(in Quaternion,in Axis,out Result)"}.Result
   ) {is=struct RigVMUnitNode, template="Dot,|::Execute(in A,in B,out Result)"}
   &RigUnit_ModifyTransforms(
    ItemToModify="((Item=(Type=Bone,Name=\"deform_upperarm_bk_r\"),Transform=(Rotation=(X=0.2076,Y=-0.3751,Z=-0.0799,W=-0.8999),Translation=(X=-6.8539,Y=1.1081,Z=2.6609),Scale3D=(X=1,Y=1,Z=1))),(Item=(Type=Bone,Name=\"upperarm_twist_01_r\"),Transform=(Rotation=(X=-0.5971,Y=0,Z= ...(truncated)"
    Weight=&RigVMFunction_MathDoubleRemap(SourceMaximum=-1, SourceMinimum=-0.53, TargetMaximum=1, TargetMinimum=0, Value=@Dot___1.Result).Result
    ExecutePin=@n12.False
   )
   Dot___2 = &RigVMFunction_MathVectorDot(
    B=&RigVMFunction_MathQuaternionGetAxis(Axis=X, Quaternion=&RigUnit_GetTransform(Item=(Type=Bone,Name="upperarm_l"), bInitial=False).Transform.Rotation) {template="Axis::Execute(in Quaternion,in Axis,out Result)"}.Result
    A=@n13.Result
   ) {template="Dot,|::Execute(in A,in B,out Result)"}
   n14 = &RigVMFunction_AnimEvalRichCurve(
    Curve=(EditorCurveData=(Keys=((),(Time=0.7381,Value=0.1374),(Time=1,Value=1))))
    SourceMinimum=0
    TargetMinimum=0
    Value=&RigVMFunction_MathDoubleRemap(SourceMaximum=1, SourceMinimum=0.53, TargetMaximum=1, TargetMinimum=0, Value=@Dot___2.Result).Result
   )
   > &RigUnit_ModifyTransforms(
    ItemToModify="((Item=(Type=Bone,Name=\"deform_upperarm_bk_l\"),Transform=(Rotation=(X=0.2146,Y=0.0283,Z=-0.1274,W=0.9679),Translation=(X=1.3255,Y=-0.0387,Z=-1.4725),Scale3D=(X=1,Y=1,Z=1))),(Item=(Type=Bone,Name=\"upperarm_twist_01_l\"),Transform=(Rotation=(X=0,Y=0,Z=0,W=1), ...(truncated)"
    Weight=@n14.Result
   )
   n15 = &RigUnit_SphericalPoseReader(
    ActiveRegionScaleFactors=(PositiveWidth=2.5,NegativeWidth=3,PositiveHeight=3,NegativeHeight=3)
    ActiveRegionSize=0.1
    Debug=(bDrawDebug=true,DebugScale=25,DebugSegments=20,DebugThickness=0.25)
    DriverAxis=(X=1,Y=0,Z=0)
    DriverItem=(Type=Bone,Name="upperarm_r")
    FalloffRegionScaleFactors=(PositiveWidth=1,NegativeWidth=1,PositiveHeight=1,NegativeHeight=1)
    FalloffSize=0.7
    OptionalParentItem=(Type=Bone)
    RotationOffset=(X=0,Y=0,Z=0)
    ExecutePin=@RigVMFunction_ControlFlowBranch.True
   )
   > &RigUnit_ModifyTransforms(
    ItemToModify=((Item=(Type=Bone,Name="deform_upperarm_up_r"),Transform=(Rotation=(X=0,Y=0,Z=0,W=1),Translation=(X=0,Y=2,Z=4),Scale3D=(X=1,Y=1,Z=1))))
    Weight=@n15.OutputParam
   )
   > n16 = &RigUnit_SphericalPoseReader(
    ActiveRegionScaleFactors=(PositiveWidth=2.5,NegativeWidth=3,PositiveHeight=3,NegativeHeight=3)
    ActiveRegionSize=0.1
    Debug=(bDrawDebug=false,DebugScale=25,DebugSegments=20,DebugThickness=0.25)
    DriverAxis=(X=1,Y=0,Z=0)
    DriverItem=(Type=Bone,Name="upperarm_l")
    FalloffRegionScaleFactors=(PositiveWidth=1,NegativeWidth=1,PositiveHeight=1,NegativeHeight=1)
    FalloffSize=0.7
    OptionalParentItem=(Type=Bone)
    RotationOffset=(X=0,Y=0,Z=0)
   )
   > &RigUnit_ModifyTransforms(
    ItemToModify=((Item=(Type=Bone,Name="deform_upperarm_up_l"),Transform=(Rotation=(X=0,Y=0,Z=0,W=1),Translation=(X=0,Y=-2,Z=-4),Scale3D=(X=1,Y=1,Z=1))),(Item=(Type=Bone,Name="None"),Transform=(Rotation=(X=0,Y=0,Z=0,W=1),Translation=(X=0,Y=0,Z=0),Scale3D=(X=1,Y=1,Z=1))))
    Weight=@n16.OutputParam
   )
   n17 = &RigUnit_SphericalPoseReader(
    ActiveRegionScaleFactors=(PositiveWidth=1,NegativeWidth=1,PositiveHeight=3,NegativeHeight=3)
    ActiveRegionSize=0.1
    Debug=(bDrawDebug=false,bDraw2D=False,bDrawLocalAxes=False,DebugScale=25,DebugSegments=20,DebugThickness=0.25)
    DriverAxis=(X=1,Y=0,Z=0)
    DriverItem=(Type=Bone,Name="foot_l")
    FalloffRegionScaleFactors=(PositiveWidth=1,NegativeWidth=1,PositiveHeight=1,NegativeHeight=1)
    FalloffSize=0.5
    FlipHeightScaling=False
    FlipWidthScaling=False
    OptionalParentItem=(Type=Bone,Name="None")
    RotationOffset=(X=90,Y=0,Z=0)
    ExecutePin=@RigVMFunction_Sequence.B
   )
   > n18 = &RigUnit_SphericalPoseReader(
    ActiveRegionScaleFactors=(PositiveWidth=1,NegativeWidth=1,PositiveHeight=3,NegativeHeight=3)
    ActiveRegionSize=0.1
    Debug=(bDrawDebug=false,bDraw2D=False,bDrawLocalAxes=False,DebugScale=25,DebugSegments=20,DebugThickness=0.25)
    DriverAxis=(X=1,Y=0,Z=0)
    DriverItem=(Type=Bone,Name="foot_r")
    FalloffRegionScaleFactors=(PositiveWidth=1,NegativeWidth=1,PositiveHeight=1,NegativeHeight=1)
    FalloffSize=0.5
    FlipHeightScaling=False
    FlipWidthScaling=False
    OptionalParentItem=(Type=Bone,Name="None")
    RotationOffset=(X=90,Y=0,Z=0)
   )
   > &RigUnit_ModifyTransforms(
    ItemToModify="((Item=(Type=Bone,Name=\"dyn_foot_l\"),Transform=(Rotation=(X=0,Y=0,Z=0,W=1),Translation=(X=5,Y=0,Z=0),Scale3D=(X=1,Y=0.1,Z=1))),(Item=(Type=Bone,Name=\"deform_foot_l\"),Transform=(Rotation=(X=0,Y=0,Z=0.342,W=0.9397),Translation=(X=0,Y=0,Z=0),Scale3D=(X=1,Y=1, ...(truncated)"
    WeightMinimum=0
    Weight=@n17.OutputParam
   )
   > &RigUnit_ModifyTransforms(
    ItemToModify="((Item=(Type=Bone,Name=\"dyn_foot_r\"),Transform=(Rotation=(X=0,Y=0,Z=0,W=1),Translation=(X=5,Y=0,Z=0),Scale3D=(X=1,Y=0.1,Z=1))),(Item=(Type=Bone,Name=\"deform_foot_r\"),Transform=(Rotation=(X=0,Y=0,Z=0.342,W=0.9397),Translation=(X=0,Y=0,Z=0),Scale3D=(X=1,Y=1, ...(truncated)"
    WeightMinimum=0
    Weight=@n18.OutputParam
   )
   RigVMFunction_Sequence_1 = RigVMAggregateNode(ExecuteContext=@RigVMFunction_Sequence.C)
   &RigUnit_AimConstraintLocalSpaceOffset(
    Child=(Type=Bone,Name="fold_01_ba_ct_MSTR")
    Filter=(bX=True,bY=True,bZ=True)
    Parents=((Item=(Type=Bone,Name="deform_glute_l"),Weight=0.5),(Item=(Type=Bone,Name="deform_glute_r"),Weight=0.5))
    UpAxis=(X=0,Y=-1,Z=0)
    Weight=1
    WorldUp=(Target=(X=0,Y=0,Z=1),Kind=Location,Space=(Type=Bone,Name="fold_01_ba_ct_MSTR"))
    bMaintainOffset=true
    ExecutePin=@RigVMFunction_Sequence_1.A
   )
   > &RigUnit_AimConstraintLocalSpaceOffset(
    Child=(Type=Bone,Name="fold_01_leg_ba_r_MSTR")
    Filter=(bX=True,bY=True,bZ=True)
    Parents=((Item=(Type=Bone,Name="deform_glute_r"),Weight=1))
    UpAxis=(X=0,Y=-1,Z=0)
    Weight=1
    WorldUp=(Target=(X=0,Y=0,Z=1),Kind=Location,Space=(Type=Bone,Name="fold_01_leg_ba_r_MSTR"))
   )
   > &RigUnit_AimConstraintLocalSpaceOffset(
    Child=(Type=Bone,Name="fold_01_leg_ba_l_MSTR")
    Filter=(bX=True,bY=True,bZ=True)
    Parents=((Item=(Type=Bone,Name="deform_glute_l"),Weight=1))
    UpAxis=(X=0,Y=-1,Z=0)
    Weight=1
    WorldUp=(Target=(X=0,Y=0,Z=1),Kind=Location,Space=(Type=Bone,Name="fold_01_leg_ba_l_MSTR"))
   )
   > &RigUnit_AimConstraintLocalSpaceOffset(
    Child=(Type=Bone,Name="fold_01_diag_ba_r_MSTR")
    Filter=(bX=True,bY=True,bZ=True)
    Parents=((Item=(Type=Bone,Name="deform_glute_r"),Weight=1))
    UpAxis=(X=0,Y=-1,Z=0)
    Weight=1
    WorldUp=(Target=(X=0,Y=0,Z=1),Kind=Location,Space=(Type=Bone,Name="fold_01_diag_ba_r_MSTR"))
   )
   > &RigUnit_AimConstraintLocalSpaceOffset(
    Child=(Type=Bone,Name="fold_01_diag_ba_l_MSTR")
    Filter=(bX=True,bY=True,bZ=True)
    Parents=((Item=(Type=Bone,Name="deform_glute_l"),Weight=1))
    UpAxis=(X=0,Y=-1,Z=0)
    Weight=1
    WorldUp=(Target=(X=0,Y=0,Z=1),Kind=Location,Space=(Type=Bone,Name="fold_01_diag_ba_l_MSTR"))
    bMaintainOffset=true
   )
   n19 = &RigUnit_SphericalPoseReader(
    ActiveRegionScaleFactors=(PositiveWidth=3,NegativeWidth=1,PositiveHeight=3,NegativeHeight=3)
    ActiveRegionSize=0.1
    Debug=(bDrawDebug=false,DebugScale=25,DebugSegments=20,DebugThickness=0.25)
    DriverAxis=(X=-1,Y=0,Z=0)
    DriverItem=(Type=Bone,Name="thigh_l")
    FalloffRegionScaleFactors=(PositiveWidth=1,NegativeWidth=2,PositiveHeight=1.1,NegativeHeight=4)
    FalloffSize=0.4
    OptionalParentItem=(Type=Bone,Name="spine_04")
    RotationOffset=(X=130,Y=55,Z=-55)
    ExecutePin=@RigVMFunction_Sequence_1.B
   )
   > n20 = &RigUnit_SphericalPoseReader(
    ActiveRegionScaleFactors=(PositiveWidth=3.5,NegativeWidth=2,PositiveHeight=2,NegativeHeight=2)
    ActiveRegionSize=0.1
    Debug=(bDrawDebug=false,DebugScale=25,DebugSegments=20,DebugThickness=0.25)
    DriverAxis=(X=-1,Y=0,Z=0)
    DriverItem=(Type=Bone,Name="thigh_l")
    FalloffRegionScaleFactors=(PositiveWidth=2,NegativeWidth=2,PositiveHeight=2,NegativeHeight=2)
    FalloffSize=0.45
    OptionalParentItem=(Type=Bone,Name="spine_03")
    RotationOffset=(X=-160,Y=0,Z=-30)
   )
   RigVMFunction_AnimEvalRichCurve = &RigVMFunction_AnimEvalRichCurve(Curve=(EditorCurveData=(Keys=((),(Time=0.2767,Value=0.5517),(Time=1,Value=1)))), SourceMinimum=0, TargetMinimum=0, Value=@n19.OutputParam)
   AlphaInterp = &RigVMFunction_AlphaInterp(InterpSpeedDecreasing=6, InterpSpeedIncreasing=50, Scale=1, bClampResult=true, bMapRange=true, Value=@RigVMFunction_AnimEvalRichCurve.Result)
   > &RigUnit_ModifyTransforms(
    ItemToModify="((Item=(Type=Bone,Name=\"fold_01_fr_l_MSTR\"),Transform=(Rotation=(X=0.3679,Y=0.3788,Z=-0.1966,W=0.8261),Translation=(X=-0.0013,Y=-0.0062,Z=-4.4993),Scale3D=(X=1,Y=1,Z=1))),(Item=(Type=Bone,Name=\"fold_01_diag_fr_l_MSTR\"),Transform=(Rotation=(X=-0.7283,Y=0.13 ...(truncated)"
    WeightMinimum=0
    Weight=@AlphaInterp.Result
   )
   Multiply = &RigVMFunction_MathDoubleMul(A=@n20.OutputParam, B=@n19.OutputParam)
   n21 = &RigVMFunction_AlphaInterp(
    InterpSpeedDecreasing=6
    InterpSpeedIncreasing=50
    Scale=1
    bClampResult=true
    Value=&RigVMFunction_MathDoubleRemap(SourceMaximum=1.5, SourceMinimum=0, TargetMaximum=1, TargetMinimum=0, Value=@Multiply.Result).Result
   )
   > &RigUnit_ModifyTransforms(
    ItemToModify="((Item=(Type=Bone,Name=\"fold_01_diag_fr_l_MSTR\"),Transform=(Rotation=(X=0.1024,Y=0.2959,Z=-0.942,W=0.1205),Translation=(X=0,Y=0,Z=0),Scale3D=(X=1,Y=1,Z=1))),(Item=(Type=Bone,Name=\"fold_01_fr_l_MSTR\"),Transform=(Rotation=(X=0.1582,Y=0.8788,Z=-0.0809,W=0.442 ...(truncated)"
    WeightMinimum=0
    Weight=@n21.Result
   )
   n22 = &RigUnit_SphericalPoseReader(
    ActiveRegionScaleFactors=(PositiveWidth=3,NegativeWidth=1,PositiveHeight=3,NegativeHeight=3)
    ActiveRegionSize=0.1
    Debug=(bDrawDebug=false,DebugScale=25,DebugSegments=20,DebugThickness=0.25)
    DriverAxis=(X=1,Y=0,Z=0)
    DriverItem=(Type=Bone,Name="thigh_r")
    FalloffRegionScaleFactors=(PositiveWidth=1,NegativeWidth=2,PositiveHeight=4,NegativeHeight=1.1)
    FalloffSize=0.4
    OptionalParentItem=(Type=Bone,Name="spine_04")
    RotationOffset=(X=55,Y=-55,Z=120)
    ExecutePin=@RigVMFunction_Sequence_1.C
   )
   > n23 = &RigUnit_SphericalPoseReader(
    ActiveRegionScaleFactors=(PositiveWidth=3.5,NegativeWidth=2,PositiveHeight=2,NegativeHeight=2)
    ActiveRegionSize=0.1
    Debug=(bDrawDebug=false,DebugScale=25,DebugSegments=20,DebugThickness=0.25)
    DriverAxis=(X=1,Y=0,Z=0)
    DriverItem=(Type=Bone,Name="thigh_r")
    FalloffRegionScaleFactors=(PositiveWidth=2,NegativeWidth=2,PositiveHeight=2,NegativeHeight=2)
    FalloffSize=0.45
    OptionalParentItem=(Type=Bone,Name="spine_03")
    RotationOffset=(X=-20,Y=0,Z=150)
   )
   n24 = &RigVMFunction_AnimEvalRichCurve(Curve=(EditorCurveData=(Keys=((),(Time=0.2353,Value=0.2105),(Time=1,Value=1)))), SourceMinimum=0, TargetMinimum=0, Value=@n22.OutputParam)
   n25 = &RigVMFunction_AlphaInterp(InterpSpeedDecreasing=6, InterpSpeedIncreasing=50, Scale=1, bClampResult=true, bMapRange=true, Value=@n24.Result)
   > &RigUnit_ModifyTransforms(
    ItemToModify="((Item=(Type=Bone,Name=\"fold_01_fr_r_MSTR\"),Transform=(Rotation=(X=-0.43,Y=-0.3583,Z=-0.2909,W=0.776),Translation=(X=-0.0001,Y=-0.0061,Z=4.4975),Scale3D=(X=1,Y=1,Z=1))),(Item=(Type=Bone,Name=\"fold_01_diag_fr_r_MSTR\"),Transform=(Rotation=(X=0.3147,Y=-0.0274 ...(truncated)"
    WeightMinimum=0
    Weight=@n25.Result
   )
   n26 = &RigVMFunction_AlphaInterp(
    InterpSpeedDecreasing=6
    InterpSpeedIncreasing=50
    Scale=1
    bClampResult=true
    Value=&RigVMFunction_MathDoubleRemap(
     SourceMaximum=1.5
     SourceMinimum=0
     TargetMaximum=1
     TargetMinimum=0
     Value=&RigVMFunction_MathDoubleMul(B=@n22.OutputParam, A=@n23.OutputParam).Result
    ).Result
   )
   > &RigUnit_ModifyTransforms(
    ItemToModify="((Item=(Type=Bone,Name=\"fold_01_diag_fr_r_MSTR\"),Transform=(Rotation=(X=-0.1024,Y=-0.2959,Z=-0.942,W=0.1205),Translation=(X=0,Y=0,Z=0),Scale3D=(X=1,Y=1,Z=1))),(Item=(Type=Bone,Name=\"fold_01_fr_r_MSTR\"),Transform=(Rotation=(X=0.1582,Y=0.8788,Z=0.0809,W=-0.4 ...(truncated)"
    WeightMinimum=0
    Weight=@n26.Result
   )
   n27 = &RigVMFunction_ControlFlowBranch(ExecuteContext=@RigVMFunction_Sequence.D, Condition=RigVMVariableNode() {variable=Pancake_Alpha}.Value)
   &RigUnit_AimConstraintLocalSpaceOffset(
    Child=(Type=Bone,Name="fold_01_diag_fr_l_MSTR")
    Filter=(bX=True,bY=True,bZ=True)
    Parents=((Item=(Type=Bone,Name="thigh_l"),Weight=1))
    UpAxis=(X=0,Y=-1,Z=0)
    Weight=1
    WorldUp=(Target=(X=0,Y=0,Z=1),Kind=Location,Space=(Type=Bone,Name="fold_01_diag_fr_l_MSTR"))
    bMaintainOffset=true
    ExecutePin=@n27.True
   )
   > &RigUnit_AimConstraintLocalSpaceOffset(
    Child=(Type=Bone,Name="fold_01_fr_l_MSTR")
    Filter=(bX=True,bY=True,bZ=True)
    Parents=((Item=(Type=Bone,Name="thigh_l"),Weight=1))
    UpAxis=(X=0,Y=-1,Z=0)
    Weight=1
    WorldUp=(Target=(X=0,Y=0,Z=1),Kind=Location,Space=(Type=Bone,Name="fold_01_fr_l_MSTR"))
    bMaintainOffset=true
   )
   > &RigUnit_AimConstraintLocalSpaceOffset(
    Child=(Type=Bone,Name="fold_01_diag_fr_r_MSTR")
    Filter=(bX=True,bY=True,bZ=True)
    Parents=((Item=(Type=Bone,Name="thigh_r"),Weight=1))
    UpAxis=(X=0,Y=-1,Z=0)
    Weight=1
    WorldUp=(Target=(X=0,Y=0,Z=1),Kind=Location,Space=(Type=Bone,Name="fold_01_diag_fr_r_MSTR"))
    bMaintainOffset=true
   )
   > &RigUnit_AimConstraintLocalSpaceOffset(
    Child=(Type=Bone,Name="fold_01_fr_r_MSTR")
    Filter=(bX=True,bY=True,bZ=True)
    Parents=((Item=(Type=Bone,Name="thigh_r"),Weight=1))
    UpAxis=(X=0,Y=-1,Z=0)
    Weight=1
    WorldUp=(Target=(X=0,Y=0,Z=1),Kind=Location,Space=(Type=Bone,Name="fold_01_fr_r_MSTR"))
    bMaintainOffset=true
   )
   > &RigUnit_ParentConstraint(
    AdvancedSettings=(InterpolationType=Average,RotationOrderForFilter=XZY)
    Child=(Type=Bone,Name="fold_03_fr_r")
    Filter=(TranslationFilter=(bX=false,bY=false,bZ=true),RotationFilter=(bX=true,bY=true,bZ=true),ScaleFilter=(bX=True,bY=True,bZ=True))
    Parents=((Item=(Type=Bone,Name="thigh_r"),Weight=1))
    Weight=0
    bMaintainOffset=true
   )
   > &RigUnit_ModifyTransforms(
    ItemToModify="((Item=(Type=Bone,Name=\"fold_01_diag_fr_l_MSTR\"),Transform=(Rotation=(X=0,Y=0,Z=0,W=1),Translation=(X=0,Y=0,Z=0),Scale3D=(X=1,Y=1,Z=1))),(Item=(Type=Bone,Name=\"fold_01_fr_l_MSTR\"),Transform=(Rotation=(X=0,Y=0,Z=0,W=1),Translation=(X=0,Y=0,Z=0),Scale3D=(X=1 ...(truncated)"
    Weight=1
    WeightMinimum=0
   )
   > &RigUnit_ModifyTransforms(
    ItemToModify="((Item=(Type=Bone,Name=\"fold_03_fr_r\"),Transform=(Rotation=(X=0,Y=0,Z=0.1305,W=0.9914),Translation=(X=0,Y=0,Z=0),Scale3D=(X=0.7,Y=1,Z=1))),(Item=(Type=Bone,Name=\"fold_03_diag_fr_r\"),Transform=(Rotation=(X=0,Y=0,Z=0.1305,W=0.9914),Translation=(X=0,Y=0,Z=0), ...(truncated)"
    Weight=1
    WeightMinimum=0
   )
   n28 = &RigUnit_SphericalPoseReader(
    ActiveRegionScaleFactors=(PositiveWidth=1,NegativeWidth=1,PositiveHeight=1,NegativeHeight=1)
    ActiveRegionSize=0.15
    Debug=(bDrawDebug=false,bDraw2D=False,bDrawLocalAxes=False,DebugScale=25,DebugSegments=20,DebugThickness=0.25)
    DriverAxis=(X=-1,Y=0,Z=0)
    DriverItem=(Type=Bone,Name="upperarm_r")
    ExecutePin=()
    FalloffRegionScaleFactors=(PositiveWidth=1,NegativeWidth=1,PositiveHeight=1,NegativeHeight=1)
    FalloffSize=0.5
    FlipHeightScaling=False
    FlipWidthScaling=False
    OptionalParentItem=(Type=Bone,Name="spine_05")
    RotationOffset=(X=-115,Y=0,Z=55)
   )
   > &RigUnit_ModifyTransforms(
    ItemToModify=((Item=(Type=Bone,Name="deform_collar_r"),Transform=(Rotation=(X=-0.1998,Y=0.0406,Z=-0.1371,W=-0.9694),Translation=(X=0.3396,Y=-0.9492,Z=-0.2542),Scale3D=(X=1,Y=0.5,Z=1))))
    WeightMinimum=0
    Weight=@n28.OutputParam
   )
   n29 = &RigUnit_SphericalPoseReader(
    ActiveRegionScaleFactors=(PositiveWidth=1,NegativeWidth=1,PositiveHeight=1,NegativeHeight=1)
    ActiveRegionSize=0.15
    Debug=(bDrawDebug=false,bDraw2D=False,bDrawLocalAxes=False,DebugScale=25,DebugSegments=20,DebugThickness=0.25)
    DriverAxis=(X=-1,Y=0,Z=0)
    DriverItem=(Type=Bone,Name="clavicle_r")
    ExecutePin=()
    FalloffRegionScaleFactors=(PositiveWidth=1,NegativeWidth=1,PositiveHeight=1,NegativeHeight=1)
    FalloffSize=0.36
    FlipHeightScaling=False
    FlipWidthScaling=False
    OptionalParentItem=(Type=Bone,Name="spine_05")
    RotationOffset=(X=0,Y=155,Z=0)
   )
   n30 = &RigVMFunction_MathDoubleSub(B=@n28.OutputParam, A=@n29.OutputParam)
   > &RigUnit_ModifyTransforms(
    ItemToModify=((Item=(Type=Bone,Name="deform_collar_r"),Transform=(Rotation=(X=-0.122,Y=0.0478,Z=-0.2222,W=-0.9662),Translation=(X=0.5634,Y=-1.3382,Z=-0.3337),Scale3D=(X=1,Y=1,Z=1))))
    WeightMinimum=0
    Weight=@n30.Result
   )
   > n31 = &RigUnit_SphericalPoseReader(
    ActiveRegionScaleFactors=(PositiveWidth=1,NegativeWidth=1,PositiveHeight=1,NegativeHeight=1)
    ActiveRegionSize=0.05
    Debug=(bDrawDebug=true,bDraw2D=False,bDrawLocalAxes=False,DebugScale=25,DebugSegments=20,DebugThickness=0.25)
    DriverAxis=(X=-1,Y=0,Z=0)
    DriverItem=(Type=Bone,Name="upperarm_r")
    FalloffRegionScaleFactors=(PositiveWidth=1,NegativeWidth=1,PositiveHeight=1,NegativeHeight=1)
    FalloffSize=0.3
    FlipHeightScaling=False
    FlipWidthScaling=False
    OptionalParentItem=(Type=Bone,Name="spine_05")
    RotationOffset=(X=35,Y=55,Z=0)
   )
   > &RigUnit_ModifyTransforms(
    ItemToModify=((Item=(Type=Bone,Name="deform_collar_r"),Transform=(Rotation=(X=-0.0922,Y=0.0696,Z=-0.0569,W=-0.9917),Translation=(X=0,Y=0,Z=0),Scale3D=(X=1,Y=1,Z=1))))
    WeightMinimum=0
    Weight=@n31.OutputParam
   )
   &RigVMFunction_AnimEvalRichCurve(Curve=(EditorCurveData=(Keys=((),(Time=1,Value=1)))), SourceMinimum=0, TargetMinimum=0, Value=@n30.Result)
  ]}
]
hierarchy [
 root {loc=<0 0 0>, rot=<0 0 0 1>}
 pelvis {parent=root, loc=<0 -0.7071 93.1191>, rot=<0.0127 0.707 -0.0127 -0.707>}
 spine_01 {parent=pelvis, loc=<6.7494 -0.6388 0>, rot=<0 0 0.034 -0.9994>}
 spine_02 {parent=spine_01, loc=<7.2518 -0.035 0>, rot=<0 0 0.0238 -0.9997>}
 spine_03 {parent=spine_02, loc=<6.0104 0 0>, rot=<0 0 0.0149 -0.9999>}
 spine_04 {parent=spine_03, loc=<8.1024 0 0>, rot=<0 0 -0.0132 -0.9999>}
 spine_05 {parent=spine_04, loc=<10.0146 0 0>, rot=<0 0 -0.0015 -1>}
 neck_01 {parent=spine_05, loc=<11.3138 0.1529 0>, rot=<0 0 -0.2003 -0.9797>}
 neck_02 {parent=neck_01, loc=<4.5627 0 0>, rot=<0 0 0.0504 -0.9987>}
 head {parent=neck_02, loc=<5.3914 0 0>, rot=<0 0 0.115 -0.9934>}
 attach_cape {parent=spine_05, loc=<8.9817 -11.7244 0.0039>, rot=<0.1185 -0.737 -0.1273 -0.6532>}
 cape {parent=attach_cape, type=Null, loc=<0 0 0>, rot=<0 0 0 1>}
 attach_backpack {parent=spine_05, loc=<-0.1556 -13.4785 -0.0018>, rot=<-0.0092 -0.7071 0.0093 -0.707>}
 Backpack_BR {parent=attach_backpack, type=Null, loc=<0 0 0>, rot=<0 0 0 1>}
 clavicle_l {parent=spine_05, loc=<4.3575 -0.1466 -2.3223>, rot=<0.0465 -0.7289 -0.009 -0.6829>}
 upperarm_l {parent=clavicle_l, loc=<15.5111 0 0>, rot=<0.0226 -0.3234 -0.0088 -0.9459>}
 lowerarm_l {parent=upperarm_l, loc=<26.6514 0 0>, rot=<0.0391 -0.0099 -0.2453 -0.9686>}
 hand_l {parent=lowerarm_l, loc=<22.3491 0 0>, rot=<-0.7081 -0.0507 0.0975 -0.6975>}
 thumb_01_l {parent=hand_l, loc=<1.3118 -0.3191 -2.459>, rot=<0.6546 -0.4042 0.0664 -0.6354>}
 thumb_02_l {parent=thumb_01_l, loc=<5.9336 -0.0448 0.0401>, rot=<-0.0037 0.0213 0.298 -0.9543>}
 thumb_03_l {parent=thumb_02_l, loc=<2.2223 0.3682 -0.0907>, rot=<0.0118 -0.013 0.0066 -0.9998>}
 index_metacarpal_l {parent=hand_l, loc=<1.6535 0.0183 -2.96>, rot=<0.0158 -0.0124 -0.0173 -0.9996>}
 index_01_l {parent=index_metacarpal_l, loc=<8.3008 0.0357 -0.0524>, rot=<-0.0044 0.0485 0.0924 -0.9945>}
 index_02_l {parent=index_01_l, loc=<3.3587 -0.3004 0.0452>, rot=<-0.002 0.0027 0.3225 -0.9466>}
 index_03_l {parent=index_02_l, loc=<2.5578 -0.123 -0.0252>, rot=<-0.002 0.0067 0.2558 -0.9667>}
 middle_metacarpal_l {parent=hand_l, loc=<2.1856 -0.1725 -0.5984>, rot=<0.006 0.0168 -0.0329 -0.9993>}
 middle_01_l {parent=middle_metacarpal_l, loc=<7.3275 -0.2447 0.0129>, rot=<-0.0208 0.0042 0.0858 -0.9961>}
 middle_02_l {parent=middle_01_l, loc=<4.0123 0.1401 0.1175>, rot=<-0.0218 0.0258 0.3512 -0.9357>}
 middle_03_l {parent=middle_02_l, loc=<2.6077 0.0534 0.1373>, rot=<0.0082 0.0188 0.1708 -0.9851>}
 ring_metacarpal_l {parent=hand_l, loc=<2.2074 -0.3625 1.3638>, rot=<-0.0284 0.0722 -0.0458 -0.9959>}
 ring_01_l {parent=ring_metacarpal_l, loc=<7.0295 -0.6197 -0.1446>, rot=<-0.0489 -0.0252 0.1892 -0.9804>}
 ring_02_l {parent=ring_01_l, loc=<3.6174 0.4647 -0.1027>, rot=<0 0 0.2806 -0.9598>}
 ring_03_l {parent=ring_02_l, loc=<2.4152 0.114 0.0199>, rot=<0 0 0.1189 -0.9929>}
 pinky_metacarpal_l {parent=hand_l, loc=<1.7616 -0.9753 3.3814>, rot=<-0.0397 0.0814 -0.069 -0.9935>}
 pinky_01_l {parent=pinky_metacarpal_l, loc=<7.1709 -0.4831 -0.1094>, rot=<-0.0654 -0.0035 0.2242 -0.9723>}
 pinky_02_l {parent=pinky_01_l, loc=<2.8274 0.0893 -0.1043>, rot=<0.0037 0.0011 0.2854 -0.9584>}
 pinky_03_l {parent=pinky_02_l, loc=<2.2046 0.0307 -0.0078>, rot=<-0.0002 0.0098 0.1445 -0.9895>}
 hand_attach_l {parent=hand_l, loc=<8.5 -3 0>, rot=<0.7071 0.7071 0 0>}
 deform_wrist_dn_base_l {parent=hand_l, loc=<0 -2.2342 0>, rot=<0 0 0 -1>}
 deform_wrist_dn_l {parent=deform_wrist_dn_base_l, loc=<0 0 0>, rot=<0 0 0 1>}
 deform_wrist_up_base_l {parent=hand_l, loc=<0 2.2682 0>, rot=<0 0 0 -1>}
 deform_wrist_up_l {parent=deform_wrist_up_base_l, loc=<0 0 0>, rot=<0 0 0 1>}
 lowerarm_twist_02_l {parent=lowerarm_l, loc=<9.6199 0 0>, rot=<0 0 0 -1>}
 deform_sleeve_l {parent=lowerarm_twist_02_l, loc=<0 0 0>, rot=<0 0 0 1>}
 dyn_sleeve_l {parent=deform_sleeve_l, loc=<13.7271 0 0>, rot=<0 0 0 1>}
 sleeve_l {parent=lowerarm_twist_02_l, loc=<0 0 0>, rot=<0 0 0 1>}
 sleeve_l_end {parent=sleeve_l, loc=<13.7271 0 0>, rot=<0 0 0 1>}
 lowerarm_twist_01_l {parent=lowerarm_l, loc=<17.8549 0 0>, rot=<0 0 0 1>}
 deform_elbow_in_base_l {parent=lowerarm_l, loc=<0.2112 5.259 0.8274>, rot=<0 0 0 -1>}
 deform_elbow_in_l {parent=deform_elbow_in_base_l, loc=<0 0 0>, rot=<0 0 0 1>}
 deform_elbow_out_base_l {parent=lowerarm_l, loc=<0 -4 0>, rot=<0 0 0 -1>}
 deform_elbow_out_l {parent=deform_elbow_out_base_l, loc=<0 0 0>, rot=<0 0 0 1>}
 upperarm_twist_01_l {parent=upperarm_l, loc=<0.0041 0 0>, rot=<0 0 0 1>}
 deform_upperarm_bk_base_l {parent=upperarm_twist_01_l, loc=<1.0813 -7.2193 -1.9453>, rot=<0 0 0 1>}
 deform_upperarm_bk_l {parent=deform_upperarm_bk_base_l, loc=<0 0 0>, rot=<0 0 0 1>}
 deform_upperarm_fr_base_l {parent=upperarm_twist_01_l, loc=<-0.3691 7.0506 -2.765>, rot=<0 0 0 1>}
 deform_upperarm_fr_l {parent=deform_upperarm_fr_base_l, loc=<0 0 0>, rot=<0 0 0 1>}
 deform_upperarm_up_base_l {parent=upperarm_twist_01_l, loc=<-3.2216 -1.1391 3.3311>, rot=<0 0 0 1>}
 deform_upperarm_up_l {parent=deform_upperarm_up_base_l, loc=<0 0 0>, rot=<0 0 0 1>}
 upperarm_twist_02_l {parent=upperarm_l, loc=<13.7941 0 0>, rot=<0 0 0 1>}
 tricep_l {parent=upperarm_twist_02_l, loc=<0 -8.019 0>, rot=<0 0 0 1>}
 bicep_l {parent=upperarm_twist_02_l, loc=<0 6.7438 0>, rot=<0 0 0 1>}
 elbow_l {parent=upperarm_l, loc=<26.6514 0 0>, rot=<0.0391 -0.0099 -0.2453 -0.9686>}
 elbow_fr_l {parent=elbow_l, loc=<0.3197 7.3801 0>, rot=<0 0 0 1>}
 elbow_ba_l {parent=elbow_l, loc=<-1.524 -4.717 0>, rot=<0 0 0 1>}
 def_arm_tp_l {parent=clavicle_l, loc=<17.8594 -1.0553 6.8356>, rot=<0 0 0 1>}
 def_arm_tp_l_end {parent=def_arm_tp_l, loc=<0 0 0>, rot=<0 0 0 1>}
 def_arm_fr_l {parent=clavicle_l, loc=<16.8097 7.219 -2.9938>, rot=<0 0 0 1>}
 def_arm_fr_l_end {parent=def_arm_fr_l, loc=<0 0 0>, rot=<0 0 0 1>}
 def_arm_ba_l {parent=clavicle_l, loc=<17.803 -9.1305 -2.9297>, rot=<0 0 0 1>}
 def_arm_ba_l_end {parent=def_arm_ba_l, loc=<0 0 0>, rot=<0 0 0 1>}
 clavicle_r {parent=spine_05, loc=<4.3575 -0.1466 2.3223>, rot=<0.7289 0.0465 -0.6829 0.009>}
 upperarm_r {parent=clavicle_r, loc=<-15.5111 0 0>, rot=<0.0226 -0.3234 -0.0088 -0.9459>}
 lowerarm_r {parent=upperarm_r, loc=<-26.6514 0 0>, rot=<0.0391 -0.0099 -0.2453 -0.9686>}
 hand_r {parent=lowerarm_r, loc=<-22.3491 0 0>, rot=<-0.7081 -0.0507 0.0975 -0.6975>}
 thumb_01_r {parent=hand_r, loc=<-1.3118 0.3191 2.459>, rot=<0.6546 -0.4042 0.0664 -0.6354>}
 thumb_02_r {parent=thumb_01_r, loc=<-5.9336 0.0448 -0.0401>, rot=<-0.0037 0.0213 0.298 -0.9543>}
 thumb_03_r {parent=thumb_02_r, loc=<-2.2223 -0.3682 0.0907>, rot=<0.0118 -0.013 0.0066 -0.9998>}
 index_metacarpal_r {parent=hand_r, loc=<-1.6535 -0.0183 2.96>, rot=<0.0158 -0.0124 -0.0173 -0.9996>}
 index_01_r {parent=index_metacarpal_r, loc=<-8.3008 -0.0357 0.0524>, rot=<-0.0044 0.0485 0.0924 -0.9945>}
 index_02_r {parent=index_01_r, loc=<-3.3587 0.3004 -0.0452>, rot=<-0.002 0.0027 0.3225 -0.9466>}
 index_03_r {parent=index_02_r, loc=<-2.5578 0.123 0.0252>, rot=<-0.002 0.0067 0.2558 -0.9667>}
 middle_metacarpal_r {parent=hand_r, loc=<-2.1856 0.1725 0.5984>, rot=<0.006 0.0168 -0.0329 -0.9993>}
 middle_01_r {parent=middle_metacarpal_r, loc=<-7.3275 0.2447 -0.0129>, rot=<-0.0208 0.0042 0.0858 -0.9961>}
 middle_02_r {parent=middle_01_r, loc=<-4.0123 -0.1401 -0.1175>, rot=<-0.0218 0.0258 0.3512 -0.9357>}
 middle_03_r {parent=middle_02_r, loc=<-2.6077 -0.0534 -0.1373>, rot=<0.0082 0.0188 0.1708 -0.9851>}
 ring_metacarpal_r {parent=hand_r, loc=<-2.2074 0.3625 -1.3638>, rot=<-0.0284 0.0722 -0.0458 -0.9959>}
 ring_01_r {parent=ring_metacarpal_r, loc=<-7.0295 0.6197 0.1446>, rot=<-0.0489 -0.0252 0.1892 -0.9804>}
 ring_02_r {parent=ring_01_r, loc=<-3.6174 -0.4647 0.1027>, rot=<0 0 0.2806 -0.9598>}
 ring_03_r {parent=ring_02_r, loc=<-2.4152 -0.114 -0.0199>, rot=<0 0 0.1189 -0.9929>}
 pinky_metacarpal_r {parent=hand_r, loc=<-1.7616 0.9753 -3.3814>, rot=<-0.0397 0.0814 -0.069 -0.9935>}
 pinky_01_r {parent=pinky_metacarpal_r, loc=<-7.1709 0.4831 0.1094>, rot=<-0.0654 -0.0035 0.2242 -0.9723>}
 pinky_02_r {parent=pinky_01_r, loc=<-2.8274 -0.0893 0.1043>, rot=<0.0037 0.0011 0.2854 -0.9584>}
 pinky_03_r {parent=pinky_02_r, loc=<-2.2046 -0.0307 0.0078>, rot=<-0.0002 0.0098 0.1445 -0.9895>}
 hand_attach_r {parent=hand_r, loc=<-8.5001 2.9999 0>, rot=<0 0 -0.7071 -0.7071>}
 deform_wrist_dn_base_r {parent=hand_r, loc=<-0.0004 2.2345 -0.0001>, rot=<0 0 0 -1>}
 deform_wrist_dn_r {parent=deform_wrist_dn_base_r, loc=<0 0 0>, rot=<0 0 0 1>}
 deform_wrist_up_base_r {parent=hand_r, loc=<0.0003 -2.2684 0.0001>, rot=<0 0 0 -1>}
 deform_wrist_up_r {parent=deform_wrist_up_base_r, loc=<0 0 0>, rot=<0 0 0 1>}
 lowerarm_twist_02_r {parent=lowerarm_r, loc=<-9.6199 0 0>, rot=<0 0 0 1>}
 deform_sleeve_r {parent=lowerarm_twist_02_r, loc=<0 0 0>, rot=<0 0 0 1>}
 dyn_sleeve_r {parent=deform_sleeve_r, loc=<-13.727 0 0>, rot=<0 0 0 1>}
 sleeve_r {parent=lowerarm_twist_02_r, loc=<0 0 0>, rot=<0 0 0 1>}
 sleeve_r_end {parent=sleeve_r, loc=<-13.727 0 0>, rot=<0 0 0 1>}
 lowerarm_twist_01_r {parent=lowerarm_r, loc=<-17.8549 0 0>, rot=<0 0 0 1>}
 deform_elbow_in_base_r {parent=lowerarm_r, loc=<-0.2111 -5.2591 -0.8275>, rot=<0 0 0 -1>}
 deform_elbow_in_r {parent=deform_elbow_in_base_r, loc=<-0.0001 0 0.0002>, rot=<0 0 0 1>}
 deform_elbow_out_base_r {parent=lowerarm_r, loc=<0 4 -0.0001>, rot=<0 0 0 -1>}
 deform_elbow_out_r {parent=deform_elbow_out_base_r, loc=<0 0 0>, rot=<0 0 0 1>}
 upperarm_twist_01_r {parent=upperarm_r, loc=<-0.0041 0 0>, rot=<0 0 0 1>}
 deform_upperarm_bk_base_r {parent=upperarm_twist_01_r, loc=<-1.0814 7.2192 1.9455>, rot=<0 0 0 1>}
 deform_upperarm_bk_r {parent=deform_upperarm_bk_base_r, loc=<0 0 0>, rot=<0 0 0 1>}
 deform_upperarm_fr_base_r {parent=upperarm_twist_01_r, loc=<0.3689 -7.0506 2.7652>, rot=<0 0 0 1>}
 deform_upperarm_fr_r {parent=deform_upperarm_fr_base_r, loc=<0 0 0>, rot=<0 0 0 1>}
 deform_upperarm_up_base_r {parent=upperarm_twist_01_r, loc=<3.2214 1.1391 -3.3308>, rot=<0 0 0 1>}
 deform_upperarm_up_r {parent=deform_upperarm_up_base_r, loc=<0 0 0>, rot=<0 0 0 -1>}
 upperarm_twist_02_r {parent=upperarm_r, loc=<-13.7941 0 0>, rot=<0 0 0 1>}
 bicep_r {parent=upperarm_twist_02_r, loc=<0 -6.744 0>, rot=<0 0 0 1>}
 tricep_r {parent=upperarm_twist_02_r, loc=<0 8.019 0>, rot=<0 0 0 1>}
 elbow_r {parent=upperarm_r, loc=<-26.6514 0 0>, rot=<0.0391 -0.0099 -0.2453 -0.9686>}
 elbow_fr_r {parent=elbow_r, loc=<-0.32 -7.38 0>, rot=<0 0 0 1>}
 elbow_ba_r {parent=elbow_r, loc=<1.524 4.717 0>, rot=<0 0 0 1>}
 def_arm_tp_r {parent=clavicle_r, loc=<-17.859 1.055 -6.836>, rot=<0 0 0 1>}
 def_arm_tp_r_end {parent=def_arm_tp_r, loc=<0 0 0>, rot=<0 0 0 1>}
 def_arm_fr_r {parent=clavicle_r, loc=<-16.81 -7.219 2.994>, rot=<0 0 0 1>}
 def_arm_fr_r_end {parent=def_arm_fr_r, loc=<0 0 0>, rot=<0 0 0 1>}
 def_arm_ba_r {parent=clavicle_r, loc=<-17.803 9.131 2.93>, rot=<0 0 0 1>}
 def_arm_ba_r_end {parent=def_arm_ba_r, loc=<0 0 0>, rot=<0 0 0 1>}
 deform_lapel_l {parent=spine_05, loc=<3.4312 9.6078 -9.8701>, rot=<0.6439 0.0445 0.682 0.344>}
 deform_collar_l {parent=spine_05, loc=<12.9452 4.4339 -8.6556>, rot=<-0.6559 -0.2447 -0.141 -0.7>}
 dyn_collar_l {parent=deform_collar_l, loc=<0 0 0>, rot=<0 0 0 1>}
 dyn_collar_1_l {parent=dyn_collar_l, loc=<0.4733 -3.1657 -3.8072>, rot=<0.0461 0.2215 -0.2622 -0.9381>}
 deform_collar_r {parent=spine_05, loc=<13.1926 4.3403 7.7495>, rot=<-0.2772 0.648 0.702 -0.1025>}
 dyn_collar_r {parent=deform_collar_r, loc=<0 0 0>, rot=<0 0 0 1>}
 deform_lapel_r {parent=spine_05, loc=<3.4841 9.7278 10.028>, rot=<0.6678 0.2095 -0.5976 -0.3911>}
 collar_l {parent=spine_05, loc=<13.0069 2.1738 -7.5576>, rot=<0 0 -0.2003 -0.9797>}
 collar_l_end {parent=collar_l, loc=<0 0 0>, rot=<0 0 0 1>}
 flap_fr_l_MSTR {parent=collar_l, loc=<0.8856 7.2152 -1.3091>, rot=<0.0443 -0.7514 0.0427 -0.657>}
 flap_fr_l {parent=flap_fr_l_MSTR, loc=<0 0 0>, rot=<0 0 0 1>}
 flap_fr_l_end {parent=flap_fr_l, loc=<2.562 0 0>, rot=<0 0.2228 0 -0.9749>}
 flap_l_MSTR {parent=collar_l, loc=<1.0871 -0.7137 -0.3885>, rot=<0 -0.8235 0 -0.5673>}
 flap_l {parent=flap_l_MSTR, loc=<0 0 0>, rot=<0 0 0 1>}
 collar_r {parent=spine_05, loc=<13.0069 2.1738 7.558>, rot=<0 0 -0.2003 -0.9797>}
 collar_r_end {parent=collar_r, loc=<0 0 0>, rot=<0 0 0 1>}
 flap_fr_r_MSTR {parent=collar_r, loc=<0.9054 6.8336 1.4519>, rot=<0.0747 -0.9056 -0.0373 0.4157>}
 flap_fr_r {parent=flap_fr_r_MSTR, loc=<0 0 0>, rot=<0 0 0 1>}
 flap_fr_r_end {parent=flap_fr_r, loc=<3.9967 0 0>, rot=<0 -0.0857 0 -0.9963>}
 flap_r_MSTR {parent=collar_r, loc=<1.6664 -0.9642 -0.0424>, rot=<0 0.8372 0 -0.547>}
 flap_r {parent=flap_r_MSTR, loc=<0 0 0>, rot=<0 0 0 1>}
 collar_ba {parent=spine_05, loc=<16.0415 -5.1253 0>, rot=<0 0 -0.2003 -0.9797>}
 def_arm_bt_l {parent=spine_05, loc=<-7.569 -2.4642 -18.1899>, rot=<0.0465 -0.7289 -0.009 -0.6829>}
 def_arm_bt_l_end {parent=def_arm_bt_l, loc=<0 0 0>, rot=<0 0 0 1>}
 def_arm_bt_r {parent=spine_05, loc=<-7.5691 -2.4638 18.1901>, rot=<0.7289 0.0465 -0.6829 0.009>}
 def_arm_bt_r_end {parent=def_arm_bt_r, loc=<0 0 0>, rot=<0 0 0 1>}
 lapel_r {parent=spine_05, loc=<1.4269 13.8503 6.61>, rot=<0 0 -0.0401 -0.9992>}
 lapel_l {parent=spine_05, loc=<1.4269 13.8503 -6.61>, rot=<0 0 -0.0401 -0.9992>}
 deform_pec_aim_base_l {parent=spine_04, loc=<7.8732 12.6388 -2.4123>, rot=<0.0102 -0.4602 0.3979 -0.7936>}
 deform_pec_aim_l {parent=deform_pec_aim_base_l, loc=<0 0 0>, rot=<0 0 0 1>}
 deform_pec_base_l {parent=deform_pec_aim_l, loc=<10.766 0 0>, rot=<0 0 0 1>}
 deform_pec_l {parent=deform_pec_base_l, loc=<0.055 0 0>, rot=<0 0 0 1>}
 deform_pec_aim_base_r {parent=spine_04, loc=<7.8727 12.6388 2.4123>, rot=<-0.4602 -0.0102 0.7936 0.3979>}
 deform_pec_aim_r {parent=deform_pec_aim_base_r, loc=<0 0 0>, rot=<0 0 0 1>}
 deform_pec_base_r {parent=deform_pec_aim_r, loc=<-10.7662 -0.0006 -0.0007>, rot=<0 0 0 1>}
 deform_pec_r {parent=deform_pec_base_r, loc=<0 0 0>, rot=<0 0 0 1>}
 jacket {parent=spine_04, loc=<10.0146 0 0>, rot=<0 0 -0.0015 -1>}
 fold_01_fr_l_MSTR {parent=spine_03, loc=<-1.0505 12.5735 -7.7969>, rot=<0 0 0.9985 -0.0547>}
 fold_01_fr_l {parent=fold_01_fr_l_MSTR, loc=<0 0 0>, rot=<0 0 0 1>}
 fold_02_fr_l {parent=fold_01_fr_l, loc=<6.0078 0 0>, rot=<0 0 0 1>}
 fold_03_fr_l {parent=fold_02_fr_l, loc=<7.2452 0 0>, rot=<0 0 0 1>}
 fold_04_fr_l {parent=fold_03_fr_l, loc=<6.7772 0 0>, rot=<0 0 0 1>}
 fold_01_fr_r_MSTR {parent=spine_03, loc=<-1.0505 12.5735 7.797>, rot=<0 0 0.9985 -0.0547>}
 fold_01_fr_r {parent=fold_01_fr_r_MSTR, loc=<0 0 0>, rot=<0 0 0 1>}
 fold_02_fr_r {parent=fold_01_fr_r, loc=<6.0078 0 0>, rot=<0 0 0 1>}
 fold_03_fr_r {parent=fold_02_fr_r, loc=<7.2452 0 0>, rot=<0 0 0 1>}
 fold_04_fr_r {parent=fold_03_fr_r, loc=<6.7772 0 0>, rot=<0 0 0 1>}
 fold_01_side_l_MSTR {parent=spine_03, loc=<-0.0211 0.7385 -17.1498>, rot=<-0.0387 0.706 0.706 -0.0387>}
 fold_01_side_l {parent=fold_01_side_l_MSTR, loc=<0 0 0>, rot=<0 0 0 1>}
 fold_02_side_l {parent=fold_01_side_l, loc=<6.0078 0 0>, rot=<0 0 0 1>}
 fold_03_side_l {parent=fold_02_side_l, loc=<7.2452 0 0>, rot=<0 0 0 1>}
 fold_04_side_l {parent=fold_03_side_l, loc=<6.7772 0 0>, rot=<0 0 0 1>}
 fold_01_diag_fr_l_MSTR {parent=spine_03, loc=<-0.6589 9.0118 -13.1444>, rot=<-0.0176 0.321 0.9455 -0.0518>}
 fold_01_diag_fr_l {parent=fold_01_diag_fr_l_MSTR, loc=<0 0 0>, rot=<0 0 0 1>}
 fold_02_diag_fr_l {parent=fold_01_diag_fr_l, loc=<6.0078 0 0>, rot=<0 0 0 1>}
 fold_03_diag_fr_l {parent=fold_02_diag_fr_l, loc=<7.2452 0 0>, rot=<0 0 0 1>}
 fold_04_diag_fr_l {parent=fold_03_diag_fr_l, loc=<6.7772 0 0>, rot=<0 0 0 1>}
 fold_01_ba_ct_MSTR {parent=spine_03, loc=<1.2158 -14.0855 0>, rot=<0.0547 -0.9985 0 0>}
 fold_01_ba_ct {parent=fold_01_ba_ct_MSTR, loc=<0 0 0>, rot=<0 0 0 1>}
 fold_02_ba_ct {parent=fold_01_ba_ct, loc=<6.0078 0 0>, rot=<0 0 0 1>}
 fold_03_ba_ct {parent=fold_02_ba_ct, loc=<7.2452 0 0>, rot=<0 0 0 1>}
 fold_04_ba_ct {parent=fold_03_ba_ct, loc=<6.7772 0 0>, rot=<0 0 0 1>}
 fold_01_leg_ba_l_MSTR {parent=spine_03, loc=<1.2121 -13.2719 -9.0047>, rot=<0.0547 -0.9985 0 0>}
 fold_01_leg_ba_l {parent=fold_01_leg_ba_l_MSTR, loc=<0 0 0>, rot=<0 0 0 1>}
 fold_02_leg_ba_l {parent=fold_01_leg_ba_l, loc=<6.0078 0 0>, rot=<0 0 0 1>}
 fold_03_leg_ba_l {parent=fold_02_leg_ba_l, loc=<7.2452 0 0>, rot=<0 0 0 1>}
 fold_04_leg_ba_l {parent=fold_03_leg_ba_l, loc=<6.7772 0 0>, rot=<0 0 0 1>}
 fold_01_leg_ba_r_MSTR {parent=spine_03, loc=<1.2121 -13.2719 9.005>, rot=<0.0547 -0.9985 0 0>}
 fold_01_leg_ba_r {parent=fold_01_leg_ba_r_MSTR, loc=<0 0 0>, rot=<0 0 0 1>}
 fold_02_leg_ba_r {parent=fold_01_leg_ba_r, loc=<6.0078 0 0>, rot=<0 0 0 1>}
 fold_03_leg_ba_r {parent=fold_02_leg_ba_r, loc=<7.2452 0 0>, rot=<0 0 0 1>}
 fold_04_leg_ba_r {parent=fold_03_leg_ba_r, loc=<6.7772 0 0>, rot=<0 0 0 1>}
 fold_01_diag_ba_l_MSTR {parent=spine_03, loc=<0.5121 -7.4862 -16.38>, rot=<0.0491 -0.8955 -0.4416 0.0242>}
 fold_01_diag_ba_l {parent=fold_01_diag_ba_l_MSTR, loc=<0 0 0>, rot=<0 0 0 1>}
 fold_02_diag_ba_l {parent=fold_01_diag_ba_l, loc=<6.0078 0 0>, rot=<0 0 0 1>}
 fold_03_diag_ba_l {parent=fold_02_diag_ba_l, loc=<7.2452 0 0>, rot=<0 0 0 1>}
 fold_04_diag_ba_l {parent=fold_03_diag_ba_l, loc=<6.7772 0 0>, rot=<0 0 0 1>}
 fold_01_diag_ba_r_MSTR {parent=spine_03, loc=<0.512 -7.486 16.38>, rot=<0.0491 -0.8955 0.4416 -0.0242>}
 fold_01_diag_ba_r {parent=fold_01_diag_ba_r_MSTR, loc=<0 0 0>, rot=<0 0 0 1>}
 fold_02_diag_ba_r {parent=fold_01_diag_ba_r, loc=<6.0078 0 0>, rot=<0 0 0 1>}
 fold_03_diag_ba_r {parent=fold_02_diag_ba_r, loc=<7.2452 0 0>, rot=<0 0 0 1>}
 fold_04_diag_ba_r {parent=fold_03_diag_ba_r, loc=<6.7772 0 0>, rot=<0 0 0 1>}
 fold_01_side_r_MSTR {parent=spine_03, loc=<-0.0072 0.9023 17.2844>, rot=<0.0387 -0.706 0.706 -0.0387>}
 fold_01_side_r {parent=fold_01_side_r_MSTR, loc=<0 0 0>, rot=<0 0 0 1>}
 fold_02_side_r {parent=fold_01_side_r, loc=<6.0078 0 0>, rot=<0 0 0 1>}
 fold_03_side_r {parent=fold_02_side_r, loc=<7.2452 0 0>, rot=<0 0 0 1>}
 fold_04_side_r {parent=fold_03_side_r, loc=<6.7772 0 0>, rot=<0 0 0 1>}
 fold_01_diag_fr_r_MSTR {parent=spine_03, loc=<-0.6589 9.0118 13.144>, rot=<0.0242 -0.4416 0.8955 -0.0491>}
 fold_01_diag_fr_r {parent=fold_01_diag_fr_r_MSTR, loc=<0 0 0>, rot=<0 0 0 1>}
 fold_02_diag_fr_r {parent=fold_01_diag_fr_r, loc=<6.0078 0 0>, rot=<0 0 0 1>}
 fold_03_diag_fr_r {parent=fold_02_diag_fr_r, loc=<7.2452 0 0>, rot=<0 0 0 1>}
 fold_04_diag_fr_r {parent=fold_03_diag_fr_r, loc=<6.7772 0 0>, rot=<0 0 0 1>}
 thigh_l {parent=pelvis, loc=<-2.5693 -0.0109 -7.7501>, rot=<0.0012 0.0572 0.0155 -0.9982>}
 calf_l {parent=thigh_l, loc=<-39.0484 0 0>, rot=<0.0006 -0.0161 -0.0391 -0.9991>}
 foot_l {parent=calf_l, loc=<-42.561 0.1077 0.4292>, rot=<0.0418 -0.0455 0.0343 -0.9975>}
 ball_l {parent=foot_l, loc=<-6.1599 13.8264 0.6806>, rot=<0.0001 -0.0001 -0.7063 -0.7079>}
 deform_foot_l {parent=foot_l, loc=<0 0 0>, rot=<0.7296 -0.6811 0.0026 -0.0617>}
 dyn_foot_l {parent=deform_foot_l, loc=<4.1363 0 0>, rot=<0 0 0 1>}
 calf_twist_02_l {parent=calf_l, loc=<-18.8064 -0.0967 0.0397>, rot=<0.0237 0.0095 -0.0035 -0.9997>}
 calf_twist_01_l {parent=calf_l, loc=<-35.813 -0.0967 0.0397>, rot=<0.0237 0.0095 -0.0035 -0.9997>}
 hem_l {parent=calf_twist_01_l, loc=<-6 0 0>, rot=<0 0 0 1>}
 deform_knee_in_base_l {parent=calf_l, loc=<1.298 -5.2957 -0.0496>, rot=<0 0 0 -1>}
 deform_knee_in_l {parent=deform_knee_in_base_l, loc=<0 0 0>, rot=<0 0 0 1>}
 deform_knee_out_base_l {parent=calf_l, loc=<0 4 0>, rot=<0 0 0 -1>}
 deform_knee_out_l {parent=deform_knee_out_base_l, loc=<0 0 0>, rot=<0 0 0 1>}
 thigh_twist_01_l {parent=thigh_l, loc=<-21.6303 0 0>, rot=<0.0013 0 0 -1>}
 deform_groin_base_l {parent=thigh_l, loc=<-2.1908 8.1142 -1.7442>, rot=<0 0 0 -1>}
 deform_groin_l {parent=deform_groin_base_l, loc=<0 0 0>, rot=<0 0 0 1>}
 knee_l {parent=thigh_l, loc=<-39.0484 0 0>, rot=<0.0006 -0.0161 -0.0391 -0.9991>}
 knee_ba_l {parent=knee_l, loc=<-0.75 -7.5 0>, rot=<0 0 0 1>}
 knee_fr_tp_l {parent=knee_l, loc=<1.1 6.25 0>, rot=<0 0 0 1>}
 knee_fr_bt_l {parent=knee_l, loc=<-3.8 5.6 0>, rot=<0 0 0 1>}
 thigh_r {parent=pelvis, loc=<-2.5693 -0.0109 7.7501>, rot=<0.0572 -0.0012 0.9982 0.0155>}
 calf_r {parent=thigh_r, loc=<39.0484 0 0>, rot=<0.0006 -0.0161 -0.0391 -0.9991>}
 foot_r {parent=calf_r, loc=<42.561 -0.1077 -0.4292>, rot=<0.0418 -0.0455 0.0343 -0.9975>}
 ball_r {parent=foot_r, loc=<6.1599 -13.8264 -0.6806>, rot=<0.0001 -0.0001 -0.7063 -0.7079>}
 deform_foot_r {parent=foot_r, loc=<0 0 0>, rot=<-0.6811 -0.7296 -0.0617 -0.0026>}
 dyn_foot_r {parent=deform_foot_r, loc=<4.1363 0 0>, rot=<0 0 0 1>}
 calf_twist_02_r {parent=calf_r, loc=<18.8064 0.0967 -0.0397>, rot=<0.0237 0.0095 -0.0035 -0.9997>}
 calf_twist_01_r {parent=calf_r, loc=<35.813 0.0967 -0.0397>, rot=<0.0237 0.0095 -0.0035 -0.9997>}
 hem_r {parent=calf_twist_01_r, loc=<6 0 0>, rot=<0 0 0 1>}
 deform_knee_in_base_r {parent=calf_r, loc=<-1.298 5.2957 0.0496>, rot=<0 0 0 -1>}
 deform_knee_in_r {parent=deform_knee_in_base_r, loc=<0 0 0>, rot=<0 0 0 1>}
 deform_knee_out_base_r {parent=calf_r, loc=<0 -4 0>, rot=<0 0 0 -1>}
 deform_knee_out_r {parent=deform_knee_out_base_r, loc=<0 0 0>, rot=<0 0 0 1>}
 thigh_twist_01_r {parent=thigh_r, loc=<21.6303 0 0>, rot=<0.0013 0 0 -1>}
 deform_groin_base_r {parent=thigh_r, loc=<2.1908 -8.1142 1.7442>, rot=<0 0 0 -1>}
 deform_groin_r {parent=deform_groin_base_r, loc=<0 0 0>, rot=<0 0 0 1>}
 knee_r {parent=thigh_r, loc=<39.0484 0 0>, rot=<0.0006 -0.0161 -0.0391 -0.9991>}
 knee_ba_r {parent=knee_r, loc=<0.75 7.5 0>, rot=<0 0 0 1>}
 knee_fr_tp_r {parent=knee_r, loc=<-1.1 -6.25 0>, rot=<0 0 0 1>}
 knee_fr_bt_r {parent=knee_r, loc=<3.8 -5.6 0>, rot=<0 0 0 1>}
 deform_glute_base_l {parent=pelvis, loc=<-8.8064 -8.8222 -8.4287>, rot=<0.0275 0.0501 -0.4472 -0.8926>}
 deform_glute_l {parent=deform_glute_base_l, loc=<0 0 0>, rot=<0 0 0 1>}
 deform_glute_base_r {parent=pelvis, loc=<-8.8062 -8.8222 8.4287>, rot=<0.0501 -0.0275 0.8926 -0.4472>}
 deform_glute_r {parent=deform_glute_base_r, loc=<0 0 0>, rot=<0 0 0 1>}
 belt {parent=pelvis, loc=<0 0 0>, rot=<0 0 0 1>}
 belt_fr_l {parent=belt, loc=<-2.6206 9.1871 -8.9295>, rot=<0 0 0 1>}
 belt_fr_r {parent=belt, loc=<-2.6206 9.1871 8.93>, rot=<0 0 0 1>}
 belt_l {parent=belt, loc=<-1.5433 0.8525 -16.1077>, rot=<0 0 0 1>}
 belt_r {parent=belt, loc=<-1.5433 0.8525 16.108>, rot=<0 0 0 1>}
 belt_fr_ct {parent=belt, loc=<-2.7259 10.9431 0>, rot=<0 0 0 1>}
 belt_ba_l {parent=belt, loc=<-1.5565 -11.9027 -8.389>, rot=<0 0 0 1>}
 belt_ba_r {parent=belt, loc=<-1.5565 -11.9027 8.389>, rot=<0 0 0 1>}
 groin {parent=pelvis, loc=<-13.8038 4.0879 0>, rot=<0 0 0 1>}
 leg_crease_fr_l {parent=pelvis, loc=<-12.2219 8.8627 -9.441>, rot=<0 0 0 1>}
 leg_crease_fr_r {parent=pelvis, loc=<-12.2219 8.8627 9.441>, rot=<0 0 0 1>}
 glute_l {parent=pelvis, loc=<-7.8694 -11.8543 -8.2823>, rot=<0 0 0 1>}
 glute_r {parent=pelvis, loc=<-7.8694 -11.8543 8.282>, rot=<0 0 0 1>}
 leg_crease_l {parent=pelvis, loc=<-9.7842 0.7492 -16.9203>, rot=<0 0 0 1>}
 leg_crease_r {parent=pelvis, loc=<-9.7842 0.7492 16.92>, rot=<0 0 0 1>}
]
variables [
 Emotes {type=bool, default=False}
 ModifyTransforms_30_ItemToModify_0_Transform {type=FTransform, default=(Rotation=(X=0,Y=0,Z=0,W=1),Translation=(X=-5,Y=0,Z=0),Scale3D=(X=1,Y=1,Z=1))}
 Pancake_Alpha {type=bool, default=False}
 Skydiving {type=bool, default=False}
]
