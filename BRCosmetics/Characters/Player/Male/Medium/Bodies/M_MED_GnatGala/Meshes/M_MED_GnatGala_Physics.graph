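# Physics asset for the medium male GnatGala body cosmetic, giving its hanging garment folds rigid-body physics.
# Eleven four-bone capsule chains ring the body, each hanging by limited joints from a master bone locked to spine_03.
# Pelvis, spine and thigh capsules supply body shapes.
# Limited cross-constraints tie front, diagonal and side chains at every level, while the back links to neighbors only at the bottom row through free constraints.
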PhysicsAsset /BRCosmetics/Characters/Player/Male/Medium/Bodies/M_MED_GnatGala/Meshes/M_MED_GnatGala_Physics
bodies {
 fold_01_ba_ct [{shapeType=Capsule, length=9.6476, radius=5, shapeName=fold_01_fr_l_capsule, center=<4.9923 -3.0536 0>, rotation=<rot 90 0 0>}]
 fold_01_ba_ct_MSTR [{shapeType=Capsule, length=0, radius=2.5, shapeName=fold_01_fr_l_MSTR_capsule}]
 fold_01_diag_ba_l [{shapeType=Capsule, length=9.6476, radius=5, shapeName=fold_01_fr_l_capsule, center=<2.6752 -3.1656 0>, rotation=<rot 90 0 0>}]
 fold_01_diag_ba_l_MSTR [{shapeType=Capsule, length=0, radius=2.5, shapeName=fold_01_fr_l_MSTR_capsule}]
 fold_01_diag_ba_r [{shapeType=Capsule, length=9, radius=5, shapeName=fold_01_fr_l_capsule, center=<-0.3565 -3.1656 0>, rotation=<rot 90 0 0>}]
 fold_01_diag_ba_r_MSTR [{shapeType=Capsule, length=0, radius=2.5, shapeName=fold_01_fr_l_MSTR_capsule, rotation=<rot 90 0 0>}]
 fold_01_diag_fr_l [{shapeType=Capsule, length=7, radius=5, shapeName=fold_01_fr_l_capsule, center=<4.5483 -3.7698 0>}]
 fold_01_diag_fr_l_MSTR [{shapeType=Capsule, length=0, radius=2.5, shapeName=fold_01_fr_l_MSTR_capsule}]
 fold_01_diag_fr_r [{shapeType=Capsule, length=7, radius=5, shapeName=fold_01_fr_l_capsule, center=<4.5483 -3.7698 0>}]
 fold_01_diag_fr_r_MSTR [{shapeType=Capsule, length=0, radius=2.5, shapeName=fold_01_fr_l_MSTR_capsule}]
 fold_01_fr_l [{shapeType=Capsule, length=1.9928, radius=4.0498, shapeName=fold_01_fr_l_capsule, center=<3.9214 -2.1525 -3.1176>, rotation=<rot 90 0 0>}]
 fold_01_fr_l_MSTR [{shapeType=Capsule, length=0, radius=2.5, shapeName=fold_01_fr_l_MSTR_capsule}]
 fold_01_fr_r [{shapeType=Capsule, length=1.9928, radius=4.0498, shapeName=fold_01_fr_l_capsule, center=<3.9214 -2.1525 3.1176>, rotation=<rot 90 0 0>}]
 fold_01_fr_r_MSTR [{shapeType=Capsule, length=0, radius=2.5, shapeName=fold_01_fr_l_MSTR_capsule}]
 fold_01_leg_ba_l [{shapeType=Capsule, length=9.6476, radius=5, shapeName=fold_01_fr_l_capsule, center=<4.9923 -3.0536 0>, rotation=<rot 90 0 0>}]
 fold_01_leg_ba_l_MSTR [{shapeType=Capsule, length=0, radius=2.5, shapeName=fold_01_fr_l_MSTR_capsule}]
 fold_01_leg_ba_r [{shapeType=Capsule, length=9.6476, radius=5, shapeName=fold_01_fr_l_capsule, center=<4.9923 -2.2214 -1.8152>, rotation=<rot 90 0 0>}]
 fold_01_leg_ba_r_MSTR [{shapeType=Capsule, length=0, radius=2.5, shapeName=fold_01_fr_l_MSTR_capsule}]
 fold_01_side_l [{shapeType=Capsule, length=9.6476, radius=5, shapeName=fold_01_fr_l_capsule, center=<4.5483 -3.7698 0>}]
 fold_01_side_l_MSTR [{shapeType=Capsule, length=0, radius=2.5, shapeName=fold_01_fr_l_MSTR_capsule}]
 fold_01_side_r [{shapeType=Capsule, length=9.6476, radius=5, shapeName=fold_01_fr_l_capsule, center=<4.5483 -3.7698 0>}]
 fold_01_side_r_MSTR [{shapeType=Capsule, length=0, radius=2.5, shapeName=fold_01_fr_l_MSTR_capsule}]
 fold_02_ba_ct [{shapeType=Capsule, length=9.6476, radius=5, shapeName=fold_02_fr_l_capsule, center=<4.9923 -3.0536 0>, rotation=<rot 90 0 0>}]
 fold_02_diag_ba_l [{shapeType=Capsule, length=9.6476, radius=5, shapeName=fold_02_fr_l_capsule, center=<2.6752 -3.1656 0>, rotation=<rot 90 0 0>}]
 fold_02_diag_ba_r [{shapeType=Capsule, length=9, radius=5, shapeName=fold_02_fr_l_capsule, center=<-0.3565 -3.1656 0>, rotation=<rot 90 0 0>}]
 fold_02_diag_fr_l [{shapeType=Capsule, length=7, radius=5, shapeName=fold_02_fr_l_capsule, center=<4.5483 -3.7698 0>}]
 fold_02_diag_fr_r [{shapeType=Capsule, length=7, radius=5, shapeName=fold_02_fr_l_capsule, center=<4.5483 -3.7698 0>}]
 fold_02_fr_l [{shapeType=Capsule, length=6.3978, radius=4.4119, shapeName=fold_02_fr_l_capsule, center=<3.9214 -2.8921 -3.392>, rotation=<rot 90 0 0>}]
 fold_02_fr_r [{shapeType=Capsule, length=6.7599, radius=4.0498, shapeName=fold_02_fr_l_capsule, center=<3.9214 -2.1525 3.1176>, rotation=<rot 90 0 0>}]
 fold_02_leg_ba_l [{shapeType=Capsule, length=9.6476, radius=5, shapeName=fold_02_fr_l_capsule, center=<4.9923 -3.0536 0>, rotation=<rot 90 0 0>}]
 fold_02_leg_ba_r [{shapeType=Capsule, length=9.6476, radius=5, shapeName=fold_02_fr_l_capsule, center=<4.9923 -2.2214 -1.8152>, rotation=<rot 90 0 0>}]
 fold_02_side_l [{shapeType=Capsule, length=9.6476, radius=5, shapeName=fold_02_fr_l_capsule, center=<4.5483 -3.7698 0>}]
 fold_02_side_r [{shapeType=Capsule, length=9.6476, radius=5, shapeName=fold_02_fr_l_capsule, center=<4.5483 -3.7698 0>}]
 fold_03_ba_ct [{shapeType=Capsule, length=9.6476, radius=5, shapeName=fold_03_fr_l_capsule, center=<4.9923 -3.0536 0>, rotation=<rot 90 0 0>}]
 fold_03_diag_ba_l [{shapeType=Capsule, length=9.6476, radius=5, shapeName=fold_03_fr_l_capsule, center=<2.6752 -3.1656 0>, rotation=<rot 90 0 0>}]
 fold_03_diag_ba_r [{shapeType=Capsule, length=9, radius=5, shapeName=fold_03_fr_l_capsule, center=<1.9683 -3.1656 0>, rotation=<rot 90 0 0>}]
 fold_03_diag_fr_l [{shapeType=Capsule, length=7, radius=5, shapeName=fold_03_fr_l_capsule, center=<4.5483 -3.7698 0>}]
 fold_03_diag_fr_r [{shapeType=Capsule, length=7, radius=5, shapeName=fold_03_fr_l_capsule, center=<4.5483 -3.7698 0>}]
 fold_03_fr_l [{shapeType=Capsule, length=8.6684, radius=4.4119, shapeName=fold_03_fr_l_capsule, center=<3.9214 -3.7414 -2.5156>, rotation=<rot 90 0 0>}]
 fold_03_fr_r [{shapeType=Capsule, length=6.7599, radius=4.0498, shapeName=fold_03_fr_l_capsule, center=<3.9214 -2.1525 3.1176>, rotation=<rot 90 0 0>}]
 fold_03_leg_ba_l [{shapeType=Capsule, length=9.6476, radius=5, shapeName=fold_03_fr_l_capsule, center=<4.9923 -3.0536 0>, rotation=<rot 90 0 0>}]
 fold_03_leg_ba_r [{shapeType=Capsule, length=9.6476, radius=5, shapeName=fold_03_fr_l_capsule, center=<4.9923 -2.2214 -1.8152>, rotation=<rot 90 0 0>}]
 fold_03_side_l [{shapeType=Capsule, length=9.6476, radius=5, shapeName=fold_03_fr_l_capsule, center=<4.5483 -3.7698 0>}]
 fold_03_side_r [{shapeType=Capsule, length=9.6476, radius=5, shapeName=fold_03_fr_l_capsule, center=<5 -3.7698 0>}]
 fold_04_ba_ct [{shapeType=Capsule, length=9.6476, radius=5, shapeName=fold_04_fr_l_capsule, center=<4.9923 -3.0536 0>, rotation=<rot 90 0 0>}]
 fold_04_diag_ba_l [{shapeType=Capsule, length=9.6476, radius=5, shapeName=fold_04_fr_l_capsule, center=<4.3276 -3.1656 0>, rotation=<rot 90 0 0>}]
 fold_04_diag_ba_r [{shapeType=Capsule, length=9, radius=5, shapeName=fold_04_fr_l_capsule, center=<2.9683 -3.1656 0>, rotation=<rot 90 0 0>}]
 fold_04_diag_fr_l [
  {shapeType=Capsule, length=7, radius=5, shapeName=fold_04_fr_l_capsule, center=<4.5483 -3.7698 0>}
  {shapeType=Capsule, length=7, radius=5, shapeName=fold_04_diag_fr_l_capsule, center=<10.8385 -4.7249 0>}
 ]
 fold_04_diag_fr_r [
  {shapeType=Capsule, length=7, radius=5, shapeName=fold_04_fr_l_capsule, center=<4.5483 -3.7698 0>}
  {shapeType=Capsule, length=7, radius=5, shapeName=fold_04_diag_fr_r_capsule, center=<8.9808 -3.1815 0>}
 ]
 fold_04_fr_l [{shapeType=Capsule, length=11.4025, radius=4.4119, shapeName=fold_04_fr_l_capsule, center=<5.3266 -3.7414 -2.5156>, rotation=<rot 90 0 0>}]
 fold_04_fr_r [{shapeType=Capsule, length=10.3369, radius=4.0498, shapeName=fold_04_fr_l_capsule, center=<4.4718 -2.1525 3.1176>, rotation=<rot 90 0 0>}]
 fold_04_leg_ba_l [{shapeType=Capsule, length=9.6476, radius=5, shapeName=fold_04_fr_l_capsule, center=<4.9923 -3.0536 0>, rotation=<rot 90 0 0>}]
 fold_04_leg_ba_r [{shapeType=Capsule, length=9.6476, radius=5, shapeName=fold_04_fr_l_capsule, center=<4.9923 -2.2214 -1.8152>, rotation=<rot 90 0 0>}]
 fold_04_side_l [{shapeType=Capsule, length=9.6476, radius=5, shapeName=fold_04_fr_l_capsule, center=<4.5483 -3.7698 0>}]
 fold_04_side_r [{shapeType=Capsule, length=9.6476, radius=5, shapeName=fold_04_fr_l_capsule, center=<6 -3.1656 0>}]
 pelvis [
  {shapeType=Capsule, length=8.75, radius=11.5408, shapeName=pelvis_capsule, center=<0 -1.9445 0>, rotation=<rot 0 -90 0>}
  {shapeType=Capsule, length=13, radius=13, shapeName=pelvis_capsule, center=<1.8465 0 0>, rotation=<rot 90 0 0>}
  {shapeType=Capsule, length=13, radius=10, shapeName=pelvis_capsule, center=<-7.4887 -2.4499 0>}
 ]
 spine_01 [{shapeType=Capsule, length=10, radius=10.9411, shapeName=pelvis_capsule, center=<0 -0.7017 0>, rotation=<rot 0 -90 0>}]
 spine_02 [{shapeType=Capsule, length=11.25, radius=10.6116, shapeName=pelvis_capsule, center=<0 -1.03 0>, rotation=<rot 0 -90 0>}]
 spine_03 [{shapeType=Capsule, length=9.8892, radius=10.0897, shapeName=pelvis_capsule, rotation=<rot 0 -90 0>}]
 thigh_l [
  {
   shapeType=Capsule
   length=32.3491
   radius=10.5327
   shapeName=thigh_l_capsule
   center=<-18.7523 -0.8999 1.9066>
   rotation=<rot 87.1875 2.8123 0.0001>
  }
  {
   shapeType=Capsule
   length=17.6273
   radius=10.7527
   shapeName=thigh_l_capsule
   center=<-15.9142 0.4936 8.3408>
   rotation=<rot 88.0414 -44.5973 -39.079>
  }
 ]
 thigh_r [
  {
   shapeType=Capsule
   length=32.3491
   radius=10.5327
   shapeName=thigh_r_capsule
   center=<18.7523 0.8999 -1.9066>
   rotation=<rot 87.1875 2.8123 -179.9999>
  }
  {
   shapeType=Capsule
   length=17.6273
   radius=10.7527
   shapeName=thigh_r_capsule
   center=<15.9142 -0.4936 -8.3408>
   rotation=<rot 88.0414 -44.5973 140.921>
  }
 ]
}
constraints [
 {
  bone1Name=spine_01
  bone2Name=pelvis
  swing1LimitDegrees=45
  swing1Motion=Free
  swing2LimitDegrees=45
  swing2Motion=Free
  twistLimitDegrees=45
  twistMotion=Free
 }
 {
  bone1Name=spine_02
  bone2Name=spine_01
  swing1LimitDegrees=45
  swing1Motion=Free
  swing2LimitDegrees=45
  swing2Motion=Free
  twistLimitDegrees=45
  twistMotion=Free
 }
 {
  bone1Name=spine_03
  bone2Name=spine_02
  swing1LimitDegrees=45
  swing1Motion=Free
  swing2LimitDegrees=45
  swing2Motion=Free
  twistLimitDegrees=45
  twistMotion=Free
 }
 {
  bone1Name=thigh_l
  bone2Name=pelvis
  swing1LimitDegrees=45
  swing1Motion=Free
  swing2LimitDegrees=45
  swing2Motion=Free
  twistLimitDegrees=45
  twistMotion=Free
 }
 {
  bone1Name=thigh_r
  bone2Name=pelvis
  swing1LimitDegrees=45
  swing1Motion=Free
  swing2LimitDegrees=45
  swing2Motion=Free
  twistLimitDegrees=45
  twistMotion=Free
 }
 {
  bone1Name=fold_01_fr_l_MSTR
  bone2Name=spine_03
  swing1LimitDegrees=45
  swing1Motion=Locked
  swing2LimitDegrees=45
  swing2Motion=Locked
  twistLimitDegrees=45
  twistMotion=Locked
 }
 {bone1Name=fold_01_fr_l, bone2Name=fold_01_fr_l_MSTR, swing1LimitDegrees=30, swing2LimitDegrees=15, twistLimitDegrees=5}
 {bone1Name=fold_02_fr_l, bone2Name=fold_01_fr_l, swing1LimitDegrees=45, swing2LimitDegrees=15, twistLimitDegrees=5}
 {bone1Name=fold_03_fr_l, bone2Name=fold_02_fr_l, swing1LimitDegrees=45, swing2LimitDegrees=15, twistLimitDegrees=5}
 {bone1Name=fold_04_fr_l, bone2Name=fold_03_fr_l, swing1LimitDegrees=25, swing2LimitDegrees=10, twistLimitDegrees=5}
 {
  bone1Name=fold_01_fr_r_MSTR
  bone2Name=spine_03
  swing1LimitDegrees=45
  swing1Motion=Locked
  swing2LimitDegrees=45
  swing2Motion=Locked
  twistLimitDegrees=45
  twistMotion=Locked
 }
 {bone1Name=fold_01_fr_r, bone2Name=fold_01_fr_r_MSTR, swing1LimitDegrees=30, swing2LimitDegrees=15, twistLimitDegrees=5}
 {bone1Name=fold_02_fr_r, bone2Name=fold_01_fr_r, swing1LimitDegrees=45, swing2LimitDegrees=15, twistLimitDegrees=5}
 {bone1Name=fold_03_fr_r, bone2Name=fold_02_fr_r, swing1LimitDegrees=45, swing2LimitDegrees=15, twistLimitDegrees=5}
 {bone1Name=fold_04_fr_r, bone2Name=fold_03_fr_r, swing1LimitDegrees=25, swing2LimitDegrees=10, twistLimitDegrees=5}
 {
  bone1Name=fold_01_diag_fr_l_MSTR
  bone2Name=spine_03
  swing1LimitDegrees=45
  swing1Motion=Locked
  swing2LimitDegrees=45
  swing2Motion=Locked
  twistLimitDegrees=45
  twistMotion=Locked
 }
 {bone1Name=fold_01_diag_fr_l, bone2Name=fold_01_diag_fr_l_MSTR, swing1LimitDegrees=30, swing2LimitDegrees=15, twistLimitDegrees=5}
 {bone1Name=fold_02_diag_fr_l, bone2Name=fold_01_diag_fr_l, swing1LimitDegrees=45, swing2LimitDegrees=15, twistLimitDegrees=5}
 {bone1Name=fold_03_diag_fr_l, bone2Name=fold_02_diag_fr_l, swing1LimitDegrees=45, swing2LimitDegrees=15, twistLimitDegrees=5}
 {bone1Name=fold_04_diag_fr_l, bone2Name=fold_03_diag_fr_l, swing1LimitDegrees=25, swing2LimitDegrees=10, twistLimitDegrees=5}
 {
  bone1Name=fold_01_diag_fr_r_MSTR
  bone2Name=spine_03
  swing1LimitDegrees=45
  swing1Motion=Locked
  swing2LimitDegrees=45
  swing2Motion=Locked
  twistLimitDegrees=45
  twistMotion=Locked
 }
 {bone1Name=fold_01_diag_fr_r, bone2Name=fold_01_diag_fr_r_MSTR, swing1LimitDegrees=30, swing2LimitDegrees=15, twistLimitDegrees=5}
 {bone1Name=fold_02_diag_fr_r, bone2Name=fold_01_diag_fr_r, swing1LimitDegrees=45, swing2LimitDegrees=15, twistLimitDegrees=5}
 {bone1Name=fold_03_diag_fr_r, bone2Name=fold_02_diag_fr_r, swing1LimitDegrees=45, swing2LimitDegrees=15, twistLimitDegrees=5}
 {bone1Name=fold_04_diag_fr_r, bone2Name=fold_03_diag_fr_r, swing1LimitDegrees=25, swing2LimitDegrees=10, twistLimitDegrees=5}
 {
  bone1Name=fold_01_side_l_MSTR
  bone2Name=spine_03
  swing1LimitDegrees=45
  swing1Motion=Locked
  swing2LimitDegrees=45
  swing2Motion=Locked
  twistLimitDegrees=45
  twistMotion=Locked
 }
 {bone1Name=fold_01_side_l, bone2Name=fold_01_side_l_MSTR, swing1LimitDegrees=30, swing2LimitDegrees=15, twistLimitDegrees=5}
 {bone1Name=fold_02_side_l, bone2Name=fold_01_side_l, swing1LimitDegrees=45, swing2LimitDegrees=15, twistLimitDegrees=5}
 {bone1Name=fold_03_side_l, bone2Name=fold_02_side_l, swing1LimitDegrees=45, swing2LimitDegrees=15, twistLimitDegrees=5}
 {bone1Name=fold_04_side_l, bone2Name=fold_03_side_l, swing1LimitDegrees=25, swing2LimitDegrees=10, twistLimitDegrees=5}
 {
  bone1Name=fold_01_side_r_MSTR
  bone2Name=spine_03
  swing1LimitDegrees=45
  swing1Motion=Locked
  swing2LimitDegrees=45
  swing2Motion=Locked
  twistLimitDegrees=45
  twistMotion=Locked
 }
 {bone1Name=fold_01_side_r, bone2Name=fold_01_side_r_MSTR, swing1LimitDegrees=30, swing2LimitDegrees=15, twistLimitDegrees=1}
 {bone1Name=fold_02_side_r, bone2Name=fold_01_side_r, swing1LimitDegrees=45, swing2LimitDegrees=15, twistLimitDegrees=1}
 {bone1Name=fold_03_side_r, bone2Name=fold_02_side_r, swing1LimitDegrees=45, swing2LimitDegrees=15, twistLimitDegrees=1}
 {bone1Name=fold_04_side_r, bone2Name=fold_03_side_r, swing1LimitDegrees=25, swing2LimitDegrees=10, twistLimitDegrees=1}
 {
  bone1Name=fold_01_diag_ba_l_MSTR
  bone2Name=spine_03
  swing1LimitDegrees=45
  swing1Motion=Locked
  swing2LimitDegrees=45
  swing2Motion=Locked
  twistLimitDegrees=45
  twistMotion=Locked
 }
 {bone1Name=fold_01_diag_ba_l, bone2Name=fold_01_diag_ba_l_MSTR, swing1LimitDegrees=30, swing2LimitDegrees=25, twistLimitDegrees=25}
 {bone1Name=fold_02_diag_ba_l, bone2Name=fold_01_diag_ba_l, swing1LimitDegrees=35, swing2LimitDegrees=25, twistLimitDegrees=25}
 {bone1Name=fold_03_diag_ba_l, bone2Name=fold_02_diag_ba_l, swing1LimitDegrees=35, swing2LimitDegrees=25, twistLimitDegrees=25}
 {bone1Name=fold_04_diag_ba_l, bone2Name=fold_03_diag_ba_l, swing1LimitDegrees=35, swing2LimitDegrees=15, twistLimitDegrees=5}
 {
  bone1Name=fold_01_diag_ba_r_MSTR
  bone2Name=spine_03
  swing1LimitDegrees=45
  swing1Motion=Locked
  swing2LimitDegrees=45
  swing2Motion=Locked
  twistLimitDegrees=45
  twistMotion=Locked
 }
 {bone1Name=fold_01_diag_ba_r, bone2Name=fold_01_diag_ba_r_MSTR, swing1LimitDegrees=35, swing2LimitDegrees=10, twistLimitDegrees=1}
 {bone1Name=fold_02_diag_ba_r, bone2Name=fold_01_diag_ba_r, swing1LimitDegrees=35, swing2LimitDegrees=10, twistLimitDegrees=1}
 {bone1Name=fold_03_diag_ba_r, bone2Name=fold_02_diag_ba_r, swing1LimitDegrees=35, swing2LimitDegrees=10, twistLimitDegrees=1}
 {bone1Name=fold_04_diag_ba_r, bone2Name=fold_03_diag_ba_r, swing1LimitDegrees=35, swing2LimitDegrees=15, twistLimitDegrees=1}
 {
  bone1Name=fold_01_leg_ba_l_MSTR
  bone2Name=spine_03
  swing1LimitDegrees=45
  swing1Motion=Locked
  swing2LimitDegrees=45
  swing2Motion=Locked
  twistLimitDegrees=45
  twistMotion=Locked
 }
 {bone1Name=fold_01_leg_ba_l, bone2Name=fold_01_leg_ba_l_MSTR, swing1LimitDegrees=15, swing2LimitDegrees=10, twistLimitDegrees=2.5}
 {bone1Name=fold_02_leg_ba_l, bone2Name=fold_01_leg_ba_l, swing1LimitDegrees=35, swing2LimitDegrees=10, twistLimitDegrees=2.5}
 {bone1Name=fold_03_leg_ba_l, bone2Name=fold_02_leg_ba_l, swing1LimitDegrees=35, swing2LimitDegrees=10, twistLimitDegrees=2.5}
 {bone1Name=fold_04_leg_ba_l, bone2Name=fold_03_leg_ba_l, swing1LimitDegrees=35, swing2LimitDegrees=10, twistLimitDegrees=2.5}
 {
  bone1Name=fold_01_leg_ba_r_MSTR
  bone2Name=spine_03
  swing1LimitDegrees=45
  swing1Motion=Locked
  swing2LimitDegrees=45
  swing2Motion=Locked
  twistLimitDegrees=45
  twistMotion=Locked
 }
 {bone1Name=fold_01_leg_ba_r, bone2Name=fold_01_leg_ba_r_MSTR, swing1LimitDegrees=15, swing2LimitDegrees=10, twistLimitDegrees=2.5}
 {bone1Name=fold_02_leg_ba_r, bone2Name=fold_01_leg_ba_r, swing1LimitDegrees=35, swing2LimitDegrees=10, twistLimitDegrees=2.5}
 {bone1Name=fold_03_leg_ba_r, bone2Name=fold_02_leg_ba_r, swing1LimitDegrees=35, swing2LimitDegrees=10, twistLimitDegrees=2.5}
 {bone1Name=fold_04_leg_ba_r, bone2Name=fold_03_leg_ba_r, swing1LimitDegrees=35, swing2LimitDegrees=10, twistLimitDegrees=2.5}
 {bone1Name=fold_04_ba_ct, bone2Name=fold_03_ba_ct, swing1LimitDegrees=35, swing2LimitDegrees=10, twistLimitDegrees=2.5}
 {
  bone1Name=fold_01_ba_ct_MSTR
  bone2Name=spine_03
  swing1LimitDegrees=45
  swing1Motion=Locked
  swing2LimitDegrees=45
  swing2Motion=Locked
  twistLimitDegrees=45
  twistMotion=Locked
 }
 {bone1Name=fold_01_ba_ct, bone2Name=fold_01_ba_ct_MSTR, swing1LimitDegrees=15, swing2LimitDegrees=10, twistLimitDegrees=2.5}
 {bone1Name=fold_02_ba_ct, bone2Name=fold_01_ba_ct, swing1LimitDegrees=35, swing2LimitDegrees=10, twistLimitDegrees=2.5}
 {bone1Name=fold_03_ba_ct, bone2Name=fold_02_ba_ct, swing1LimitDegrees=35, swing2LimitDegrees=10, twistLimitDegrees=2.5}
 {bone1Name=fold_04_diag_fr_l, bone2Name=fold_04_fr_l, swing1LimitDegrees=35, swing2LimitDegrees=10, twistLimitDegrees=10}
 {bone1Name=fold_03_diag_fr_l, bone2Name=fold_03_fr_l, swing1LimitDegrees=35, swing2LimitDegrees=25, twistLimitDegrees=25}
 {bone1Name=fold_02_diag_fr_l, bone2Name=fold_02_fr_l, swing1LimitDegrees=35, swing2LimitDegrees=25, twistLimitDegrees=25}
 {bone1Name=fold_01_diag_fr_l, bone2Name=fold_01_fr_l, swing1LimitDegrees=35, swing2LimitDegrees=25, twistLimitDegrees=25}
 {bone1Name=fold_04_side_l, bone2Name=fold_04_diag_fr_l, swing1LimitDegrees=35, swing2LimitDegrees=25, twistLimitDegrees=25}
 {bone1Name=fold_03_side_l, bone2Name=fold_03_diag_fr_l, swing1LimitDegrees=35, swing2LimitDegrees=25, twistLimitDegrees=25}
 {bone1Name=fold_02_side_l, bone2Name=fold_02_diag_fr_l, swing1LimitDegrees=35, swing2LimitDegrees=25, twistLimitDegrees=25}
 {bone1Name=fold_01_side_l, bone2Name=fold_01_diag_fr_l, swing1LimitDegrees=35, swing2LimitDegrees=25, twistLimitDegrees=25}
 {
  bone1Name=fold_04_side_l
  bone2Name=fold_04_diag_ba_l
  swing1LimitDegrees=45
  swing1Motion=Free
  swing2LimitDegrees=15
  swing2Motion=Free
  twistLimitDegrees=10
  twistMotion=Free
 }
 {bone1Name=fold_01_diag_fr_r, bone2Name=fold_01_fr_r, swing1LimitDegrees=35, swing2LimitDegrees=25, twistLimitDegrees=25}
 {bone1Name=fold_02_diag_fr_r, bone2Name=fold_02_fr_r, swing1LimitDegrees=35, swing2LimitDegrees=25, twistLimitDegrees=25}
 {bone1Name=fold_03_diag_fr_r, bone2Name=fold_03_fr_r, swing1LimitDegrees=35, swing2LimitDegrees=25, twistLimitDegrees=25}
 {bone1Name=fold_04_diag_fr_r, bone2Name=fold_04_fr_r, swing1LimitDegrees=35, swing2LimitDegrees=10, twistLimitDegrees=10}
 {bone1Name=fold_01_side_r, bone2Name=fold_01_diag_fr_r, swing1LimitDegrees=35, swing2LimitDegrees=25, twistLimitDegrees=25}
 {bone1Name=fold_02_side_r, bone2Name=fold_02_diag_fr_r, swing1LimitDegrees=35, swing2LimitDegrees=25, twistLimitDegrees=25}
 {bone1Name=fold_03_side_r, bone2Name=fold_03_diag_fr_r, swing1LimitDegrees=35, swing2LimitDegrees=25, twistLimitDegrees=5}
 {bone1Name=fold_04_side_r, bone2Name=fold_04_diag_fr_r, swing1LimitDegrees=35, swing2LimitDegrees=5, twistLimitDegrees=1}
 {
  bone1Name=fold_04_diag_ba_l
  bone2Name=fold_04_leg_ba_l
  swing1LimitDegrees=45
  swing1Motion=Free
  swing2LimitDegrees=45
  swing2Motion=Free
  twistLimitDegrees=45
  twistMotion=Free
 }
 {
  bone1Name=fold_04_leg_ba_l
  bone2Name=fold_04_ba_ct
  swing1LimitDegrees=45
  swing1Motion=Free
  swing2LimitDegrees=45
  swing2Motion=Free
  twistLimitDegrees=45
  twistMotion=Free
 }
 {
  bone1Name=fold_04_leg_ba_r
  bone2Name=fold_04_ba_ct
  swing1LimitDegrees=45
  swing1Motion=Free
  swing2LimitDegrees=45
  swing2Motion=Free
  twistLimitDegrees=45
  twistMotion=Free
 }
 {
  bone1Name=fold_04_diag_ba_r
  bone2Name=fold_04_leg_ba_r
  swing1LimitDegrees=45
  swing1Motion=Free
  swing2LimitDegrees=45
  swing2Motion=Free
  twistLimitDegrees=45
  twistMotion=Free
 }
 {
  bone1Name=fold_04_diag_ba_r
  bone2Name=fold_04_side_r
  swing1LimitDegrees=45
  swing1Motion=Free
  swing2LimitDegrees=45
  swing2Motion=Free
  twistLimitDegrees=45
  twistMotion=Free
 }
]
