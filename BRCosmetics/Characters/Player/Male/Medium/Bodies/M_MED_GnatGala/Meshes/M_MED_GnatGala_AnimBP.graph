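# Post-process-style AnimBP for the GnatGala male body cosmetic, adding secondary motion to clothing atop the pose copied from the parent mesh.
# Collar AnimDynamics and a Control Rig feed a spine-relative RigidBody, followed by lapel and collar dynamics and additive fold-bone rotations.
# Each update sets gravity and sim-space variables by pancake or front-end state.
# During emotes, the emote_pancake curve halves RigidBody alpha, flags the Control Rig and gates fold tweaks.
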AnimBlueprint /BRCosmetics/Characters/Player/Male/Medium/Bodies/M_MED_GnatGala/Meshes/M_MED_GnatGala_AnimBP
target_skeleton=&/M_MED_GnatGala_Skeleton
graphs [
 AnimGraph {nodes [
   n1 = RigidBody!(
    ComponentPose=Local To Component(LocalPose=Use cached pose 'preRigid'())
    Alpha=Get emote_pancake_alpha() -> (emote_pancake_alpha:float)
    OverrideWorldGravity:Vector="0,0,-980"
    SimSpaceSettings:Sim Space Settings Structure=(WorldAlpha=1,VelocityScaleZ=0.3,DampingAlpha=1,MaxLinearVelocity=3000,MaxAngularVelocity=1500,MaxLinearAcceleration=1000,MaxAngularAcceleration=65,ExternalLinearDragV=(X=1,Y=1,Z=1),ExternalLinearVelocity=(X=0,Y=0,Z=0),ExternalAngularVelocity=(X=0,Y=0,Z=0))
   ) -> (Pose) {
    bClampLinearTranslationLimitToRefPose=true
    bOverrideWorldGravity=true
    overridePhysicsAsset=&/M_MED_GnatGala_Physics
    simulationSpace=BaseBoneSpace
    alphaBoolBlend {blendOption=Linear}
    baseBoneRef {boneName=spine_03}
    overrideWorldGravity {z=-980}
    simSpaceSettings {
     maxAngularAcceleration=65
     maxAngularVelocity=1500
     maxLinearAcceleration=1000
     maxLinearVelocity=3000
     velocityScaleZ=0.3
     worldAlpha=1
     externalLinearDragV=<1 1 1>
    }
   }
   n2 = Anim Dynamics - Bone: deform_lapel_l!(ComponentPose=@n1, Alpha=1, GravityOverride:Vector="1,1,1") -> (Pose) {
    angularDampingOverride=0.999
    angularSpringConstant=200
    bAngularSpring=true
    bGravityOverrideInSimSpace=true
    bOverrideAngularDamping=true
    bOverrideLinearDamping=true
    bUseGravityOverride=true
    linearDampingOverride=1
    simulationSpace=RootRelative
    alphaBoolBlend {blendOption=Linear}
    boundBone {boneName=deform_lapel_l}
    componentAppliedLinearAccClamp=<200 200 200>
    gravityOverride=<1 1 1>
    physicsBodyDefinitions [{
      sphereCollisionRadius=0
      boundBone {boneName=deform_lapel_l}
      boxExtents=<10 3 10>
      constraintSetup {angularLimitsMax=<2 5 0>, angularLimitsMin=<-2 -5 -30>, angularTarget=<1 0 0>}
      localJointOffset=<0.1 0 0>
     }]
    relativeSpaceBone {boneName=spine_05}
   }
   n3 = Anim Dynamics - Bone: deform_lapel_r!(ComponentPose=@n2, Alpha=1, GravityOverride:Vector="1,1,1") -> (Pose) {
    angularDampingOverride=0.999
    angularSpringConstant=200
    bAngularSpring=true
    bGravityOverrideInSimSpace=true
    bOverrideAngularDamping=true
    bOverrideLinearDamping=true
    bUseGravityOverride=true
    linearDampingOverride=1
    simulationSpace=RootRelative
    alphaBoolBlend {blendOption=Linear}
    boundBone {boneName=deform_lapel_r}
    componentAppliedLinearAccClamp=<200 200 200>
    gravityOverride=<1 1 1>
    physicsBodyDefinitions [{
      sphereCollisionRadius=0
      boundBone {boneName=deform_lapel_r}
      boxExtents=<10 3 10>
      constraintSetup {angularLimitsMax=<2 5 0>, angularLimitsMin=<-2 -5 -30>, angularTarget=<1 0 0>}
      localJointOffset=<0.1 0 0>
     }]
    relativeSpaceBone {boneName=spine_05}
   }
   n4 = Anim Dynamics - Bone: dyn_collar_1_l!(ComponentPose=@n3, Alpha=1, GravityOverride:Vector="1,1,1") -> (Pose) {
    angularDampingOverride=0.999
    angularSpringConstant=500
    bAngularSpring=true
    bGravityOverrideInSimSpace=true
    bOverrideAngularDamping=true
    bOverrideLinearDamping=true
    bUseGravityOverride=true
    linearDampingOverride=1
    simulationSpace=RootRelative
    alphaBoolBlend {blendOption=Linear}
    boundBone {boneName=dyn_collar_1_l}
    componentAppliedLinearAccClamp=<200 200 200>
    gravityOverride=<1 1 1>
    physicsBodyDefinitions [{
      sphereCollisionRadius=0
      boundBone {boneName=dyn_collar_1_l}
      boxExtents=<4 7 7>
      constraintSetup {angularLimitsMax=<2 5 15>, angularLimitsMin=<-2 -5 -5>, angularTarget=<1 0 0>}
      localJointOffset=<0 -0.1 0>
     }]
    relativeSpaceBone {boneName=spine_05}
   }
   n5 = Get bIsPlayingEmote()
   n6 = "Transform (Modify) Bone - Bone: fold_02_leg_ba_r"!(ComponentPose=@n4, bAlphaBoolEnabled:bool=@n5, Rotation="0,-5,0") -> (Pose) {
    alphaInputType=Bool
    rotationMode=BMM_Additive
    rotationSpace=BCS_BoneSpace
    alphaBoolBlend {blendInTime=0.1, blendOption=Linear, blendOutTime=0.5}
    boneToModify {boneName=fold_02_leg_ba_r}
    rotation {yaw=-5}
   }
   n7 = "Transform (Modify) Bone - Bone: fold_02_diag_ba_r"!(ComponentPose=@n6, bAlphaBoolEnabled:bool=@n5, Rotation="0,-5,0") -> (Pose) {
    alphaInputType=Bool
    rotationMode=BMM_Additive
    rotationSpace=BCS_BoneSpace
    alphaBoolBlend {blendInTime=0.1, blendOption=Linear, blendOutTime=0.5}
    boneToModify {boneName=fold_02_diag_ba_r}
    rotation {yaw=-5}
   }
   n8 = Get NotPancake() -> (NotPancake:bool)
   n9 = "Transform (Modify) Bone - Bone: fold_01_leg_ba_r_MSTR"!(ComponentPose=@n7, bAlphaBoolEnabled:bool=@n8, Rotation="0,-5,0") -> (Pose) {
    alphaInputType=Bool
    rotationMode=BMM_Additive
    rotationSpace=BCS_BoneSpace
    alphaBoolBlend {blendInTime=0.1, blendOption=Linear, blendOutTime=0.5}
    boneToModify {boneName=fold_01_leg_ba_r_MSTR}
    rotation {yaw=-5}
   }
   n10 = "Transform (Modify) Bone - Bone: fold_01_leg_ba_l_MSTR"!(ComponentPose=@n9, bAlphaBoolEnabled:bool=@n8, Rotation="0,-5,0") -> (Pose) {
    alphaInputType=Bool
    rotationMode=BMM_Additive
    rotationSpace=BCS_BoneSpace
    alphaBoolBlend {blendInTime=0.1, blendOption=Linear, blendOutTime=0.5}
    boneToModify {boneName=fold_01_leg_ba_l_MSTR}
    rotation {yaw=-5}
   }
   n11 = "Transform (Modify) Bone - Bone: fold_01_ba_ct_MSTR"!(ComponentPose=@n10, bAlphaBoolEnabled:bool=@n8, Rotation="0,-5,0") -> (Pose) {
    alphaInputType=Bool
    rotationMode=BMM_Additive
    rotationSpace=BCS_BoneSpace
    alphaBoolBlend {blendInTime=0.1, blendOption=Linear, blendOutTime=0.5}
    boneToModify {boneName=fold_01_ba_ct_MSTR}
    rotation {yaw=-5}
   }
   Output Pose(Result=Component To Local(ComponentPose=@n11)) {layerGroup=None}
   n12 = Anim Dynamics - Bone: dyn_collar_l!(ComponentPose=Local To Component(LocalPose=Copy Pose From Mesh() {bUseAttachedParent=true}), Alpha=1, GravityOverride:Vector="1,1,1") -> (Pose) {
    angularDampingOverride=0.999
    angularSpringConstant=200
    bAngularSpring=true
    bGravityOverrideInSimSpace=true
    bOverrideAngularDamping=true
    bOverrideLinearDamping=true
    bUseGravityOverride=true
    linearDampingOverride=1
    simulationSpace=RootRelative
    alphaBoolBlend {blendOption=Linear}
    boundBone {boneName=dyn_collar_l}
    componentAppliedLinearAccClamp=<200 200 200>
    gravityOverride=<1 1 1>
    physicsBodyDefinitions [{
      sphereCollisionRadius=0
      boundBone {boneName=dyn_collar_l}
      boxExtents=<4 7 7>
      constraintSetup {angularLimitsMax=<2 2 15>, angularLimitsMin=<-2 -2 0>, angularTarget=<1 0 0>}
      localJointOffset=<0 -0.1 0>
     }]
    relativeSpaceBone {boneName=spine_05}
   }
   n13 = Anim Dynamics - Bone: dyn_collar_r!(ComponentPose=@n12, Alpha=1, GravityOverride:Vector="1,1,1") -> (Pose) {
    angularDampingOverride=0.999
    angularSpringConstant=200
    bAngularSpring=true
    bGravityOverrideInSimSpace=true
    bOverrideAngularDamping=true
    bOverrideLinearDamping=true
    bUseGravityOverride=true
    linearDampingOverride=1
    simulationSpace=RootRelative
    alphaBoolBlend {blendOption=Linear}
    boundBone {boneName=dyn_collar_r}
    componentAppliedLinearAccClamp=<200 200 200>
    gravityOverride=<1 1 1>
    physicsBodyDefinitions [{
      sphereCollisionRadius=0
      boundBone {boneName=dyn_collar_r}
      boxExtents=<4 7 7>
      constraintSetup {angularLimitsMax=<2 2 15>, angularLimitsMin=<-2 -2 0>, angularTarget=<1 0 0>}
      localJointOffset=<0 0.1 0>
     }]
    relativeSpaceBone {boneName=spine_05}
   }
   n14 = Control Rig!(
    Source=Component To Local(ComponentPose=@n13)
    Alpha
    Skydiving:bool=Get bIsSkyDiving()
    Emotes:bool=Get bIsPlayingEmote()
    Pancake_Alpha:bool=Get Pancake_CtrlRig_Alpha() -> (Pancake_CtrlRig_Alpha:bool)
   ) -> (Pose) {
    alphaBoolBlend {blendOption=Linear}
    controlRigAssetReference {blueprintRigClass=&/M_MED_GnatGala_CtrlRig1.M_MED_GnatGala_CtrlRig1_C}
   }
   Save cached pose 'preRigid'(Pose=@n14) {cacheName=preRigid}
  ]}
 EventGraph {nodes [
   Event BlueprintUpdateAnimation()
   > n1 = "Gravity Override State Function (Parameter Asset) (Nativized)"(
    CharacterPartAnimInstance=Self-Reference()
    GameTimeInSeconds=GetGameTimeInSeconds()
    Parameters=Select(
     Option 0:Fort Character Dynamics Parameters Object Reference=&/M_MED_GnatGala_Parameters
     Option 1:Fort Character Dynamics Parameters Object Reference=&/M_MED_GnatGala_PancakeParameters
     Index=Get Pancake_CtrlRig_Alpha()
    ) -> (ReturnValue:Fort Character Dynamics Parameters Object Reference)
   )
   > Set Gravity Override(Gravity Override=@n1)
   > n2 = Sequence()
   > n3 = Branch(Condition=Get Pancake_CtrlRig_Alpha())
   > Set!(
    execute
    Sim Space Settings_WorldAlpha:float=0.01
    Sim Space Settings_VelocityScaleZ:float=0.1
    Sim Space Settings_DampingAlpha:float=1
    Sim Space Settings_MaxLinearVelocity:float=2500
    Sim Space Settings_MaxAngularVelocity:float=1
    Sim Space Settings_MaxLinearAcceleration:float=1
    Sim Space Settings_MaxAngularAcceleration:float=1
    Sim Space Settings_ExternalLinearDragV:Vector="0,0,0"
    Sim Space Settings_ExternalLinearVelocity:Vector="0,0,0"
    Sim Space Settings_ExternalAngularVelocity:Vector="0,0,0"
   ) -> (then, Output_Get:Sim Space Settings Structure)
   n4 = Branch(execute=@n3.else, Condition=Get bIsInFrontEnd())
   > Set(
    Sim Space Settings_WorldAlpha=0.07
    Sim Space Settings_VelocityScaleZ=0
    Sim Space Settings_DampingAlpha=1
    Sim Space Settings_MaxLinearVelocity=500
    Sim Space Settings_MaxAngularVelocity=3
    Sim Space Settings_MaxLinearAcceleration=700
    Sim Space Settings_MaxAngularAcceleration=250
    Sim Space Settings_ExternalLinearDragV="0.1,0.1,0"
    Sim Space Settings_ExternalLinearVelocity="0,0,0"
    Sim Space Settings_ExternalAngularVelocity="0,0,0.1"
   )
   Set(
    execute=@n4.else
    Sim Space Settings_WorldAlpha=0.4
    Sim Space Settings_VelocityScaleZ=0.1
    Sim Space Settings_DampingAlpha=1
    Sim Space Settings_MaxLinearVelocity=2500
    Sim Space Settings_MaxAngularVelocity=2500
    Sim Space Settings_MaxLinearAcceleration=1500
    Sim Space Settings_MaxAngularAcceleration=500
    Sim Space Settings_ExternalLinearDragV="0.2,0.2,0"
    Sim Space Settings_ExternalLinearVelocity="0,0,0"
    Sim Space Settings_ExternalAngularVelocity="0,0,0.1"
   )
   n5 = Branch(execute=@n2.then_1, Condition=Get bIsPlayingEmote())
   n6 = GetAnimInstance(self=GetOwnerSkeletalMeshComponent())
   > n7 = Branch(Condition="float >= float"(A=GetCurveValue(self=@n6, CurveName=emote_pancake), B=1))
   > Set emote_pancake_alpha(emote_pancake_alpha:float=0.5) -> (Output_Get:float)
   > Set Pancake_CtrlRig_Alpha(Pancake_CtrlRig_Alpha:bool=true) -> (Output_Get:bool)
   > Set NotPancake(NotPancake:bool=false) -> (Output_Get:bool)
   Set NotPancake(execute=@n7.else, NotPancake=true)
   Set emote_pancake_alpha(execute=@n5.else, emote_pancake_alpha=1)
   > Set Pancake_CtrlRig_Alpha(Pancake_CtrlRig_Alpha=false)
   > Set NotPancake(NotPancake=false)
   Event BlueprintBeginPlay() {is=title K2Node_Event}
   > Set emote_pancake_alpha(emote_pancake_alpha=1)
   Set Alpha(Alpha:float=1) -> (Output_Get:float)
  ]}
]
variables [
 Alpha {type=number}
 emote_pancake_alpha {type=number}
 Gravity Override {type=Vector}
 Gravity Override Params {type=GravityOverrideParamsStruct}
 NotPancake {type=boolean, default=true}
 Pancake_CtrlRig_Alpha {type=boolean}
 RBAN Params {
  type=RBANParamsStruct
  default {
   dBNO {simSpaceSettings {
     maxAngularAcceleration=1000
     maxAngularVelocity=3
     maxLinearAcceleration=300
     maxLinearVelocity=700
     velocityScaleZ=0
     externalLinearDragV=<0 0.5 0>
    }}
   dBNO_Carried {simSpaceSettings {
     maxAngularAcceleration=1000
     maxAngularVelocity=3
     maxLinearAcceleration=300
     maxLinearVelocity=700
     velocityScaleZ=0
     externalLinearDragV=<0 0.5 0>
    }}
   emoteOrMelee {simSpaceSettings {
     maxAngularAcceleration=1000
     maxAngularVelocity=3
     maxLinearAcceleration=300
     maxLinearVelocity=700
     velocityScaleZ=0
     externalLinearDragV=<0 0.5 0>
    }}
   front_End {simSpaceSettings {
     maxAngularAcceleration=1000
     maxAngularVelocity=3
     maxLinearAcceleration=300
     maxLinearVelocity=700
     velocityScaleZ=0
     externalLinearDragV=<0 0.5 0>
    }}
   onGround_Crouch_Moving {simSpaceSettings {
     maxAngularAcceleration=1000
     maxAngularVelocity=3
     maxLinearAcceleration=300
     maxLinearVelocity=700
     velocityScaleZ=0
     externalLinearDragV=<0 0.5 0>
    }}
   onGround_Falling {simSpaceSettings {
     maxAngularAcceleration=1000
     maxAngularVelocity=3
     maxLinearAcceleration=300
     maxLinearVelocity=700
     velocityScaleZ=0
     externalLinearDragV=<0 0.5 0>
    }}
   onGround_Moving {simSpaceSettings {
     maxAngularAcceleration=1000
     maxAngularVelocity=3
     maxLinearAcceleration=300
     maxLinearVelocity=700
     velocityScaleZ=0
     externalLinearDragV=<0 0.5 0>
    }}
   onGround_Standing {simSpaceSettings {
     maxAngularAcceleration=1000
     maxAngularVelocity=3
     maxLinearAcceleration=300
     maxLinearVelocity=700
     velocityScaleZ=0
     externalLinearDragV=<0 0.5 0>
    }}
   rail_Grinding {simSpaceSettings {
     maxAngularAcceleration=1000
     maxAngularVelocity=3
     maxLinearAcceleration=300
     maxLinearVelocity=700
     velocityScaleZ=0
     worldAlpha=0.1
     externalLinearDragV=<0 0.5 0>
    }}
   riding_Moving {simSpaceSettings {
     maxAngularAcceleration=1000
     maxAngularVelocity=3
     maxLinearAcceleration=300
     maxLinearVelocity=700
     velocityScaleZ=0
     externalLinearDragV=<0 0.5 0>
    }}
   riding_Moving_Targeting {simSpaceSettings {
     maxAngularAcceleration=1000
     maxAngularVelocity=3
     maxLinearAcceleration=300
     maxLinearVelocity=700
     velocityScaleZ=0
     externalLinearDragV=<0 0.5 0>
    }}
   riding_Standing {simSpaceSettings {
     maxAngularAcceleration=1000
     maxAngularVelocity=3
     maxLinearAcceleration=300
     maxLinearVelocity=700
     velocityScaleZ=0
     externalLinearDragV=<0 0.5 0>
    }}
   skydive_DivingDown {simSpaceSettings {
     maxAngularAcceleration=1000
     maxAngularVelocity=3
     maxLinearAcceleration=300
     maxLinearVelocity=700
     velocityScaleZ=0
     externalLinearDragV=<0 0.5 0>
    }}
   skydive_FloatingDown {
    componentLinearAccScale=<0 0.5 0>
    componentLinearVelScale=<0 0.5 0>
    simSpaceSettings {
     maxAngularAcceleration=1000
     maxAngularVelocity=3
     maxLinearAcceleration=300
     maxLinearVelocity=700
     velocityScaleZ=0
     externalLinearDragV=<0 0.5 0>
    }
   }
   skydive_GoingUp {simSpaceSettings {
     maxAngularAcceleration=1000
     maxAngularVelocity=3
     maxLinearAcceleration=300
     maxLinearVelocity=700
     velocityScaleZ=0
     externalLinearDragV=<0 0.5 0>
    }}
   skydive_Parachuting {
    componentLinearAccScale=<0 0.1 0.3>
    componentLinearVelScale=<0 0.1 0.3>
    simSpaceSettings {
     maxAngularAcceleration=1000
     maxAngularVelocity=3
     maxLinearAcceleration=300
     maxLinearVelocity=700
     velocityScaleZ=0
     externalLinearDragV=<0 0.5 0>
    }
   }
   skydive_Surfing {simSpaceSettings {
     maxAngularAcceleration=1000
     maxAngularVelocity=3
     maxLinearAcceleration=300
     maxLinearVelocity=700
     velocityScaleZ=0
     externalLinearDragV=<0 0.5 0>
    }}
   swimming {simSpaceSettings {
     maxAngularAcceleration=1000
     maxAngularVelocity=3
     maxLinearAcceleration=300
     maxLinearVelocity=700
     velocityScaleZ=0
     externalLinearDragV=<0 0.5 0>
    }}
   tactical_Sprint {simSpaceSettings {
     maxAngularAcceleration=1000
     maxAngularVelocity=3
     maxLinearAcceleration=300
     maxLinearVelocity=700
     velocityScaleZ=0
     externalLinearDragV=<0 0.5 0>
    }}
   ziplining {simSpaceSettings {
     maxAngularAcceleration=1000
     maxAngularVelocity=3
     maxLinearAcceleration=300
     maxLinearVelocity=700
     velocityScaleZ=0
     externalLinearDragV=<0 0.5 0>
    }}
  }
 }
 Sim Space Settings {
  type=SimSpaceSettings
  default {
   dampingAlpha=1
   maxAngularAcceleration=1000
   maxAngularVelocity=3
   maxLinearAcceleration=700
   maxLinearVelocity=500
   velocityScaleZ=0
   worldAlpha=0.333
   externalAngularVelocity=<0 0 0.1>
   externalLinearDragV=<0.2 0.2 0>
  }
 }
]
