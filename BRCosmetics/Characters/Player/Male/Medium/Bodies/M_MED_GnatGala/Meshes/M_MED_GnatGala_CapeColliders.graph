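# Physics asset for the cape colliders of the medium male GnatGala player body.
# Capsules cover the pelvis, spine, clavicles and neck, with a sphere on the head, while the listed limb bodies have no shapes.
# Five back fold chains (center, diagonal and leg on each side) each link capsule bodies from spine_03 through MSTR, 01 and 04.
# All chain constraints leave swing and twist free, so their 45-degree limits are inactive.
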PhysicsAsset /BRCosmetics/Characters/Player/Male/Medium/Bodies/M_MED_GnatGala/Meshes/M_MED_GnatGala_CapeColliders
bodies {
 calf_l []
 calf_r []
 clavicle_l [{shapeType=Capsule, length=9.0192, radius=9.8871, center=<9.4266 -1.1806 -2.4977>, rotation=<rot 69.3297 1.4476 176.9462>}]
 clavicle_r [{shapeType=Capsule, length=9.0192, radius=9.8871, center=<-9.4266 1.1806 2.4977>, rotation=<rot 69.3297 1.4476 -3.0538>}]
 fold_01_ba_ct [{shapeType=Capsule, length=10, radius=10, shapeName=fold_01_ba_ct_capsule, center=<6.3947 8.1589 0>, rotation=<rot 90 0 0>}]
 fold_01_ba_ct_MSTR [{shapeType=Capsule, length=9.4394, radius=10.5606, shapeName=fold_01_ba_ct_MSTR_capsule, center=<8.0897 8.2472 0>, rotation=<rot 90 0 0>}]
 fold_01_diag_ba_l [{shapeType=Capsule, length=10, radius=10, shapeName=fold_01_diag_ba_l_capsule, center=<6.3947 8.1589 0>, rotation=<rot 90 0 0>}]
 fold_01_diag_ba_l_MSTR [{shapeType=Capsule, length=9.4394, radius=10.5606, shapeName=fold_01_diag_ba_l_MSTR_capsule, center=<8.0897 8.2472 0>, rotation=<rot 90 0 0>}]
 fold_01_diag_ba_r [{shapeType=Capsule, length=10, radius=10, shapeName=fold_01_diag_ba_r_capsule, center=<6.3947 8.1589 0>, rotation=<rot 90 0 0>}]
 fold_01_diag_ba_r_MSTR [{shapeType=Capsule, length=9.4394, radius=10.5606, shapeName=fold_01_diag_ba_r_MSTR_capsule, center=<8.0897 8.2472 0>, rotation=<rot 90 0 0>}]
 fold_01_leg_ba_l [{shapeType=Capsule, length=10, radius=10, shapeName=fold_01_leg_ba_l_capsule, center=<6.3947 8.1589 0>, rotation=<rot 90 0 0>}]
 fold_01_leg_ba_l_MSTR [{shapeType=Capsule, length=9.4394, radius=10.5606, shapeName=fold_01_leg_ba_l_MSTR_capsule, center=<8.0897 8.2472 0>, rotation=<rot 90 0 0>}]
 fold_01_leg_ba_r [{shapeType=Capsule, length=10, radius=10, shapeName=fold_01_leg_ba_r_capsule, center=<6.3947 8.1589 0>, rotation=<rot 90 0 0>}]
 fold_01_leg_ba_r_MSTR [{shapeType=Capsule, length=9.4394, radius=10.5606, shapeName=fold_01_leg_ba_r_MSTR_capsule, center=<8.0897 8.2472 0>, rotation=<rot 90 0 0>}]
 fold_04_ba_ct [{shapeType=Capsule, length=9.4394, radius=10.5606, shapeName=fold_04_ba_ct_capsule, center=<8.0897 9.973 0>, rotation=<rot 90 0 0>}]
 fold_04_diag_ba_l [{shapeType=Capsule, length=9.4394, radius=10.5606, shapeName=fold_04_diag_ba_l_capsule, center=<8.0897 9.973 0>, rotation=<rot 90 0 0>}]
 fold_04_diag_ba_r [{shapeType=Capsule, length=9.4394, radius=10.5606, shapeName=fold_04_diag_ba_r_capsule, center=<8.0897 9.973 0>, rotation=<rot 90 0 0>}]
 fold_04_leg_ba_l [{shapeType=Capsule, length=9.4394, radius=10.5606, shapeName=fold_04_leg_ba_l_capsule, center=<8.0897 9.973 0>, rotation=<rot 90 0 0>}]
 fold_04_leg_ba_r [{shapeType=Capsule, length=9.4394, radius=10.5606, shapeName=fold_04_leg_ba_r_capsule, center=<8.0897 9.973 0>, rotation=<rot 90 0 0>}]
 foot_l []
 foot_r []
 head [{shapeType=Sphere, radius=17.3125, center=<11.028 -0.174 0>}]
 lowerarm_l []
 lowerarm_r []
 neck_02 [{shapeType=Capsule, length=13.4375, radius=7.3175, center=<-0.2642 -0.3408 0>, rotation=<rot 90 -171.1193 -175.3978>}]
 pelvis [{shapeType=Capsule, length=10.3278, radius=15.7769, center=<-2.4887 0.1881 0>}]
 spine_01 [{shapeType=Capsule, length=11.9869, radius=16.1551, center=<-1.7965 0.1884 0>}]
 spine_02 [
  {shapeType=Capsule, length=11.0128, radius=15.5258, center=<-1.711 -0.1851 0.0001>}
  {shapeType=Capsule, length=29.3956, radius=11.5807, center=<2.5024 0.4992 3.9825>, rotation=<rot 90 0.0049 0.0049>}
  {shapeType=Capsule, length=29.875, radius=11.1013, center=<2.4704 0.4991 -3.9825>, rotation=<rot 90 -0.0396 -0.0396>}
 ]
 spine_03 [{shapeType=Capsule, length=8.9477, radius=15.4663, center=<0.1445 0.2396 0.0001>}]
 spine_04 [{shapeType=Capsule, length=14.8807, radius=14.9958, center=<2.766 0.7765 0.0002>}]
 spine_05 [
  {shapeType=Capsule, length=18.3107, radius=15.695, center=<-2.2795 1.8109 0>}
  {shapeType=Capsule, length=13.9518, radius=13.9425, center=<0.9139 0.4452 0>}
 ]
 thigh_l []
 thigh_r []
 upperarm_l []
 upperarm_r []
}
constraints [
 {
  bone1Name=fold_01_ba_ct_MSTR
  bone2Name=spine_03
  swing1LimitDegrees=45
  swing1Motion=Free
  swing2LimitDegrees=45
  swing2Motion=Free
  twistLimitDegrees=45
  twistMotion=Free
 }
 {
  bone1Name=fold_01_diag_ba_l_MSTR
  bone2Name=spine_03
  swing1LimitDegrees=45
  swing1Motion=Free
  swing2LimitDegrees=45
  swing2Motion=Free
  twistLimitDegrees=45
  twistMotion=Free
 }
 {
  bone1Name=fold_01_diag_ba_r_MSTR
  bone2Name=spine_03
  swing1LimitDegrees=45
  swing1Motion=Free
  swing2LimitDegrees=45
  swing2Motion=Free
  twistLimitDegrees=45
  twistMotion=Free
 }
 {
  bone1Name=fold_01_leg_ba_l_MSTR
  bone2Name=spine_03
  swing1LimitDegrees=45
  swing1Motion=Free
  swing2LimitDegrees=45
  swing2Motion=Free
  twistLimitDegrees=45
  twistMotion=Free
 }
 {
  bone1Name=fold_01_leg_ba_r_MSTR
  bone2Name=spine_03
  swing1LimitDegrees=45
  swing1Motion=Free
  swing2LimitDegrees=45
  swing2Motion=Free
  twistLimitDegrees=45
  twistMotion=Free
 }
 {
  bone1Name=fold_04_leg_ba_l
  bone2Name=fold_01_leg_ba_l
  swing1LimitDegrees=45
  swing1Motion=Free
  swing2LimitDegrees=45
  swing2Motion=Free
  twistLimitDegrees=45
  twistMotion=Free
 }
 {
  bone1Name=fold_04_leg_ba_r
  bone2Name=fold_01_leg_ba_r
  swing1LimitDegrees=45
  swing1Motion=Free
  swing2LimitDegrees=45
  swing2Motion=Free
  twistLimitDegrees=45
  twistMotion=Free
 }
 {
  bone1Name=fold_04_diag_ba_r
  bone2Name=fold_01_diag_ba_r
  swing1LimitDegrees=45
  swing1Motion=Free
  swing2LimitDegrees=45
  swing2Motion=Free
  twistLimitDegrees=45
  twistMotion=Free
 }
 {
  bone1Name=fold_04_diag_ba_l
  bone2Name=fold_01_diag_ba_l
  swing1LimitDegrees=45
  swing1Motion=Free
  swing2LimitDegrees=45
  swing2Motion=Free
  twistLimitDegrees=45
  twistMotion=Free
 }
 {
  bone1Name=fold_04_ba_ct
  bone2Name=fold_01_ba_ct
  swing1LimitDegrees=45
  swing1Motion=Free
  swing2LimitDegrees=45
  swing2Motion=Free
  twistLimitDegrees=45
  twistMotion=Free
 }
 {
  bone1Name=fold_01_ba_ct
  bone2Name=fold_01_ba_ct_MSTR
  swing1LimitDegrees=45
  swing1Motion=Free
  swing2LimitDegrees=45
  swing2Motion=Free
  twistLimitDegrees=45
  twistMotion=Free
 }
 {
  bone1Name=fold_01_diag_ba_l
  bone2Name=fold_01_diag_ba_l_MSTR
  swing1LimitDegrees=45
  swing1Motion=Free
  swing2LimitDegrees=45
  swing2Motion=Free
  twistLimitDegrees=45
  twistMotion=Free
 }
 {
  bone1Name=fold_01_diag_ba_r
  bone2Name=fold_01_diag_ba_r_MSTR
  swing1LimitDegrees=45
  swing1Motion=Free
  swing2LimitDegrees=45
  swing2Motion=Free
  twistLimitDegrees=45
  twistMotion=Free
 }
 {
  bone1Name=fold_01_leg_ba_l
  bone2Name=fold_01_leg_ba_l_MSTR
  swing1LimitDegrees=45
  swing1Motion=Free
  swing2LimitDegrees=45
  swing2Motion=Free
  twistLimitDegrees=45
  twistMotion=Free
 }
 {
  bone1Name=fold_01_leg_ba_r
  bone2Name=fold_01_leg_ba_r_MSTR
  swing1LimitDegrees=45
  swing1Motion=Free
  swing2LimitDegrees=45
  swing2Motion=Free
  twistLimitDegrees=45
  twistMotion=Free
 }
]
