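# Control Rig for the GnatGala medium male player body cosmetic, with reusable functions that measure the orientation of driver items.
# The visible functions run SphericalPoseReaders with tuned offsets and falloffs, two of them relative to the driver's parent.
# They return named directional weights such as up, down, front, back, in, out, aim and bend, including pre and post variants.
# Per-direction Viz flags toggle debug drawing and on-screen value printing.
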ControlRigBlueprint /BRCosmetics/Characters/Player/Male/Medium/Bodies/M_MED_GnatGala/Meshes/M_MED_GnatGala_CtrlRig
graphs [
 New Function_ContainedGraph {nodes [
   n1 = RigVMFunctionEntryNode()
   > RigVMVariableNode(Value=/Script/ControlRig.RigUnit_HierarchyGetParent(bDefaultParent=True, Child=@n1.Driver Item) {is=struct RigVMUnitNode}.Parent) {variable=Driver Item Parent}
   To_String = RigVMDispatchNode(Value=@n1.Driver Item.Name) {template="DISPATCH_RigDispatch_ToString(in Value,out Result)"}
   > RigVMVariableNode(Value=@To_String.Result) {variable=Driver Item Name}
   > RigVMFunction_Sequence = RigVMAggregateNode()
   RigVMFunctionReturnNode(
    ExecuteContext=@RigVMFunction_Sequence.F
    Up Post=RigVMVariableNode() {variable=Result Up Post}.Value
    Aim Up In=RigVMVariableNode() {variable=Result Aim Up In}.Value
    Aim Up Ot=RigVMVariableNode() {variable=Result Aim Up Ot}.Value
    Aim Dn In=RigVMVariableNode() {variable=Result Aim Dn In}.Value
    Aim Dn Ot=RigVMVariableNode() {variable=Result Aim Dn Ot}.Value
    Dn=RigVMVariableNode() {variable=Result Dn}.Value
    Dn Pre=RigVMVariableNode() {variable=Result Dn Pre}.Value
    Dn Post=RigVMVariableNode() {variable=Result Dn Post}.Value
    Fr=RigVMVariableNode() {variable=Result Fr}.Value
    Fr Pre=RigVMVariableNode() {variable=Result Fr Pre}.Value
    Fr Post=RigVMVariableNode() {variable=Result Fr Post}.Value
    Fr Up=RigVMVariableNode() {variable=Result Fr Up}.Value
    Fr Dn=RigVMVariableNode() {variable=Result Fr Dn}.Value
    Ba=RigVMVariableNode() {variable=Result Ba}.Value
    Ba Pre=RigVMVariableNode() {variable=Result Ba Pre}.Value
    Ba Post=RigVMVariableNode() {variable=Result Ba Post}.Value
    Up=RigVMVariableNode() {variable=Result Up}.Value
    Ba Up=RigVMVariableNode() {variable=Result Ba Up}.Value
    Ba Dn=RigVMVariableNode() {variable=Result Ba Dn}.Value
    Up Pre=RigVMVariableNode() {variable=Result Up Pre}.Value
   )
   n2 = &RigUnit_SphericalPoseReader(
    ActiveRegionScaleFactors=(PositiveWidth=1,NegativeWidth=1,PositiveHeight=1,NegativeHeight=1)
    ActiveRegionSize=0.001
    Debug=(bDrawDebug=false,bDraw2D=False,bDrawLocalAxes=False,DebugScale=25,DebugSegments=20,DebugThickness=0.05)
    DriverAxis=(X=1,Y=0,Z=0)
    FalloffRegionScaleFactors=(PositiveWidth=1.3471,NegativeWidth=1,PositiveHeight=1,NegativeHeight=1)
    FalloffSize=0.5
    FlipHeightScaling=False
    FlipWidthScaling=False
    RotationOffset=(X=0,Y=31.239,Z=0)
    Debug.bDrawDebug=RigVMVariableNode() {variable=Up Viz}.Value
    DriverItem=RigVMVariableNode() {variable=Driver Item}.Value
    OptionalParentItem=RigVMVariableNode() {variable=Driver Item Parent}.Value
    ExecutePin=@RigVMFunction_Sequence.A
   )
   > n3 = RigVMVariableNode(Value=@n2.OutputParam) {variable=Result Up}
   RigVMFunction_ControlFlowBranch = &RigVMFunction_ControlFlowBranch(ExecuteContext=@n3.ExecuteContext, Condition=RigVMVariableNode() {variable=Up Viz}.Value)
   n4 = &RigUnit_SphericalPoseReader(
    ActiveRegionScaleFactors=(PositiveWidth=1,NegativeWidth=1,PositiveHeight=1,NegativeHeight=1)
    ActiveRegionSize=0.001
    Debug=(bDrawDebug=false,bDraw2D=False,bDrawLocalAxes=False,DebugScale=25,DebugSegments=20,DebugThickness=0.05)
    DriverAxis=(X=1,Y=0,Z=0)
    FalloffRegionScaleFactors=(PositiveWidth=1.6942,NegativeWidth=1,PositiveHeight=2,NegativeHeight=2)
    FalloffSize=0.25
    FlipHeightScaling=False
    FlipWidthScaling=False
    RotationOffset=(X=0,Y=-13.761,Z=0)
    ExecutePin=@RigVMFunction_ControlFlowBranch.Completed
    DriverItem=RigVMVariableNode() {variable=Driver Item}.Value
    OptionalParentItem=RigVMVariableNode() {variable=Driver Item Parent}.Value
    Debug.bDrawDebug=RigVMVariableNode() {variable=Up Pre Viz}.Value
   )
   > n5 = RigVMVariableNode(Value=@n4.OutputParam) {variable=Result Up Pre}
   n6 = &RigVMFunction_ControlFlowBranch(Condition=RigVMVariableNode() {variable=Up Pre Viz}.Value, ExecuteContext=@n5.ExecuteContext)
   n7 = &RigUnit_SphericalPoseReader(
    ActiveRegionScaleFactors=(PositiveWidth=1,NegativeWidth=1,PositiveHeight=1,NegativeHeight=1)
    ActiveRegionSize=0.001
    Debug=(bDrawDebug=false,bDraw2D=False,bDrawLocalAxes=False,DebugScale=25,DebugSegments=20,DebugThickness=0.05)
    DriverAxis=(X=1,Y=0,Z=0)
    FalloffRegionScaleFactors=(PositiveWidth=0.5,NegativeWidth=0.5,PositiveHeight=1,NegativeHeight=1)
    FalloffSize=0.5
    FlipHeightScaling=False
    FlipWidthScaling=False
    RotationOffset=(X=0,Y=76.239,Z=0)
    ExecutePin=@n6.Completed
    Debug.bDrawDebug=RigVMVariableNode() {variable=Up Post Viz}.Value
    DriverItem=RigVMVariableNode() {variable=Driver Item}.Value
    OptionalParentItem=RigVMVariableNode() {variable=Driver Item Parent}.Value
   )
   > n8 = RigVMVariableNode(Value=@n7.OutputParam) {variable=Result Up Post}
   n9 = &RigVMFunction_ControlFlowBranch(Condition=RigVMVariableNode() {variable=Up Post Viz}.Value, ExecuteContext=@n8.ExecuteContext)
   RigVMDispatchNode(
    Enabled=True
    ScreenColor=(R=1,G=1,B=1,A=1)
    ScreenDuration=0.05
    Prefix=/Script/RigVM.RigVMFunction_StringConcat(B=" - UP POST:   ", A=RigVMVariableNode() {variable=Driver Item Name}.Value) {is=struct RigVMUnitNode, template="Concat::Execute(in A,in B,out Result)"}.Result
    Value=RigVMVariableNode() {variable=Result Up Post}.Value
    ExecuteContext=@n9.True
   ) {template="DISPATCH_RigVMDispatch_Print(in Prefix,in Value,in Enabled,in ScreenDuration,in ScreenColor)"}
   RigVMDispatchNode(
    Enabled=True
    ScreenColor=(R=1,G=1,B=1,A=1)
    ScreenDuration=0.05
    Prefix=&RigVMFunction_StringConcat(B=" - UP PRE:   ", A=RigVMVariableNode() {variable=Driver Item Name}.Value) {template="Concat::Execute(in A,in B,out Result)"}.Result
    Value=RigVMVariableNode() {variable=Result Up Pre}.Value
    ExecuteContext=@n6.True
   ) {template="DISPATCH_RigVMDispatch_Print(in Prefix,in Value,in Enabled,in ScreenDuration,in ScreenColor)"}
   Concat = &RigVMFunction_StringConcat(B=" - UP:   ", A=RigVMVariableNode() {variable=Driver Item Name}.Value) {template="Concat::Execute(in A,in B,out Result)"}
   RigVMDispatchNode(
    Enabled=True
    ScreenColor=(R=1,G=1,B=1,A=1)
    ScreenDuration=0.05
    Prefix=@Concat.Result
    Value=RigVMVariableNode() {variable=Result Up}.Value
    ExecuteContext=@RigVMFunction_ControlFlowBranch.True
   ) {template="DISPATCH_RigVMDispatch_Print(in Prefix,in Value,in Enabled,in ScreenDuration,in ScreenColor)"}
   n10 = &RigUnit_SphericalPoseReader(
    ActiveRegionScaleFactors=(PositiveWidth=1,NegativeWidth=1,PositiveHeight=1,NegativeHeight=1)
    ActiveRegionSize=0.001
    Debug=(bDrawDebug=false,bDraw2D=False,bDrawLocalAxes=False,DebugScale=25,DebugSegments=20,DebugThickness=0.05)
    DriverAxis=(X=1,Y=0,Z=0)
    FalloffRegionScaleFactors=(PositiveWidth=1,NegativeWidth=1,PositiveHeight=1,NegativeHeight=1)
    FalloffSize=0.5
    FlipHeightScaling=False
    FlipWidthScaling=False
    RotationOffset=(X=90,Y=31.239,Z=0)
    DriverItem=RigVMVariableNode() {variable=Driver Item}.Value
    OptionalParentItem=RigVMVariableNode() {variable=Driver Item Parent}.Value
    Debug.bDrawDebug=RigVMVariableNode() {variable=Fr Viz}.Value
    ExecutePin=@RigVMFunction_Sequence.C
   )
   > n11 = RigVMVariableNode(Value=@n10.OutputParam) {variable=Result Fr}
   n12 = &RigVMFunction_ControlFlowBranch(ExecuteContext=@n11.ExecuteContext, Condition=RigVMVariableNode() {variable=Fr Viz}.Value)
   n13 = &RigUnit_SphericalPoseReader(
    ActiveRegionScaleFactors=(PositiveWidth=1,NegativeWidth=1,PositiveHeight=1,NegativeHeight=1)
    ActiveRegionSize=0.001
    Debug=(bDrawDebug=false,bDraw2D=False,bDrawLocalAxes=False,DebugScale=25,DebugSegments=20,DebugThickness=0.05)
    DriverAxis=(X=1,Y=0,Z=0)
    FalloffRegionScaleFactors=(PositiveWidth=2,NegativeWidth=2,PositiveHeight=1,NegativeHeight=1)
    FalloffSize=0.25
    FlipHeightScaling=False
    FlipWidthScaling=False
    RotationOffset=(X=45,Y=-58.761,Z=0)
    ExecutePin=@n12.Completed
    DriverItem=RigVMVariableNode() {variable=Driver Item}.Value
    OptionalParentItem=RigVMVariableNode() {variable=Driver Item Parent}.Value
    Debug.bDrawDebug=RigVMVariableNode() {variable=Fr Pre Viz}.Value
   )
   > n14 = RigVMVariableNode(Value=@n13.OutputParam) {variable=Result Fr Pre}
   n15 = &RigVMFunction_ControlFlowBranch(ExecuteContext=@n14.ExecuteContext, Condition=RigVMVariableNode() {variable=Fr Pre Viz}.Value)
   n16 = &RigUnit_SphericalPoseReader(
    ActiveRegionScaleFactors=(PositiveWidth=1,NegativeWidth=1,PositiveHeight=1,NegativeHeight=1)
    ActiveRegionSize=0.001
    Debug=(bDrawDebug=false,bDraw2D=False,bDrawLocalAxes=False,DebugScale=25,DebugSegments=20,DebugThickness=0.05)
    DriverAxis=(X=1,Y=0,Z=0)
    FalloffRegionScaleFactors=(PositiveWidth=2,NegativeWidth=2,PositiveHeight=1,NegativeHeight=1)
    FalloffSize=0.25
    FlipHeightScaling=False
    FlipWidthScaling=False
    RotationOffset=(X=135,Y=-58.761,Z=0)
    ExecutePin=@n15.Completed
    DriverItem=RigVMVariableNode() {variable=Driver Item}.Value
    OptionalParentItem=RigVMVariableNode() {variable=Driver Item Parent}.Value
    Debug.bDrawDebug=RigVMVariableNode() {variable=Fr Post Viz}.Value
   )
   > n17 = RigVMVariableNode(Value=@n16.OutputParam) {variable=Result Fr Post}
   n18 = &RigVMFunction_ControlFlowBranch(ExecuteContext=@n17.ExecuteContext, Condition=RigVMVariableNode() {variable=Fr Post Viz}.Value)
   n19 = &RigUnit_SphericalPoseReader(
    ActiveRegionScaleFactors=(PositiveWidth=1,NegativeWidth=1,PositiveHeight=1,NegativeHeight=1)
    ActiveRegionSize=0.001
    Debug=(bDrawDebug=false,bDraw2D=False,bDrawLocalAxes=False,DebugScale=25,DebugSegments=20,DebugThickness=0.05)
    DriverAxis=(X=1,Y=0,Z=0)
    FalloffRegionScaleFactors=(PositiveWidth=1,NegativeWidth=1,PositiveHeight=0.5,NegativeHeight=0.5)
    FalloffSize=0.5
    FlipHeightScaling=False
    FlipWidthScaling=False
    RotationOffset=(X=45,Y=31.239,Z=0)
    ExecutePin=@n18.Completed
    DriverItem=RigVMVariableNode() {variable=Driver Item}.Value
    OptionalParentItem=RigVMVariableNode() {variable=Driver Item Parent}.Value
    Debug.bDrawDebug=RigVMVariableNode() {variable=Fr Up Viz}.Value
   )
   > n20 = RigVMVariableNode(Value=@n19.OutputParam) {variable=Result Fr Up}
   n21 = &RigVMFunction_ControlFlowBranch(ExecuteContext=@n20.ExecuteContext, Condition=RigVMVariableNode() {variable=Fr Up Viz}.Value)
   n22 = &RigUnit_SphericalPoseReader(
    ActiveRegionScaleFactors=(PositiveWidth=1,NegativeWidth=1,PositiveHeight=1,NegativeHeight=1)
    ActiveRegionSize=0.001
    Debug=(bDrawDebug=false,bDraw2D=False,bDrawLocalAxes=False,DebugScale=25,DebugSegments=20,DebugThickness=0.05)
    DriverAxis=(X=1,Y=0,Z=0)
    FalloffRegionScaleFactors=(PositiveWidth=1,NegativeWidth=1,PositiveHeight=0.5,NegativeHeight=0.5)
    FalloffSize=0.5
    FlipHeightScaling=False
    FlipWidthScaling=False
    RotationOffset=(X=135,Y=31.239,Z=0)
    ExecutePin=@n21.Completed
    DriverItem=RigVMVariableNode() {variable=Driver Item}.Value
    OptionalParentItem=RigVMVariableNode() {variable=Driver Item Parent}.Value
    Debug.bDrawDebug=RigVMVariableNode() {variable=Fr Dn Viz}.Value
   )
   > n23 = RigVMVariableNode(Value=@n22.OutputParam) {variable=Result Fr Dn}
   n24 = &RigVMFunction_ControlFlowBranch(ExecuteContext=@n23.ExecuteContext, Condition=RigVMVariableNode() {variable=Fr Dn Viz}.Value)
   RigVMDispatchNode(
    Enabled=True
    ScreenColor=(R=1,G=1,B=1,A=1)
    ScreenDuration=0.05
    Prefix=&RigVMFunction_StringConcat(B=" - FR DN:   ", A=RigVMVariableNode() {variable=Driver Item Name}.Value) {template="Concat::Execute(in A,in B,out Result)"}.Result
    ExecuteContext=@n24.True
    Value=RigVMVariableNode() {variable=Result Fr Dn}.Value
   ) {template="DISPATCH_RigVMDispatch_Print(in Prefix,in Value,in Enabled,in ScreenDuration,in ScreenColor)"}
   RigVMDispatchNode(
    Enabled=True
    ScreenColor=(R=1,G=1,B=1,A=1)
    ScreenDuration=0.05
    Prefix=&RigVMFunction_StringConcat(B=" - FR UP:   ", A=RigVMVariableNode() {variable=Driver Item Name}.Value) {template="Concat::Execute(in A,in B,out Result)"}.Result
    ExecuteContext=@n21.True
    Value=RigVMVariableNode() {variable=Result Fr Up}.Value
   ) {template="DISPATCH_RigVMDispatch_Print(in Prefix,in Value,in Enabled,in ScreenDuration,in ScreenColor)"}
   RigVMDispatchNode(
    Enabled=True
    ScreenColor=(R=1,G=1,B=1,A=1)
    ScreenDuration=0.05
    Prefix=&RigVMFunction_StringConcat(B=" - FR POST:   ", A=RigVMVariableNode() {variable=Driver Item Name}.Value) {template="Concat::Execute(in A,in B,out Result)"}.Result
    ExecuteContext=@n18.True
    Value=RigVMVariableNode() {variable=Result Fr Post}.Value
   ) {template="DISPATCH_RigVMDispatch_Print(in Prefix,in Value,in Enabled,in ScreenDuration,in ScreenColor)"}
   RigVMDispatchNode(
    Enabled=True
    ScreenColor=(R=1,G=1,B=1,A=1)
    ScreenDuration=0.05
    Prefix=&RigVMFunction_StringConcat(B=" - FR PRE:   ", A=RigVMVariableNode() {variable=Driver Item Name}.Value) {template="Concat::Execute(in A,in B,out Result)"}.Result
    ExecuteContext=@n15.True
    Value=RigVMVariableNode() {variable=Result Fr Pre}.Value
   ) {template="DISPATCH_RigVMDispatch_Print(in Prefix,in Value,in Enabled,in ScreenDuration,in ScreenColor)"}
   RigVMDispatchNode(
    Enabled=True
    ScreenColor=(R=1,G=1,B=1,A=1)
    ScreenDuration=0.05
    Prefix=&RigVMFunction_StringConcat(B=" - FR:   ", A=RigVMVariableNode() {variable=Driver Item Name}.Value) {template="Concat::Execute(in A,in B,out Result)"}.Result
    ExecuteContext=@n12.True
    Value=RigVMVariableNode() {variable=Result Fr}.Value
   ) {template="DISPATCH_RigVMDispatch_Print(in Prefix,in Value,in Enabled,in ScreenDuration,in ScreenColor)"}
   n25 = &RigUnit_SphericalPoseReader(
    ActiveRegionScaleFactors=(PositiveWidth=1,NegativeWidth=1,PositiveHeight=1,NegativeHeight=1)
    ActiveRegionSize=0.001
    Debug=(bDrawDebug=false,bDraw2D=False,bDrawLocalAxes=False,DebugScale=25,DebugSegments=20,DebugThickness=0.05)
    DriverAxis=(X=1,Y=0,Z=0)
    FalloffRegionScaleFactors=(PositiveWidth=1,NegativeWidth=1,PositiveHeight=1,NegativeHeight=1)
    FalloffSize=0.5
    FlipHeightScaling=False
    FlipWidthScaling=False
    RotationOffset=(X=-90,Y=31.239,Z=0)
    DriverItem=RigVMVariableNode() {variable=Driver Item}.Value
    OptionalParentItem=RigVMVariableNode() {variable=Driver Item Parent}.Value
    Debug.bDrawDebug=RigVMVariableNode() {variable=Ba Viz}.Value
    ExecutePin=@RigVMFunction_Sequence.D
   )
   > n26 = RigVMVariableNode(Value=@n25.OutputParam) {variable=Result Ba}
   n27 = &RigVMFunction_ControlFlowBranch(ExecuteContext=@n26.ExecuteContext, Condition=RigVMVariableNode() {variable=Ba Viz}.Value)
   n28 = &RigUnit_SphericalPoseReader(
    ActiveRegionScaleFactors=(PositiveWidth=1,NegativeWidth=1,PositiveHeight=1,NegativeHeight=1)
    ActiveRegionSize=0.001
    Debug=(bDrawDebug=false,bDraw2D=False,bDrawLocalAxes=False,DebugScale=25,DebugSegments=20,DebugThickness=0.05)
    DriverAxis=(X=1,Y=0,Z=0)
    FalloffRegionScaleFactors=(PositiveWidth=2,NegativeWidth=2,PositiveHeight=1,NegativeHeight=1)
    FalloffSize=0.25
    FlipHeightScaling=False
    FlipWidthScaling=False
    RotationOffset=(X=-45,Y=-58.761,Z=0)
    ExecutePin=@n27.Completed
    Debug.bDrawDebug=RigVMVariableNode() {variable=Ba Pre Viz}.Value
    DriverItem=RigVMVariableNode() {variable=Driver Item}.Value
    OptionalParentItem=RigVMVariableNode() {variable=Driver Item Parent}.Value
   )
   > n29 = RigVMVariableNode(Value=@n28.OutputParam) {variable=Result Ba Pre}
   n30 = &RigVMFunction_ControlFlowBranch(Condition=RigVMVariableNode() {variable=Ba Pre Viz}.Value, ExecuteContext=@n29.ExecuteContext)
   n31 = &RigUnit_SphericalPoseReader(
    ActiveRegionScaleFactors=(PositiveWidth=1,NegativeWidth=1,PositiveHeight=1,NegativeHeight=1)
    ActiveRegionSize=0.001
    Debug=(bDrawDebug=false,bDraw2D=False,bDrawLocalAxes=False,DebugScale=25,DebugSegments=20,DebugThickness=0.05)
    DriverAxis=(X=1,Y=0,Z=0)
    FalloffRegionScaleFactors=(PositiveWidth=2,NegativeWidth=2,PositiveHeight=1,NegativeHeight=1)
    FalloffSize=0.25
    FlipHeightScaling=False
    FlipWidthScaling=False
    RotationOffset=(X=-135,Y=-58.761,Z=0)
    ExecutePin=@n30.Completed
    Debug.bDrawDebug=RigVMVariableNode() {variable=Ba Post Viz}.Value
    DriverItem=RigVMVariableNode() {variable=Driver Item}.Value
    OptionalParentItem=RigVMVariableNode() {variable=Driver Item Parent}.Value
   )
   > n32 = RigVMVariableNode(Value=@n31.OutputParam) {variable=Result Ba Post}
   n33 = &RigVMFunction_ControlFlowBranch(Condition=RigVMVariableNode() {variable=Ba Post Viz}.Value, ExecuteContext=@n32.ExecuteContext)
   n34 = &RigUnit_SphericalPoseReader(
    ActiveRegionScaleFactors=(PositiveWidth=1,NegativeWidth=1,PositiveHeight=1,NegativeHeight=1)
    ActiveRegionSize=0.001
    Debug=(bDrawDebug=false,bDraw2D=False,bDrawLocalAxes=False,DebugScale=25,DebugSegments=20,DebugThickness=0.05)
    DriverAxis=(X=1,Y=0,Z=0)
    FalloffRegionScaleFactors=(PositiveWidth=1,NegativeWidth=1,PositiveHeight=0.5,NegativeHeight=0.5)
    FalloffSize=0.5
    FlipHeightScaling=False
    FlipWidthScaling=False
    RotationOffset=(X=-45,Y=31.239,Z=0)
    ExecutePin=@n33.Completed
    Debug.bDrawDebug=RigVMVariableNode() {variable=Ba Up Viz}.Value
    DriverItem=RigVMVariableNode() {variable=Driver Item}.Value
    OptionalParentItem=RigVMVariableNode() {variable=Driver Item Parent}.Value
   )
   > n35 = RigVMVariableNode(Value=@n34.OutputParam) {variable=Result Ba Up}
   n36 = &RigVMFunction_ControlFlowBranch(Condition=RigVMVariableNode() {variable=Ba Up Viz}.Value, ExecuteContext=@n35.ExecuteContext)
   n37 = &RigUnit_SphericalPoseReader(
    ActiveRegionScaleFactors=(PositiveWidth=1,NegativeWidth=1,PositiveHeight=1,NegativeHeight=1)
    ActiveRegionSize=0.001
    Debug=(bDrawDebug=false,bDraw2D=False,bDrawLocalAxes=False,DebugScale=25,DebugSegments=20,DebugThickness=0.05)
    DriverAxis=(X=1,Y=0,Z=0)
    FalloffRegionScaleFactors=(PositiveWidth=1,NegativeWidth=1,PositiveHeight=0.5,NegativeHeight=0.5)
    FalloffSize=0.5
    FlipHeightScaling=False
    FlipWidthScaling=False
    RotationOffset=(X=-135,Y=31.239,Z=0)
    ExecutePin=@n36.Completed
    Debug.bDrawDebug=RigVMVariableNode() {variable=Ba Dn Viz}.Value
    DriverItem=RigVMVariableNode() {variable=Driver Item}.Value
    OptionalParentItem=RigVMVariableNode() {variable=Driver Item Parent}.Value
   )
   > n38 = RigVMVariableNode(Value=@n37.OutputParam) {variable=Result Ba Dn}
   n39 = &RigVMFunction_ControlFlowBranch(Condition=RigVMVariableNode() {variable=Ba Dn Viz}.Value, ExecuteContext=@n38.ExecuteContext)
   RigVMDispatchNode(
    Enabled=True
    ScreenColor=(R=1,G=1,B=1,A=1)
    ScreenDuration=0.05
    Prefix=&RigVMFunction_StringConcat(B=" - BA DN:   ", A=RigVMVariableNode() {variable=Driver Item Name}.Value) {template="Concat::Execute(in A,in B,out Result)"}.Result
    ExecuteContext=@n39.True
    Value=RigVMVariableNode() {variable=Result Ba Dn}.Value
   ) {template="DISPATCH_RigVMDispatch_Print(in Prefix,in Value,in Enabled,in ScreenDuration,in ScreenColor)"}
   RigVMDispatchNode(
    Enabled=True
    ScreenColor=(R=1,G=1,B=1,A=1)
    ScreenDuration=0.05
    Prefix=&RigVMFunction_StringConcat(B=" - BA UP:   ", A=RigVMVariableNode() {variable=Driver Item Name}.Value) {template="Concat::Execute(in A,in B,out Result)"}.Result
    ExecuteContext=@n36.True
    Value=RigVMVariableNode() {variable=Result Ba Up}.Value
   ) {template="DISPATCH_RigVMDispatch_Print(in Prefix,in Value,in Enabled,in ScreenDuration,in ScreenColor)"}
   RigVMDispatchNode(
    Enabled=True
    ScreenColor=(R=1,G=1,B=1,A=1)
    ScreenDuration=0.05
    Prefix=&RigVMFunction_StringConcat(B=" - BA POST:   ", A=RigVMVariableNode() {variable=Driver Item Name}.Value) {template="Concat::Execute(in A,in B,out Result)"}.Result
    ExecuteContext=@n33.True
    Value=RigVMVariableNode() {variable=Result Ba Post}.Value
   ) {template="DISPATCH_RigVMDispatch_Print(in Prefix,in Value,in Enabled,in ScreenDuration,in ScreenColor)"}
   RigVMDispatchNode(
    Enabled=True
    ScreenColor=(R=1,G=1,B=1,A=1)
    ScreenDuration=0.05
    Prefix=&RigVMFunction_StringConcat(B=" - BA PRE:   ", A=RigVMVariableNode() {variable=Driver Item Name}.Value) {template="Concat::Execute(in A,in B,out Result)"}.Result
    ExecuteContext=@n30.True
    Value=RigVMVariableNode() {variable=Result Ba Pre}.Value
   ) {template="DISPATCH_RigVMDispatch_Print(in Prefix,in Value,in Enabled,in ScreenDuration,in ScreenColor)"}
   RigVMDispatchNode(
    Enabled=True
    ScreenColor=(R=1,G=1,B=1,A=1)
    ScreenDuration=0.05
    Prefix=&RigVMFunction_StringConcat(B=" - BA:   ", A=RigVMVariableNode() {variable=Driver Item Name}.Value) {template="Concat::Execute(in A,in B,out Result)"}.Result
    ExecuteContext=@n27.True
    Value=RigVMVariableNode() {variable=Result Ba}.Value
   ) {template="DISPATCH_RigVMDispatch_Print(in Prefix,in Value,in Enabled,in ScreenDuration,in ScreenColor)"}
   n40 = &RigUnit_SphericalPoseReader(
    ActiveRegionScaleFactors=(PositiveWidth=1,NegativeWidth=1,PositiveHeight=1,NegativeHeight=1)
    ActiveRegionSize=0.001
    Debug=(bDrawDebug=false,bDraw2D=False,bDrawLocalAxes=False,DebugScale=25,DebugSegments=20,DebugThickness=0.05)
    DriverAxis=(X=1,Y=0,Z=0)
    FalloffRegionScaleFactors=(PositiveWidth=1,NegativeWidth=0.65,PositiveHeight=1,NegativeHeight=1)
    FalloffSize=0.5
    FlipHeightScaling=False
    FlipWidthScaling=False
    RotationOffset=(X=0,Y=-148.761,Z=0)
    DriverItem=RigVMVariableNode() {variable=Driver Item}.Value
    OptionalParentItem=RigVMVariableNode() {variable=Driver Item Parent}.Value
    Debug.bDrawDebug=RigVMVariableNode() {variable=Dn Viz}.Value
    ExecutePin=@RigVMFunction_Sequence.B
   )
   > n41 = RigVMVariableNode(Value=@n40.OutputParam) {variable=Result Dn}
   n42 = &RigVMFunction_ControlFlowBranch(ExecuteContext=@n41.ExecuteContext, Condition=RigVMVariableNode() {variable=Dn Viz}.Value)
   n43 = &RigUnit_SphericalPoseReader(
    ActiveRegionScaleFactors=(PositiveWidth=1,NegativeWidth=1,PositiveHeight=1,NegativeHeight=1)
    ActiveRegionSize=0.001
    Debug=(bDrawDebug=false,bDraw2D=False,bDrawLocalAxes=False,DebugScale=25,DebugSegments=20,DebugThickness=0.05)
    DriverAxis=(X=1,Y=0,Z=0)
    FalloffRegionScaleFactors=(PositiveWidth=1,NegativeWidth=0.298,PositiveHeight=2,NegativeHeight=2)
    FalloffSize=0.25
    FlipHeightScaling=False
    FlipWidthScaling=False
    RotationOffset=(X=0,Y=-103.761,Z=0)
    ExecutePin=@n42.Completed
    DriverItem=RigVMVariableNode() {variable=Driver Item}.Value
    OptionalParentItem=RigVMVariableNode() {variable=Driver Item Parent}.Value
    Debug.bDrawDebug=RigVMVariableNode() {variable=Dn Pre Viz}.Value
   )
   > n44 = RigVMVariableNode(Value=@n43.OutputParam) {variable=Result Dn Pre}
   n45 = &RigVMFunction_ControlFlowBranch(ExecuteContext=@n44.ExecuteContext, Condition=RigVMVariableNode() {variable=Dn Pre Viz}.Value)
   n46 = &RigUnit_SphericalPoseReader(
    ActiveRegionScaleFactors=(PositiveWidth=1,NegativeWidth=1,PositiveHeight=1,NegativeHeight=1)
    ActiveRegionSize=0.001
    Debug=(bDrawDebug=false,bDraw2D=False,bDrawLocalAxes=False,DebugScale=25,DebugSegments=20,DebugThickness=0.05)
    DriverAxis=(X=1,Y=0,Z=0)
    FalloffRegionScaleFactors=(PositiveWidth=1,NegativeWidth=1,PositiveHeight=2,NegativeHeight=2)
    FalloffSize=0.25
    FlipHeightScaling=False
    FlipWidthScaling=False
    RotationOffset=(X=0,Y=-193.761,Z=0)
    ExecutePin=@n45.Completed
    DriverItem=RigVMVariableNode() {variable=Driver Item}.Value
    OptionalParentItem=RigVMVariableNode() {variable=Driver Item Parent}.Value
    Debug.bDrawDebug=RigVMVariableNode() {variable=Dn Post Viz}.Value
   )
   > n47 = RigVMVariableNode(Value=@n46.OutputParam) {variable=Result Dn Post}
   n48 = &RigVMFunction_ControlFlowBranch(ExecuteContext=@n47.ExecuteContext, Condition=RigVMVariableNode() {variable=Dn Post Viz}.Value)
   RigVMDispatchNode(
    Enabled=True
    ScreenColor=(R=1,G=1,B=1,A=1)
    ScreenDuration=0.05
    Prefix=&RigVMFunction_StringConcat(B=" - DN POST:   ", A=RigVMVariableNode() {variable=Driver Item Name}.Value) {template="Concat::Execute(in A,in B,out Result)"}.Result
    ExecuteContext=@n48.True
    Value=RigVMVariableNode() {variable=Result Dn Post}.Value
   ) {template="DISPATCH_RigVMDispatch_Print(in Prefix,in Value,in Enabled,in ScreenDuration,in ScreenColor)"}
   RigVMDispatchNode(
    Enabled=True
    ScreenColor=(R=1,G=1,B=1,A=1)
    ScreenDuration=0.05
    Prefix=&RigVMFunction_StringConcat(B=" - DN PRE:   ", A=RigVMVariableNode() {variable=Driver Item Name}.Value) {template="Concat::Execute(in A,in B,out Result)"}.Result
    ExecuteContext=@n45.True
    Value=RigVMVariableNode() {variable=Result Dn Pre}.Value
   ) {template="DISPATCH_RigVMDispatch_Print(in Prefix,in Value,in Enabled,in ScreenDuration,in ScreenColor)"}
   RigVMDispatchNode(
    Enabled=True
    ScreenColor=(R=1,G=1,B=1,A=1)
    ScreenDuration=0.05
    Prefix=&RigVMFunction_StringConcat(B=" - DN:   ", A=RigVMVariableNode() {variable=Driver Item Name}.Value) {template="Concat::Execute(in A,in B,out Result)"}.Result
    ExecuteContext=@n42.True
    Value=RigVMVariableNode() {variable=Result Dn}.Value
   ) {template="DISPATCH_RigVMDispatch_Print(in Prefix,in Value,in Enabled,in ScreenDuration,in ScreenColor)"}
   n49 = &RigUnit_SphericalPoseReader(
    ActiveRegionScaleFactors=(PositiveWidth=1,NegativeWidth=1,PositiveHeight=1,NegativeHeight=1)
    ActiveRegionSize=0.001
    Debug=(bDrawDebug=true,bDraw2D=False,bDrawLocalAxes=False,DebugScale=25,DebugSegments=20,DebugThickness=0.05)
    DriverAxis=(X=1,Y=0,Z=0)
    FalloffRegionScaleFactors=(PositiveWidth=1,NegativeWidth=1,PositiveHeight=1,NegativeHeight=1)
    FalloffSize=0.18
    FlipHeightScaling=False
    FlipWidthScaling=False
    RotationOffset=(X=28.4478,Y=74.4998,Z=0)
    DriverItem=RigVMVariableNode() {variable=Driver Item}.Value
    OptionalParentItem=RigVMVariableNode() {variable=Driver Item Parent}.Value
    Debug.bDrawDebug=RigVMVariableNode() {variable=Aim Up In Viz}.Value
    ExecutePin=@RigVMFunction_Sequence.E
   )
   > n50 = RigVMVariableNode(Value=@n49.OutputParam) {variable=Result Aim Up In}
   n51 = &RigVMFunction_ControlFlowBranch(Condition=RigVMVariableNode() {variable=Aim Up In Viz}.Value, ExecuteContext=@n50.ExecuteContext)
   n52 = &RigUnit_SphericalPoseReader(
    ActiveRegionScaleFactors=(PositiveWidth=1,NegativeWidth=1,PositiveHeight=1,NegativeHeight=1)
    ActiveRegionSize=0.001
    Debug=(bDrawDebug=true,bDraw2D=False,bDrawLocalAxes=False,DebugScale=25,DebugSegments=20,DebugThickness=0.05)
    DriverAxis=(X=1,Y=0,Z=0)
    FalloffRegionScaleFactors=(PositiveWidth=1,NegativeWidth=1,PositiveHeight=1,NegativeHeight=1)
    FalloffSize=0.18
    FlipHeightScaling=False
    FlipWidthScaling=False
    RotationOffset=(X=31.239,Y=-1.4062,Z=0)
    ExecutePin=@n51.Completed
    DriverItem=RigVMVariableNode() {variable=Driver Item}.Value
    OptionalParentItem=RigVMVariableNode() {variable=Driver Item Parent}.Value
    Debug.bDrawDebug=RigVMVariableNode() {variable=Aim Up Ot Viz}.Value
   )
   > n53 = RigVMVariableNode(Value=@n52.OutputParam) {variable=Result Aim Up Ot}
   n54 = &RigVMFunction_ControlFlowBranch(Condition=RigVMVariableNode() {variable=Aim Up Ot Viz}.Value, ExecuteContext=@n53.ExecuteContext)
   n55 = &RigUnit_SphericalPoseReader(
    ActiveRegionScaleFactors=(PositiveWidth=1,NegativeWidth=1,PositiveHeight=1,NegativeHeight=1)
    ActiveRegionSize=0.001
    Debug=(bDrawDebug=true,bDraw2D=False,bDrawLocalAxes=False,DebugScale=25,DebugSegments=20,DebugThickness=0.05)
    DriverAxis=(X=1,Y=0,Z=0)
    FalloffRegionScaleFactors=(PositiveWidth=1,NegativeWidth=1,PositiveHeight=1,NegativeHeight=1)
    FalloffSize=0.18
    FlipHeightScaling=False
    FlipWidthScaling=False
    RotationOffset=(X=147.9195,Y=-7.4723,Z=1.8383)
    ExecutePin=@n54.Completed
    DriverItem=RigVMVariableNode() {variable=Driver Item}.Value
    OptionalParentItem=RigVMVariableNode() {variable=Driver Item Parent}.Value
    Debug.bDrawDebug=RigVMVariableNode() {variable=Aim Dn In Viz}.Value
   )
   > n56 = RigVMVariableNode(Value=@n55.OutputParam) {variable=Result Aim Dn In}
   n57 = &RigVMFunction_ControlFlowBranch(Condition=RigVMVariableNode() {variable=Aim Dn In Viz}.Value, ExecuteContext=@n56.ExecuteContext)
   RigVMDispatchNode(
    Enabled=True
    ScreenColor=(R=1,G=1,B=1,A=1)
    ScreenDuration=0.05
    Prefix=&RigVMFunction_StringConcat(B=" - AIM DN IN:   ", A=RigVMVariableNode() {variable=Driver Item Name}.Value) {template="Concat::Execute(in A,in B,out Result)"}.Result
    ExecuteContext=@n57.True
    Value=RigVMVariableNode() {variable=Result Aim Dn In}.Value
   ) {template="DISPATCH_RigVMDispatch_Print(in Prefix,in Value,in Enabled,in ScreenDuration,in ScreenColor)"}
   RigVMDispatchNode(
    Enabled=True
    ScreenColor=(R=1,G=1,B=1,A=1)
    ScreenDuration=0.05
    Prefix=&RigVMFunction_StringConcat(B=" - AIM UP OT:   ", A=RigVMVariableNode() {variable=Driver Item Name}.Value) {template="Concat::Execute(in A,in B,out Result)"}.Result
    ExecuteContext=@n54.True
    Value=RigVMVariableNode() {variable=Result Aim Up Ot}.Value
   ) {template="DISPATCH_RigVMDispatch_Print(in Prefix,in Value,in Enabled,in ScreenDuration,in ScreenColor)"}
   RigVMDispatchNode(
    Enabled=True
    ScreenColor=(R=1,G=1,B=1,A=1)
    ScreenDuration=0.05
    Prefix=&RigVMFunction_StringConcat(B=" - AIM UP IN:   ", A=RigVMVariableNode() {variable=Driver Item Name}.Value) {template="Concat::Execute(in A,in B,out Result)"}.Result
    ExecuteContext=@n51.True
    Value=RigVMVariableNode() {variable=Result Aim Up In}.Value
   ) {template="DISPATCH_RigVMDispatch_Print(in Prefix,in Value,in Enabled,in ScreenDuration,in ScreenColor)"}
   n58 = &RigUnit_SphericalPoseReader(
    ActiveRegionScaleFactors=(PositiveWidth=1,NegativeWidth=1,PositiveHeight=1,NegativeHeight=1)
    ActiveRegionSize=0.001
    Debug=(bDrawDebug=true,bDraw2D=False,bDrawLocalAxes=False,DebugScale=25,DebugSegments=20,DebugThickness=0.05)
    DriverAxis=(X=1,Y=0,Z=0)
    ExecutePin=()
    FalloffRegionScaleFactors=(PositiveWidth=1,NegativeWidth=1,PositiveHeight=1,NegativeHeight=1)
    FalloffSize=0.18
    FlipHeightScaling=False
    FlipWidthScaling=False
    RotationOffset=(X=147.9195,Y=-7.4723,Z=1.8383)
    DriverItem=RigVMVariableNode() {variable=Driver Item}.Value
    OptionalParentItem=RigVMVariableNode() {variable=Driver Item Parent}.Value
    Debug.bDrawDebug=RigVMVariableNode() {variable=Aim Dn In Viz}.Value
   )
   > n59 = RigVMVariableNode(Value=@n58.OutputParam) {variable=Result Aim Dn In}
   n60 = &RigVMFunction_ControlFlowBranch(Condition=RigVMVariableNode() {variable=Aim Dn In Viz}.Value, ExecuteContext=@n59.ExecuteContext)
   RigVMDispatchNode(
    Enabled=True
    ScreenColor=(R=1,G=1,B=1,A=1)
    ScreenDuration=0.05
    Prefix=&RigVMFunction_StringConcat(B=" - AIM DN IN:   ", A=RigVMVariableNode() {variable=Driver Item Name}.Value) {template="Concat::Execute(in A,in B,out Result)"}.Result
    ExecuteContext=@n60.True
    Value=RigVMVariableNode() {variable=Result Aim Dn In}.Value
   ) {template="DISPATCH_RigVMDispatch_Print(in Prefix,in Value,in Enabled,in ScreenDuration,in ScreenColor)"}
  ]}
 New Function_ContainedGraph {nodes [
   RigVMFunctionEntryNode()
   > n1 = &RigUnit_SphericalPoseReader(
    ActiveRegionScaleFactors=(PositiveWidth=1,NegativeWidth=1,PositiveHeight=1,NegativeHeight=1)
    ActiveRegionSize=0.001
    Debug=(bDrawDebug=True,bDraw2D=False,bDrawLocalAxes=False,DebugScale=12.5,DebugSegments=20,DebugThickness=0.1)
    DriverAxis=(X=1,Y=0,Z=0)
    FalloffRegionScaleFactors=(PositiveWidth=1,NegativeWidth=1,PositiveHeight=1,NegativeHeight=1)
    FalloffSize=0.5
    FlipHeightScaling=False
    FlipWidthScaling=False
    OptionalParentItem=(Type=Bone,Name="None")
    RotationOffset=(X=90,Y=0,Z=0)
    Debug.bDrawDebug=RigVMVariableNode() {variable=viz bend}.Value
    DriverItem=RigVMVariableNode() {variable=driver item}.Value
   )
   > n2 = RigVMVariableNode(Value=@n1.OutputParam) {variable=Result Bend}
   n3 = &RigVMFunction_ControlFlowBranch(Condition=RigVMVariableNode() {variable=viz bend post}.Value, ExecuteContext=@n2.ExecuteContext)
   n4 = &RigUnit_SphericalPoseReader(
    ActiveRegionScaleFactors=(PositiveWidth=1,NegativeWidth=1,PositiveHeight=1,NegativeHeight=1)
    ActiveRegionSize=0.001
    Debug=(bDrawDebug=True,bDraw2D=False,bDrawLocalAxes=False,DebugScale=12.5,DebugSegments=20,DebugThickness=0.1)
    DriverAxis=(X=1,Y=0,Z=0)
    FalloffRegionScaleFactors=(PositiveWidth=1,NegativeWidth=1,PositiveHeight=1,NegativeHeight=1)
    FalloffSize=0.5
    FlipHeightScaling=False
    FlipWidthScaling=False
    OptionalParentItem=(Type=Bone,Name="None")
    RotationOffset=(X=90,Y=0,Z=90)
    ExecutePin=@n3.Completed
    DriverItem=RigVMVariableNode() {variable=driver item}.Value
    Debug.bDrawDebug=RigVMVariableNode() {variable=viz bend post}.Value
   )
   > n5 = RigVMVariableNode(Value=@n4.OutputParam) {variable=Result Bend Post}
   RigVMFunction_ControlFlowBranch = &RigVMFunction_ControlFlowBranch(ExecuteContext=@n5.ExecuteContext, Condition=RigVMVariableNode() {variable=viz bend post}.Value)
   RigVMFunctionReturnNode(
    ExecuteContext=@RigVMFunction_ControlFlowBranch.Completed
    Bend=RigVMVariableNode() {variable=Result Bend}.Value
    Bend Post=RigVMVariableNode() {variable=Result Bend Post}.Value
   )
   To_String_1 = RigVMDispatchNode(Value=/Script/RigVM.RigVMFunction_NameConcat(B="  Bend Post  ", A=RigVMVariableNode() {variable=driver item}.Value.Name) {is=struct RigVMUnitNode, template="Concat::Execute(in A,in B,out Result)"}.Result) {template="DISPATCH_RigDispatch_ToString(in Value,out Result)"}
   RigVMDispatchNode(
    Enabled=True
    ScreenColor=(R=1,G=1,B=1,A=1)
    ScreenDuration=0.05
    Prefix=@To_String_1.Result
    Value=RigVMVariableNode() {variable=Result Bend Post}.Value
    ExecuteContext=@RigVMFunction_ControlFlowBranch.True
   ) {template="DISPATCH_RigVMDispatch_Print(in Prefix,in Value,in Enabled,in ScreenDuration,in ScreenColor)"}
   To_String = RigVMDispatchNode(Value=&RigVMFunction_NameConcat(B="  Bend  ", A=RigVMVariableNode() {variable=driver item}.Value.Name) {template="Concat::Execute(in A,in B,out Result)"}.Result) {template="DISPATCH_RigDispatch_ToString(in Value,out Result)"}
   RigVMDispatchNode(
    Enabled=True
    ScreenColor=(R=1,G=1,B=1,A=1)
    ScreenDuration=0.05
    Prefix=@To_String.Result
    Value=RigVMVariableNode() {variable=Result Bend}.Value
    ExecuteContext=@n3.True
   ) {template="DISPATCH_RigVMDispatch_Print(in Prefix,in Value,in Enabled,in ScreenDuration,in ScreenColor)"}
  ]}
 New Function_ContainedGraph {nodes [
   n1 = RigVMFunctionEntryNode()
   > RigVMVariableNode(Value=&RigUnit_HierarchyGetParent(bDefaultParent=True, Child=@n1.Driver Item).Parent) {variable=Driver Item Parent}
   To_String = RigVMDispatchNode(Value=@n1.Driver Item.Name) {template="DISPATCH_RigDispatch_ToString(in Value,out Result)"}
   > RigVMVariableNode(Value=@To_String.Result) {variable=Driver Item Name}
   > RigVMFunction_Sequence = RigVMAggregateNode()
   RigVMFunctionReturnNode(
    ExecuteContext=@RigVMFunction_Sequence.F
    Fr Post=RigVMVariableNode() {variable=Result Fr Post}.Value
    Ba=RigVMVariableNode() {variable=Result Ba}.Value
    Ba Pre=RigVMVariableNode() {variable=Result Ba Pre}.Value
    Ba Post=RigVMVariableNode() {variable=Result Ba Post}.Value
    In=RigVMVariableNode() {variable=Result In}.Value
    In Pre=RigVMVariableNode() {variable=Result In Pre}.Value
    In Post=RigVMVariableNode() {variable=Result In Post}.Value
    Ot=RigVMVariableNode() {variable=Result Ot}.Value
    Ot Pre=RigVMVariableNode() {variable=Result Ot Pre}.Value
    Ot Post=RigVMVariableNode() {variable=Result Ot Post}.Value
    Fr=RigVMVariableNode() {variable=Result Fr}.Value
    Fr Pre=RigVMVariableNode() {variable=Result Fr Pre}.Value
   )
   n2 = &RigUnit_SphericalPoseReader(
    ActiveRegionScaleFactors=(PositiveWidth=1,NegativeWidth=1,PositiveHeight=1,NegativeHeight=1)
    ActiveRegionSize=0.001
    Debug=(bDrawDebug=false,bDraw2D=False,bDrawLocalAxes=False,DebugScale=25,DebugSegments=20,DebugThickness=0.05)
    DriverAxis=(X=1,Y=0,Z=0)
    FalloffRegionScaleFactors=(PositiveWidth=1,NegativeWidth=1,PositiveHeight=1,NegativeHeight=1)
    FalloffSize=0.5
    FlipHeightScaling=False
    FlipWidthScaling=False
    RotationOffset=(X=-90,Y=0,Z=0)
    Debug.bDrawDebug=RigVMVariableNode() {variable=Fr Viz}.Value
    DriverItem=RigVMVariableNode() {variable=Driver Item}.Value
    OptionalParentItem=RigVMVariableNode() {variable=Driver Item Parent}.Value
    ExecutePin=@RigVMFunction_Sequence.A
   )
   > n3 = RigVMVariableNode(Value=@n2.OutputParam) {variable=Result Fr}
   RigVMFunction_ControlFlowBranch = &RigVMFunction_ControlFlowBranch(ExecuteContext=@n3.ExecuteContext, Condition=RigVMVariableNode() {variable=Fr Viz}.Value)
   n4 = &RigUnit_SphericalPoseReader(
    ActiveRegionScaleFactors=(PositiveWidth=1,NegativeWidth=1,PositiveHeight=1,NegativeHeight=1)
    ActiveRegionSize=0.001
    Debug=(bDrawDebug=false,bDraw2D=False,bDrawLocalAxes=False,DebugScale=25,DebugSegments=20,DebugThickness=0.05)
    DriverAxis=(X=1,Y=0,Z=0)
    FalloffRegionScaleFactors=(PositiveWidth=1,NegativeWidth=1,PositiveHeight=2,NegativeHeight=2)
    FalloffSize=0.25
    FlipHeightScaling=False
    FlipWidthScaling=False
    RotationOffset=(X=-90,Y=0,Z=45)
    ExecutePin=@RigVMFunction_ControlFlowBranch.Completed
    DriverItem=RigVMVariableNode() {variable=Driver Item}.Value
    OptionalParentItem=RigVMVariableNode() {variable=Driver Item Parent}.Value
    Debug.bDrawDebug=RigVMVariableNode() {variable=Fr Pre Viz}.Value
   )
   > n5 = RigVMVariableNode(Value=@n4.OutputParam) {variable=Result Fr Pre}
   n6 = &RigVMFunction_ControlFlowBranch(Condition=RigVMVariableNode() {variable=Fr Pre Viz}.Value, ExecuteContext=@n5.ExecuteContext)
   n7 = &RigUnit_SphericalPoseReader(
    ActiveRegionScaleFactors=(PositiveWidth=1,NegativeWidth=1,PositiveHeight=1,NegativeHeight=1)
    ActiveRegionSize=0.001
    Debug=(bDrawDebug=false,bDraw2D=False,bDrawLocalAxes=False,DebugScale=25,DebugSegments=20,DebugThickness=0.05)
    DriverAxis=(X=1,Y=0,Z=0)
    FalloffRegionScaleFactors=(PositiveWidth=0.5,NegativeWidth=0.5,PositiveHeight=1,NegativeHeight=1)
    FalloffSize=0.5
    FlipHeightScaling=False
    FlipWidthScaling=False
    RotationOffset=(X=-90,Y=0,Z=-45)
    ExecutePin=@n6.Completed
    Debug.bDrawDebug=RigVMVariableNode() {variable=Fr Post Viz}.Value
    DriverItem=RigVMVariableNode() {variable=Driver Item}.Value
    OptionalParentItem=RigVMVariableNode() {variable=Driver Item Parent}.Value
   )
   > n8 = RigVMVariableNode(Value=@n7.OutputParam) {variable=Result Fr Post}
   n9 = &RigVMFunction_ControlFlowBranch(Condition=RigVMVariableNode() {variable=Fr Post Viz}.Value, ExecuteContext=@n8.ExecuteContext)
   RigVMDispatchNode(
    Enabled=True
    ScreenColor=(R=1,G=1,B=1,A=1)
    ScreenDuration=0.05
    Prefix=&RigVMFunction_StringConcat(B=" - FR POST:   ", A=RigVMVariableNode() {variable=Driver Item Name}.Value) {template="Concat::Execute(in A,in B,out Result)"}.Result
    Value=RigVMVariableNode() {variable=Result Fr Post}.Value
    ExecuteContext=@n9.True
   ) {template="DISPATCH_RigVMDispatch_Print(in Prefix,in Value,in Enabled,in ScreenDuration,in ScreenColor)"}
   RigVMDispatchNode(
    Enabled=True
    ScreenColor=(R=1,G=1,B=1,A=1)
    ScreenDuration=0.05
    Prefix=&RigVMFunction_StringConcat(B=" - FR PRE:   ", A=RigVMVariableNode() {variable=Driver Item Name}.Value) {template="Concat::Execute(in A,in B,out Result)"}.Result
    Value=RigVMVariableNode() {variable=Result Fr Pre}.Value
    ExecuteContext=@n6.True
   ) {template="DISPATCH_RigVMDispatch_Print(in Prefix,in Value,in Enabled,in ScreenDuration,in ScreenColor)"}
   Concat = &RigVMFunction_StringConcat(B=" - FR:   ", A=RigVMVariableNode() {variable=Driver Item Name}.Value) {template="Concat::Execute(in A,in B,out Result)"}
   RigVMDispatchNode(
    Enabled=True
    ScreenColor=(R=1,G=1,B=1,A=1)
    ScreenDuration=0.05
    Prefix=@Concat.Result
    Value=RigVMVariableNode() {variable=Result Fr}.Value
    ExecuteContext=@RigVMFunction_ControlFlowBranch.True
   ) {template="DISPATCH_RigVMDispatch_Print(in Prefix,in Value,in Enabled,in ScreenDuration,in ScreenColor)"}
   n10 = &RigUnit_SphericalPoseReader(
    ActiveRegionScaleFactors=(PositiveWidth=1,NegativeWidth=1,PositiveHeight=1,NegativeHeight=1)
    ActiveRegionSize=0.001
    Debug=(bDrawDebug=false,bDraw2D=False,bDrawLocalAxes=False,DebugScale=25,DebugSegments=20,DebugThickness=0.05)
    DriverAxis=(X=1,Y=0,Z=0)
    FalloffRegionScaleFactors=(PositiveWidth=1,NegativeWidth=1,PositiveHeight=1,NegativeHeight=1)
    FalloffSize=0.5
    FlipHeightScaling=False
    FlipWidthScaling=False
    RotationOffset=(X=180,Y=0,Z=0)
    DriverItem=RigVMVariableNode() {variable=Driver Item}.Value
    OptionalParentItem=RigVMVariableNode() {variable=Driver Item Parent}.Value
    Debug.bDrawDebug=RigVMVariableNode() {variable=In Viz}.Value
    ExecutePin=@RigVMFunction_Sequence.C
   )
   > n11 = RigVMVariableNode(Value=@n10.OutputParam) {variable=Result In}
   n12 = &RigVMFunction_ControlFlowBranch(ExecuteContext=@n11.ExecuteContext, Condition=RigVMVariableNode() {variable=In Viz}.Value)
   n13 = &RigUnit_SphericalPoseReader(
    ActiveRegionScaleFactors=(PositiveWidth=1,NegativeWidth=1,PositiveHeight=1,NegativeHeight=1)
    ActiveRegionSize=0.001
    Debug=(bDrawDebug=false,bDraw2D=False,bDrawLocalAxes=False,DebugScale=25,DebugSegments=20,DebugThickness=0.05)
    DriverAxis=(X=1,Y=0,Z=0)
    FalloffRegionScaleFactors=(PositiveWidth=2,NegativeWidth=2,PositiveHeight=1,NegativeHeight=1)
    FalloffSize=0.25
    FlipHeightScaling=False
    FlipWidthScaling=False
    RotationOffset=(X=225,Y=0,Z=90)
    ExecutePin=@n12.Completed
    DriverItem=RigVMVariableNode() {variable=Driver Item}.Value
    OptionalParentItem=RigVMVariableNode() {variable=Driver Item Parent}.Value
    Debug.bDrawDebug=RigVMVariableNode() {variable=In Pre Viz}.Value
   )
   > n14 = RigVMVariableNode(Value=@n13.OutputParam) {variable=Result In Pre}
   n15 = &RigVMFunction_ControlFlowBranch(ExecuteContext=@n14.ExecuteContext, Condition=RigVMVariableNode() {variable=In Pre Viz}.Value)
   n16 = &RigUnit_SphericalPoseReader(
    ActiveRegionScaleFactors=(PositiveWidth=1,NegativeWidth=1,PositiveHeight=1,NegativeHeight=1)
    ActiveRegionSize=0.001
    Debug=(bDrawDebug=false,bDraw2D=False,bDrawLocalAxes=False,DebugScale=25,DebugSegments=20,DebugThickness=0.05)
    DriverAxis=(X=1,Y=0,Z=0)
    FalloffRegionScaleFactors=(PositiveWidth=2,NegativeWidth=2,PositiveHeight=1,NegativeHeight=1)
    FalloffSize=0.25
    FlipHeightScaling=False
    FlipWidthScaling=False
    RotationOffset=(X=135,Y=0,Z=90)
    ExecutePin=@n15.Completed
    DriverItem=RigVMVariableNode() {variable=Driver Item}.Value
    OptionalParentItem=RigVMVariableNode() {variable=Driver Item Parent}.Value
    Debug.bDrawDebug=RigVMVariableNode() {variable=In Post Viz}.Value
   )
   > n17 = RigVMVariableNode(Value=@n16.OutputParam) {variable=Result In Post}
   n18 = &RigVMFunction_ControlFlowBranch(ExecuteContext=@n17.ExecuteContext, Condition=RigVMVariableNode() {variable=In Post Viz}.Value)
   RigVMDispatchNode(
    Enabled=True
    ScreenColor=(R=1,G=1,B=1,A=1)
    ScreenDuration=0.05
    Prefix=&RigVMFunction_StringConcat(B=" - IN POST:   ", A=RigVMVariableNode() {variable=Driver Item Name}.Value) {template="Concat::Execute(in A,in B,out Result)"}.Result
    ExecuteContext=@n18.True
    Value=RigVMVariableNode() {variable=Result In Post}.Value
   ) {template="DISPATCH_RigVMDispatch_Print(in Prefix,in Value,in Enabled,in ScreenDuration,in ScreenColor)"}
   RigVMDispatchNode(
    Enabled=True
    ScreenColor=(R=1,G=1,B=1,A=1)
    ScreenDuration=0.05
    Prefix=&RigVMFunction_StringConcat(B=" - IN PRE:   ", A=RigVMVariableNode() {variable=Driver Item Name}.Value) {template="Concat::Execute(in A,in B,out Result)"}.Result
    ExecuteContext=@n15.True
    Value=RigVMVariableNode() {variable=Result In Pre}.Value
   ) {template="DISPATCH_RigVMDispatch_Print(in Prefix,in Value,in Enabled,in ScreenDuration,in ScreenColor)"}
   RigVMDispatchNode(
    Enabled=True
    ScreenColor=(R=1,G=1,B=1,A=1)
    ScreenDuration=0.05
    Prefix=&RigVMFunction_StringConcat(B=" - IN:   ", A=RigVMVariableNode() {variable=Driver Item Name}.Value) {template="Concat::Execute(in A,in B,out Result)"}.Result
    ExecuteContext=@n12.True
    Value=RigVMVariableNode() {variable=Result In}.Value
   ) {template="DISPATCH_RigVMDispatch_Print(in Prefix,in Value,in Enabled,in ScreenDuration,in ScreenColor)"}
   n19 = &RigUnit_SphericalPoseReader(
    ActiveRegionScaleFactors=(PositiveWidth=1,NegativeWidth=1,PositiveHeight=1,NegativeHeight=1)
    ActiveRegionSize=0.001
    Debug=(bDrawDebug=false,bDraw2D=False,bDrawLocalAxes=False,DebugScale=25,DebugSegments=20,DebugThickness=0.05)
    DriverAxis=(X=1,Y=0,Z=0)
    FalloffRegionScaleFactors=(PositiveWidth=1,NegativeWidth=1,PositiveHeight=1,NegativeHeight=1)
    FalloffSize=0.5
    FlipHeightScaling=False
    FlipWidthScaling=False
    RotationOffset=(X=0,Y=0,Z=0)
    DriverItem=RigVMVariableNode() {variable=Driver Item}.Value
    OptionalParentItem=RigVMVariableNode() {variable=Driver Item Parent}.Value
    Debug.bDrawDebug=RigVMVariableNode() {variable=Ot Viz}.Value
    ExecutePin=@RigVMFunction_Sequence.D
   )
   > n20 = RigVMVariableNode(Value=@n19.OutputParam) {variable=Result Ot}
   n21 = &RigVMFunction_ControlFlowBranch(ExecuteContext=@n20.ExecuteContext, Condition=RigVMVariableNode() {variable=Ot Viz}.Value)
   n22 = &RigUnit_SphericalPoseReader(
    ActiveRegionScaleFactors=(PositiveWidth=1,NegativeWidth=1,PositiveHeight=1,NegativeHeight=1)
    ActiveRegionSize=0.001
    Debug=(bDrawDebug=false,bDraw2D=False,bDrawLocalAxes=False,DebugScale=25,DebugSegments=20,DebugThickness=0.05)
    DriverAxis=(X=1,Y=0,Z=0)
    FalloffRegionScaleFactors=(PositiveWidth=1,NegativeWidth=1,PositiveHeight=2,NegativeHeight=2)
    FalloffSize=0.25
    FlipHeightScaling=False
    FlipWidthScaling=False
    RotationOffset=(X=0,Y=-45,Z=0)
    ExecutePin=@n21.Completed
    Debug.bDrawDebug=RigVMVariableNode() {variable=Ot Pre Viz}.Value
    DriverItem=RigVMVariableNode() {variable=Driver Item}.Value
    OptionalParentItem=RigVMVariableNode() {variable=Driver Item Parent}.Value
   )
   > n23 = RigVMVariableNode(Value=@n22.OutputParam) {variable=Result Ot Pre}
   n24 = &RigVMFunction_ControlFlowBranch(Condition=RigVMVariableNode() {variable=Ot Pre Viz}.Value, ExecuteContext=@n23.ExecuteContext)
   n25 = &RigUnit_SphericalPoseReader(
    ActiveRegionScaleFactors=(PositiveWidth=1,NegativeWidth=1,PositiveHeight=1,NegativeHeight=1)
    ActiveRegionSize=0.001
    Debug=(bDrawDebug=false,bDraw2D=False,bDrawLocalAxes=False,DebugScale=25,DebugSegments=20,DebugThickness=0.05)
    DriverAxis=(X=1,Y=0,Z=0)
    FalloffRegionScaleFactors=(PositiveWidth=1,NegativeWidth=1,PositiveHeight=2,NegativeHeight=2)
    FalloffSize=0.25
    FlipHeightScaling=False
    FlipWidthScaling=False
    RotationOffset=(X=0,Y=45,Z=0)
    ExecutePin=@n24.Completed
    Debug.bDrawDebug=RigVMVariableNode() {variable=Ot Post Viz}.Value
    DriverItem=RigVMVariableNode() {variable=Driver Item}.Value
    OptionalParentItem=RigVMVariableNode() {variable=Driver Item Parent}.Value
   )
   > n26 = RigVMVariableNode(Value=@n25.OutputParam) {variable=Result Ot Post}
   n27 = &RigVMFunction_ControlFlowBranch(Condition=RigVMVariableNode() {variable=Ot Post Viz}.Value, ExecuteContext=@n26.ExecuteContext)
   RigVMDispatchNode(
    Enabled=True
    ScreenColor=(R=1,G=1,B=1,A=1)
    ScreenDuration=0.05
    Prefix=&RigVMFunction_StringConcat(B=" - OT POST:   ", A=RigVMVariableNode() {variable=Driver Item Name}.Value) {template="Concat::Execute(in A,in B,out Result)"}.Result
    ExecuteContext=@n27.True
    Value=RigVMVariableNode() {variable=Result Ot Post}.Value
   ) {template="DISPATCH_RigVMDispatch_Print(in Prefix,in Value,in Enabled,in ScreenDuration,in ScreenColor)"}
   RigVMDispatchNode(
    Enabled=True
    ScreenColor=(R=1,G=1,B=1,A=1)
    ScreenDuration=0.05
    Prefix=&RigVMFunction_StringConcat(B=" - OT PRE:   ", A=RigVMVariableNode() {variable=Driver Item Name}.Value) {template="Concat::Execute(in A,in B,out Result)"}.Result
    ExecuteContext=@n24.True
    Value=RigVMVariableNode() {variable=Result Ot Pre}.Value
   ) {template="DISPATCH_RigVMDispatch_Print(in Prefix,in Value,in Enabled,in ScreenDuration,in ScreenColor)"}
   RigVMDispatchNode(
    Enabled=True
    ScreenColor=(R=1,G=1,B=1,A=1)
    ScreenDuration=0.05
    Prefix=&RigVMFunction_StringConcat(B=" - OT:   ", A=RigVMVariableNode() {variable=Driver Item Name}.Value) {template="Concat::Execute(in A,in B,out Result)"}.Result
    ExecuteContext=@n21.True
    Value=RigVMVariableNode() {variable=Result Ot}.Value
   ) {template="DISPATCH_RigVMDispatch_Print(in Prefix,in Value,in Enabled,in ScreenDuration,in ScreenColor)"}
   n28 = &RigUnit_SphericalPoseReader(
    ActiveRegionScaleFactors=(PositiveWidth=1,NegativeWidth=1,PositiveHeight=1,NegativeHeight=1)
    ActiveRegionSize=0.001
    Debug=(bDrawDebug=false,bDraw2D=False,bDrawLocalAxes=False,DebugScale=25,DebugSegments=20,DebugThickness=0.05)
    DriverAxis=(X=1,Y=0,Z=0)
    FalloffRegionScaleFactors=(PositiveWidth=1,NegativeWidth=1,PositiveHeight=1,NegativeHeight=1)
    FalloffSize=0.5
    FlipHeightScaling=False
    FlipWidthScaling=False
    RotationOffset=(X=90,Y=0,Z=0)
    DriverItem=RigVMVariableNode() {variable=Driver Item}.Value
    OptionalParentItem=RigVMVariableNode() {variable=Driver Item Parent}.Value
    Debug.bDrawDebug=RigVMVariableNode() {variable=Ba Viz}.Value
    ExecutePin=@RigVMFunction_Sequence.B
   )
   > n29 = RigVMVariableNode(Value=@n28.OutputParam) {variable=Result Ba}
   n30 = &RigVMFunction_ControlFlowBranch(ExecuteContext=@n29.ExecuteContext, Condition=RigVMVariableNode() {variable=Ba Viz}.Value)
   n31 = &RigUnit_SphericalPoseReader(
    ActiveRegionScaleFactors=(PositiveWidth=1,NegativeWidth=1,PositiveHeight=1,NegativeHeight=1)
    ActiveRegionSize=0.001
    Debug=(bDrawDebug=false,bDraw2D=False,bDrawLocalAxes=False,DebugScale=25,DebugSegments=20,DebugThickness=0.05)
    DriverAxis=(X=1,Y=0,Z=0)
    FalloffRegionScaleFactors=(PositiveWidth=1,NegativeWidth=1,PositiveHeight=2,NegativeHeight=2)
    FalloffSize=0.25
    FlipHeightScaling=False
    FlipWidthScaling=False
    RotationOffset=(X=90,Y=0,Z=-45)
    ExecutePin=@n30.Completed
    DriverItem=RigVMVariableNode() {variable=Driver Item}.Value
    OptionalParentItem=RigVMVariableNode() {variable=Driver Item Parent}.Value
    Debug.bDrawDebug=RigVMVariableNode() {variable=Ba Pre Viz}.Value
   )
   > n32 = RigVMVariableNode(Value=@n31.OutputParam) {variable=Result Ba Pre}
   n33 = &RigVMFunction_ControlFlowBranch(ExecuteContext=@n32.ExecuteContext, Condition=RigVMVariableNode() {variable=Ba Pre Viz}.Value)
   n34 = &RigUnit_SphericalPoseReader(
    ActiveRegionScaleFactors=(PositiveWidth=1,NegativeWidth=1,PositiveHeight=1,NegativeHeight=1)
    ActiveRegionSize=0.001
    Debug=(bDrawDebug=false,bDraw2D=False,bDrawLocalAxes=False,DebugScale=25,DebugSegments=20,DebugThickness=0.05)
    DriverAxis=(X=1,Y=0,Z=0)
    FalloffRegionScaleFactors=(PositiveWidth=1,NegativeWidth=1,PositiveHeight=2,NegativeHeight=2)
    FalloffSize=0.25
    FlipHeightScaling=False
    FlipWidthScaling=False
    RotationOffset=(X=90,Y=0,Z=45)
    ExecutePin=@n33.Completed
    DriverItem=RigVMVariableNode() {variable=Driver Item}.Value
    OptionalParentItem=RigVMVariableNode() {variable=Driver Item Parent}.Value
    Debug.bDrawDebug=RigVMVariableNode() {variable=Ba Post Viz}.Value
   )
   > n35 = RigVMVariableNode(Value=@n34.OutputParam) {variable=Result Ba Post}
   n36 = &RigVMFunction_ControlFlowBranch(ExecuteContext=@n35.ExecuteContext, Condition=RigVMVariableNode() {variable=Ba Post Viz}.Value)
   RigVMDispatchNode(
    Enabled=True
    ScreenColor=(R=1,G=1,B=1,A=1)
    ScreenDuration=0.05
    Prefix=&RigVMFunction_StringConcat(B=" - BA POST:   ", A=RigVMVariableNode() {variable=Driver Item Name}.Value) {template="Concat::Execute(in A,in B,out Result)"}.Result
    ExecuteContext=@n36.True
    Value=RigVMVariableNode() {variable=Result Ba Post}.Value
   ) {template="DISPATCH_RigVMDispatch_Print(in Prefix,in Value,in Enabled,in ScreenDuration,in ScreenColor)"}
   RigVMDispatchNode(
    Enabled=True
    ScreenColor=(R=1,G=1,B=1,A=1)
    ScreenDuration=0.05
    Prefix=&RigVMFunction_StringConcat(B=" - BA PRE:   ", A=RigVMVariableNode() {variable=Driver Item Name}.Value) {template="Concat::Execute(in A,in B,out Result)"}.Result
    ExecuteContext=@n33.True
    Value=RigVMVariableNode() {variable=Result Ba Pre}.Value
   ) {template="DISPATCH_RigVMDispatch_Print(in Prefix,in Value,in Enabled,in ScreenDuration,in ScreenColor)"}
   RigVMDispatchNode(
    Enabled=True
    ScreenColor=(R=1,G=1,B=1,A=1)
    ScreenDuration=0.05
    Prefix=&RigVMFunction_StringConcat(B=" - BA:   ", A=RigVMVariableNode() {variable=Driver Item Name}.Value) {template="Concat::Execute(in A,in B,out Result)"}.Result
    ExecuteContext=@n30.True
    Value=RigVMVariableNode() {variable=Result Ba}.Value
   ) {template="DISPATCH_RigVMDispatch_Print(in Prefix,in Value,in Enabled,in ScreenDuration,in ScreenColor)"}
  ]}
 RigVMFunctionLibrary {nodes [
   Elbow Readers = RigVMCollapseNode(ExecuteContext=(), driver item=(Type=None,Name="None"))
   Thigh Readers = RigVMCollapseNode(Driver Item=(Type=None,Name="None"), ExecuteContext=())
   UpperArm Readers = RigVMCollapseNode(Driver Item=(Type=None,Name="None"), ExecuteContext=())
  ]}
 RigVMFunction_Sequence_1_ContainedGraph {nodes [
   n1 = RigVMFunctionEntryNode()
   RigVMFunction_Sequence = &RigVMFunction_Sequence(ExecuteContext=@n1.ExecuteContext)
   n2 = &RigVMFunction_Sequence(ExecuteContext=@RigVMFunction_Sequence.B)
   n3 = &RigVMFunction_Sequence(ExecuteContext=@n2.B)
   n4 = &RigVMFunction_Sequence(ExecuteContext=@n3.B)
   n5 = &RigVMFunction_Sequence(ExecuteContext=@n4.B)
   n6 = &RigVMFunction_Sequence(ExecuteContext=@n5.B)
   n7 = &RigVMFunction_Sequence(ExecuteContext=@n6.B)
   n8 = &RigVMFunction_Sequence(ExecuteContext=@n7.B)
   n9 = &RigVMFunction_Sequence(ExecuteContext=@n8.B)
   n10 = &RigVMFunction_Sequence(ExecuteContext=@n9.B)
   RigVMFunctionReturnNode(A=@RigVMFunction_Sequence.A, B=@n2.A, C=@n3.A, D=@n4.A, E=@n5.A, F=@n6.A, G=@n7.A, H=@n8.A, I=@n9.A, J=@n10.A, K=@n10.B)
  ]}
 RigVMFunction_Sequence_1_ContainedGraph {nodes [
   n1 = RigVMFunctionEntryNode()
   RigVMFunction_Sequence = &RigVMFunction_Sequence(ExecuteContext=@n1.ExecuteContext)
   n2 = &RigVMFunction_Sequence(ExecuteContext=@RigVMFunction_Sequence.B)
   n3 = &RigVMFunction_Sequence(ExecuteContext=@n2.B)
   n4 = &RigVMFunction_Sequence(ExecuteContext=@n3.B)
   n5 = &RigVMFunction_Sequence(ExecuteContext=@n4.B)
   RigVMFunctionReturnNode(A=@RigVMFunction_Sequence.A, B=@n2.A, C=@n3.A, D=@n4.A, E=@n5.A, F=@n5.B)
  ]}
 RigVMFunction_Sequence_1_ContainedGraph {nodes [
   n1 = RigVMFunctionEntryNode()
   RigVMFunction_Sequence = &RigVMFunction_Sequence(ExecuteContext=@n1.ExecuteContext)
   n2 = &RigVMFunction_Sequence(ExecuteContext=@RigVMFunction_Sequence.B)
   n3 = &RigVMFunction_Sequence(ExecuteContext=@n2.B)
   n4 = &RigVMFunction_Sequence(ExecuteContext=@n3.B)
   n5 = &RigVMFunction_Sequence(ExecuteContext=@n4.B)
   RigVMFunctionReturnNode(A=@RigVMFunction_Sequence.A, B=@n2.A, C=@n3.A, D=@n4.A, E=@n5.A, F=@n5.B)
  ]}
 RigVMModel {nodes [
   RigUnit_BeginExecution = &RigUnit_BeginExecution()
   RigVMFunction_Sequence = RigVMAggregateNode(ExecuteContext=@RigUnit_BeginExecution.ExecutePin)
   ItemArray = RigVMDispatchNode(Value=((Type=Bone,Name="lowerarm_l"),(Type=Bone,Name="elbow_l"))) {template="DISPATCH_RigVMDispatch_Constant(io Value)"}
   n1 = RigVMDispatchNode(Index=0, Array=@ItemArray.Value) {template="DISPATCH_RigVMDispatch_ArrayGetAtIndex(in Array,in Index,out Element)"}
   Elbow Readers = RigVMFunctionReferenceNode(viz bend=false, viz bend post=false, driver item=@n1.Element, ExecuteContext=@RigVMFunction_Sequence.C)
   Interpolate = /Script/RigVM.RigVMFunction_MathTransformLerp(
    T=0.635
    A=&RigUnit_GetTransform(Space=LocalSpace, bInitial=true, Item=@n1.Element).Transform
    B=&RigUnit_GetTransform(Space=LocalSpace, bInitial=False, Item=@n1.Element).Transform
   ) {is=struct RigVMUnitNode, template="Interpolate::Execute(in A,in B,in T,out Result)"}
   > &RigUnit_SetTransform(
    Space=LocalSpace
    Weight=1
    bInitial=False
    Value=@Interpolate.Result
    Item=RigVMDispatchNode(Index=1, Array=@ItemArray.Value) {template="DISPATCH_RigVMDispatch_ArrayGetAtIndex(in Array,in Index,out Element)"}.Element
   )
   > &RigUnit_ModifyTransforms(
    ItemToModify=((Item=(Type=Bone,Name="elbow_fr_l"),Transform=(Rotation=(X=0,Y=0,Z=-0.1192,W=0.9929),Translation=(X=5.0002,Y=-0.0434,Z=-0.0001),Scale3D=(X=0.375,Y=1,Z=1))))
    WeightMinimum=0
    Weight=@"Elbow Readers".Bend
   )
   > &RigUnit_ModifyTransforms(
    ItemToModify=((Item=(Type=Bone,Name="elbow_ba_l"),Transform=(Rotation=(X=0,Y=0,Z=-0.1023,W=0.9948),Translation=(X=0.2982,Y=-1.4399,Z=0),Scale3D=(X=1,Y=1,Z=1))))
    WeightMinimum=0
    Weight=@"Elbow Readers".Bend
   )
   > &RigUnit_ModifyTransforms(
    ItemToModify=((Item=(Type=Bone,Name="bicep_l"),Transform=(Rotation=(X=0,Y=0,Z=0,W=1),Translation=(X=-2,Y=0.6667,Z=0),Scale3D=(X=1,Y=1,Z=1))))
    WeightMinimum=0
    Weight=@"Elbow Readers".Bend
   )
   > &RigUnit_ModifyTransforms(
    ItemToModify=((Item=(Type=Bone,Name="tricep_l"),Transform=(Rotation=(X=0,Y=0,Z=0,W=1),Translation=(X=0.5078,Y=-0.2913,Z=0),Scale3D=(X=1,Y=1,Z=1))))
    WeightMinimum=0
    Weight=@"Elbow Readers".Bend
   )
   > &RigUnit_ModifyTransforms(
    ItemToModify=((Item=(Type=Bone,Name="elbow_fr_l"),Transform=(Rotation=(X=0,Y=0,Z=-0.1192,W=0.9929),Translation=(X=5.0002,Y=-0.0434,Z=-0.0001),Scale3D=(X=0.375,Y=1,Z=1))))
    WeightMinimum=0
    Weight=@"Elbow Readers".Bend Post
   )
   > &RigUnit_ModifyTransforms(
    ItemToModify=((Item=(Type=Bone,Name="elbow_ba_l"),Transform=(Rotation=(X=0,Y=0,Z=-0.1023,W=0.9948),Translation=(X=0.2982,Y=-1.4399,Z=0),Scale3D=(X=1,Y=1,Z=1))))
    WeightMinimum=0
    Weight=@"Elbow Readers".Bend Post
   )
   > &RigUnit_ModifyTransforms(
    ItemToModify=((Item=(Type=Bone,Name="bicep_l"),Transform=(Rotation=(X=0,Y=0,Z=0,W=1),Translation=(X=-4,Y=1.3335,Z=0),Scale3D=(X=1,Y=1,Z=1))))
    WeightMinimum=0
    Weight=@"Elbow Readers".Bend Post
   )
   > &RigUnit_ModifyTransforms(
    ItemToModify=((Item=(Type=Bone,Name="tricep_l"),Transform=(Rotation=(X=0,Y=0,Z=0,W=1),Translation=(X=1.0156,Y=-0.5825,Z=0),Scale3D=(X=1,Y=1,Z=1))))
    WeightMinimum=0
    Weight=@"Elbow Readers".Bend Post
   )
   ItemArray_1 = RigVMDispatchNode(Value=((Type=Bone,Name="lowerarm_r"),(Type=Bone,Name="elbow_r"))) {template="DISPATCH_RigVMDispatch_Constant(io Value)"}
   n2 = RigVMDispatchNode(Index=0, Array=@ItemArray_1.Value) {template="DISPATCH_RigVMDispatch_ArrayGetAtIndex(in Array,in Index,out Element)"}
   > Elbow Readers_1 = RigVMFunctionReferenceNode(viz bend=false, viz bend post=false, driver item=@n2.Element)
   > &RigUnit_SetTransform(
    Space=LocalSpace
    Weight=1
    bInitial=False
    Value=&RigVMFunction_MathTransformLerp(
     T=0.635
     A=&RigUnit_GetTransform(Space=LocalSpace, bInitial=true, Item=@n2.Element).Transform
     B=&RigUnit_GetTransform(Space=LocalSpace, bInitial=False, Item=@n2.Element).Transform
    ) {template="Interpolate::Execute(in A,in B,in T,out Result)"}.Result
    Item=RigVMDispatchNode(Index=1, Array=@ItemArray_1.Value) {template="DISPATCH_RigVMDispatch_ArrayGetAtIndex(in Array,in Index,out Element)"}.Element
   )
   > &RigUnit_ModifyTransforms(
    ItemToModify=((Item=(Type=Bone,Name="elbow_fr_r"),Transform=(Rotation=(X=0,Y=0,Z=0.1192,W=0.9929),Translation=(X=-5.0002,Y=0.0434,Z=0),Scale3D=(X=0.375,Y=1,Z=1))))
    WeightMinimum=0
    Weight=@"Elbow Readers_1".Bend
   )
   > &RigUnit_ModifyTransforms(
    ItemToModify=((Item=(Type=Bone,Name="elbow_ba_r"),Transform=(Rotation=(X=0,Y=0,Z=-0.0512,W=0.9987),Translation=(X=-0.2982,Y=1.4399,Z=0),Scale3D=(X=1,Y=1,Z=1))))
    WeightMinimum=0
    Weight=@"Elbow Readers_1".Bend
   )
   > &RigUnit_ModifyTransforms(
    ItemToModify=((Item=(Type=Bone,Name="bicep_r"),Transform=(Rotation=(X=0,Y=0,Z=0,W=1),Translation=(X=2,Y=-0.6667,Z=0),Scale3D=(X=1,Y=1,Z=1))))
    WeightMinimum=0
    Weight=@"Elbow Readers_1".Bend
   )
   > &RigUnit_ModifyTransforms(
    ItemToModify=((Item=(Type=Bone,Name="tricep_r"),Transform=(Rotation=(X=0,Y=0,Z=0,W=1),Translation=(X=-0.2539,Y=0.1456,Z=0),Scale3D=(X=1,Y=1,Z=1))))
    WeightMinimum=0
    Weight=@"Elbow Readers_1".Bend
   )
   > &RigUnit_ModifyTransforms(
    ItemToModify=((Item=(Type=Bone,Name="elbow_fr_r"),Transform=(Rotation=(X=0,Y=0,Z=0.1192,W=0.9929),Translation=(X=-5.0002,Y=0.0434,Z=0),Scale3D=(X=0.375,Y=1,Z=1))))
    WeightMinimum=0
    Weight=@"Elbow Readers_1".Bend Post
   )
   > &RigUnit_ModifyTransforms(
    ItemToModify=((Item=(Type=Bone,Name="elbow_ba_r"),Transform=(Rotation=(X=0,Y=0,Z=0.1023,W=0.9948),Translation=(X=-0.2982,Y=1.4399,Z=0),Scale3D=(X=1,Y=1,Z=1))))
    WeightMinimum=0
    Weight=@"Elbow Readers_1".Bend Post
   )
   > &RigUnit_ModifyTransforms(
    ItemToModify=((Item=(Type=Bone,Name="bicep_r"),Transform=(Rotation=(X=0,Y=0,Z=0,W=1),Translation=(X=4,Y=-1.3335,Z=0),Scale3D=(X=1,Y=1,Z=1))))
    WeightMinimum=0
    Weight=@"Elbow Readers_1".Bend Post
   )
   > &RigUnit_ModifyTransforms(
    ItemToModify=((Item=(Type=Bone,Name="tricep_r"),Transform=(Rotation=(X=0,Y=0,Z=0,W=1),Translation=(X=-1.0156,Y=0.5825,Z=0),Scale3D=(X=1,Y=1,Z=1))))
    WeightMinimum=0
    Weight=@"Elbow Readers_1".Bend Post
   )
   RigVMFunction_ControlFlowBranch = &RigVMFunction_ControlFlowBranch(Condition=RigVMVariableNode() {variable=Debug Controls On}.Value, ExecuteContext=@RigVMFunction_Sequence.A)
   For_Each = RigVMDispatchNode(ExecuteContext=@RigVMFunction_ControlFlowBranch.True, Array=RigVMVariableNode() {variable=Debug Ctrls}.Value) {template="DISPATCH_RigVMDispatch_ArrayIterator(in Array,out Element,out Index,out Count,out Ratio)"}
   > &RigUnit_SetTransform(
    Weight=1
    bInitial=False
    Item=RigVMDispatchNode(Array=RigVMVariableNode() {variable=Debug Bones}.Value, Index=@For_Each.Index) {template="DISPATCH_RigVMDispatch_ArrayGetAtIndex(in Array,in Index,out Element)"}.Element
    Value=&RigUnit_GetTransform(bInitial=False, Item=@For_Each.Element).Transform
   )
   UpperArm Readers = RigVMFunctionReferenceNode(
    Aim Dn In Viz=false
    Aim Dn Ot Viz=false
    Aim Up In Viz=false
    Aim Up Ot Viz=false
    Ba Viz=false
    Dn Viz=false
    Driver Item=(Type=Bone,Name="upperarm_l")
    Fr Viz=false
    Up Viz=false
    ExecuteContext=@RigVMFunction_Sequence.D
   )
   > &RigUnit_ModifyTransforms(
    ItemToModify=((Item=(Type=Bone,Name="def_arm_tp_l"),Transform=(Rotation=(X=0,Y=-0.0901,Z=0,W=0.9959),Translation=(X=-7.9655,Y=0,Z=3.6755),Scale3D=(X=0.5,Y=1,Z=1))))
    WeightMinimum=0
    Weight=@"UpperArm Readers".Up
   )
   > &RigUnit_ModifyTransforms(
    ItemToModify=((Item=(Type=Bone,Name="def_arm_tp_l_end"),Transform=(Rotation=(X=0,Y=-0.8252,Z=0,W=0.5649),Translation=(X=0,Y=0,Z=0),Scale3D=(X=1,Y=1,Z=1))))
    WeightMinimum=0
    Weight=@"UpperArm Readers".Up
   )
   > &RigUnit_ModifyTransforms(
    ItemToModify=((Item=(Type=Bone,Name="def_arm_fr_l"),Transform=(Rotation=(X=0,Y=-0.194,Z=0,W=0.981),Translation=(X=-3.3845,Y=0,Z=4.0813),Scale3D=(X=0.5,Y=1,Z=1))))
    WeightMinimum=0
    Weight=@"UpperArm Readers".Up
   )
   > &RigUnit_ModifyTransforms(
    ItemToModify=((Item=(Type=Bone,Name="def_arm_fr_l_end"),Transform=(Rotation=(X=0,Y=-0.7537,Z=0,W=0.6572),Translation=(X=0,Y=0,Z=0),Scale3D=(X=1,Y=1,Z=1))))
    WeightMinimum=0
    Weight=@"UpperArm Readers".Up
   )
   > &RigUnit_ModifyTransforms(
    ItemToModify=((Item=(Type=Bone,Name="def_arm_ba_l"),Transform=(Rotation=(X=0,Y=-0.194,Z=0,W=0.981),Translation=(X=-3.3845,Y=0,Z=4.0813),Scale3D=(X=1,Y=1,Z=1))))
    WeightMinimum=0
    Weight=@"UpperArm Readers".Up
   )
   > &RigUnit_ModifyTransforms(
    ItemToModify=((Item=(Type=Bone,Name="def_arm_fr_l_end"),Transform=(Rotation=(X=0,Y=-0.7537,Z=0,W=0.6572),Translation=(X=0,Y=0,Z=0),Scale3D=(X=1,Y=1,Z=1))))
    WeightMinimum=0
    Weight=@"UpperArm Readers".Up
   )
   > &RigUnit_ModifyTransforms(
    ItemToModify=((Item=(Type=Bone,Name="def_arm_bt_l"),Transform=(Rotation=(X=0,Y=-0.0569,Z=0,W=0.9984),Translation=(X=-0.178,Y=0,Z=3.7926),Scale3D=(X=1,Y=1,Z=1))))
    WeightMinimum=0
    Weight=@"UpperArm Readers".Up
   )
   > &RigUnit_ModifyTransforms(
    ItemToModify=((Item=(Type=Bone,Name="def_arm_bt_l_end"),Transform=(Rotation=(X=0,Y=0.6948,Z=0,W=-0.7192),Translation=(X=0,Y=0,Z=0),Scale3D=(X=1,Y=1,Z=1))))
    WeightMinimum=0
    Weight=@"UpperArm Readers".Up
   )
   > UpperArm Readers_1 = RigVMFunctionReferenceNode(Driver Item=(Type=Bone,Name="upperarm_r"), Fr Viz=false, Up Viz=false)
   > &RigUnit_ModifyTransforms(
    ItemToModify=((Item=(Type=Bone,Name="def_arm_tp_r"),Transform=(Rotation=(X=0,Y=-0.0901,Z=0,W=0.9959),Translation=(X=7.9655,Y=0,Z=-3.6755),Scale3D=(X=0.5,Y=1,Z=1))))
    WeightMinimum=0
    Weight=@"UpperArm Readers_1".Up
   )
   > &RigUnit_ModifyTransforms(
    ItemToModify=((Item=(Type=Bone,Name="def_arm_tp_r_end"),Transform=(Rotation=(X=0,Y=-0.8252,Z=0,W=0.5649),Translation=(X=0,Y=0,Z=0),Scale3D=(X=1,Y=1,Z=1))))
    WeightMinimum=0
    Weight=@"UpperArm Readers_1".Up
   )
   > &RigUnit_ModifyTransforms(
    ItemToModify=((Item=(Type=Bone,Name="def_arm_fr_r"),Transform=(Rotation=(X=0,Y=-0.194,Z=0,W=0.981),Translation=(X=3.3845,Y=0,Z=-4.0813),Scale3D=(X=0.5,Y=1,Z=1))))
    WeightMinimum=0
    Weight=@"UpperArm Readers_1".Up
   )
   > &RigUnit_ModifyTransforms(
    ItemToModify=((Item=(Type=Bone,Name="def_arm_fr_r_end"),Transform=(Rotation=(X=0,Y=-0.7537,Z=0,W=0.6572),Translation=(X=0,Y=0,Z=0),Scale3D=(X=1,Y=1,Z=1))))
    WeightMinimum=0
    Weight=@"UpperArm Readers_1".Up
   )
   > &RigUnit_ModifyTransforms(
    ItemToModify=((Item=(Type=Bone,Name="def_arm_ba_r"),Transform=(Rotation=(X=0,Y=-0.194,Z=0,W=0.981),Translation=(X=3.3845,Y=0,Z=-4.0813),Scale3D=(X=1,Y=1,Z=1))))
    WeightMinimum=0
    Weight=@"UpperArm Readers_1".Up
   )
   > &RigUnit_ModifyTransforms(
    ItemToModify=((Item=(Type=Bone,Name="def_arm_fr_r_end"),Transform=(Rotation=(X=0,Y=-0.7537,Z=0,W=0.6572),Translation=(X=0,Y=0,Z=0),Scale3D=(X=1,Y=1,Z=1))))
    WeightMinimum=0
    Weight=@"UpperArm Readers_1".Up
   )
   > &RigUnit_ModifyTransforms(
    ItemToModify=((Item=(Type=Bone,Name="def_arm_bt_r"),Transform=(Rotation=(X=0,Y=-0.0569,Z=0,W=0.9984),Translation=(X=0.178,Y=0,Z=-3.7926),Scale3D=(X=1,Y=1,Z=1))))
    WeightMinimum=0
    Weight=@"UpperArm Readers_1".Up
   )
   > &RigUnit_ModifyTransforms(
    ItemToModify=((Item=(Type=Bone,Name="def_arm_bt_r_end"),Transform=(Rotation=(X=0,Y=0.6948,Z=0,W=-0.7192),Translation=(X=0,Y=0,Z=0),Scale3D=(X=1,Y=1,Z=1))))
    WeightMinimum=0
    Weight=@"UpperArm Readers_1".Up
   )
   &RigUnit_ModifyTransforms(
    ItemToModify=((Item=(Type=Bone,Name="def_arm_bt_l"),Transform=(Rotation=(X=0,Y=0,Z=0,W=1),Translation=(X=-0.384,Y=0,Z=-8.6541),Scale3D=(X=1,Y=1,Z=1))))
    WeightMinimum=0
    Weight=@"UpperArm Readers".Dn
    ExecutePin=@RigVMFunction_Sequence.E
   )
   > &RigUnit_ModifyTransforms(
    ItemToModify=((Item=(Type=Bone,Name="def_arm_bt_l_end"),Transform=(Rotation=(X=0,Y=0.7015,Z=0,W=0.7127),Translation=(X=0,Y=0,Z=0),Scale3D=(X=1,Y=1,Z=1))))
    WeightMinimum=0
    Weight=@"UpperArm Readers".Dn
   )
   > &RigUnit_ModifyTransforms(
    ItemToModify=((Item=(Type=Bone,Name="def_arm_tp_l"),Transform=(Rotation=(X=0,Y=0,Z=0,W=1),Translation=(X=0,Y=0,Z=0),Scale3D=(X=1,Y=1,Z=1))))
    WeightMinimum=0
    Weight=@"UpperArm Readers".Dn
   )
   > &RigUnit_ModifyTransforms(
    ItemToModify=((Item=(Type=Bone,Name="def_arm_tp_l_end"),Transform=(Rotation=(X=0,Y=0.5556,Z=0,W=0.8315),Translation=(X=0,Y=0,Z=0),Scale3D=(X=1,Y=1,Z=1))))
    WeightMinimum=0
    Weight=@"UpperArm Readers".Dn
   )
   > &RigUnit_ModifyTransforms(
    ItemToModify=((Item=(Type=Bone,Name="def_arm_fr_l"),Transform=(Rotation=(X=0,Y=0,Z=0,W=1),Translation=(X=0,Y=0,Z=0),Scale3D=(X=1,Y=1,Z=1))))
    WeightMinimum=0
    Weight=@"UpperArm Readers".Dn
   )
   > &RigUnit_ModifyTransforms(
    ItemToModify=((Item=(Type=Bone,Name="def_arm_fr_l_end"),Transform=(Rotation=(X=0,Y=0.6344,Z=0,W=0.773),Translation=(X=0,Y=0,Z=0),Scale3D=(X=1,Y=1,Z=1))))
    WeightMinimum=0
    Weight=@"UpperArm Readers".Dn
   )
   > &RigUnit_ModifyTransforms(
    ItemToModify=((Item=(Type=Bone,Name="def_arm_ba_l"),Transform=(Rotation=(X=0,Y=0,Z=0,W=1),Translation=(X=0,Y=0,Z=0),Scale3D=(X=1,Y=1,Z=1))))
    WeightMinimum=0
    Weight=@"UpperArm Readers".Dn
   )
   > &RigUnit_ModifyTransforms(
    ItemToModify=((Item=(Type=Bone,Name="def_arm_ba_l_end"),Transform=(Rotation=(X=0,Y=0.5556,Z=0,W=0.8315),Translation=(X=0,Y=0,Z=0),Scale3D=(X=1,Y=1,Z=1))))
    WeightMinimum=0
    Weight=@"UpperArm Readers".Dn
   )
   > &RigUnit_ModifyTransforms(
    ItemToModify=((Item=(Type=Bone,Name="def_arm_bt_r"),Transform=(Rotation=(X=0,Y=0,Z=0,W=1),Translation=(X=0.384,Y=0,Z=8.6541),Scale3D=(X=1,Y=1,Z=1))))
    WeightMinimum=0
    Weight=@"UpperArm Readers_1".Dn
   )
   > &RigUnit_ModifyTransforms(
    ItemToModify=((Item=(Type=Bone,Name="def_arm_bt_r_end"),Transform=(Rotation=(X=0,Y=0.7015,Z=0,W=0.7127),Translation=(X=0,Y=0,Z=0),Scale3D=(X=1,Y=1,Z=1))))
    WeightMinimum=0
    Weight=@"UpperArm Readers_1".Dn
   )
   > &RigUnit_ModifyTransforms(
    ItemToModify=((Item=(Type=Bone,Name="def_arm_tp_r"),Transform=(Rotation=(X=0,Y=0,Z=0,W=1),Translation=(X=0,Y=0,Z=0),Scale3D=(X=1,Y=1,Z=1))))
    WeightMinimum=0
    Weight=@"UpperArm Readers_1".Dn
   )
   > &RigUnit_ModifyTransforms(
    ItemToModify=((Item=(Type=Bone,Name="def_arm_tp_r_end"),Transform=(Rotation=(X=0,Y=0.5556,Z=0,W=0.8315),Translation=(X=0,Y=0,Z=0),Scale3D=(X=1,Y=1,Z=1))))
    WeightMinimum=0
    Weight=@"UpperArm Readers_1".Dn
   )
   > &RigUnit_ModifyTransforms(
    ItemToModify=((Item=(Type=Bone,Name="def_arm_fr_r"),Transform=(Rotation=(X=0,Y=0,Z=0,W=1),Translation=(X=0,Y=0,Z=0),Scale3D=(X=1,Y=1,Z=1))))
    WeightMinimum=0
    Weight=@"UpperArm Readers_1".Dn
   )
   > &RigUnit_ModifyTransforms(
    ItemToModify=((Item=(Type=Bone,Name="def_arm_fr_r_end"),Transform=(Rotation=(X=0,Y=0.6344,Z=0,W=0.773),Translation=(X=0,Y=0,Z=0),Scale3D=(X=1,Y=1,Z=1))))
    WeightMinimum=0
    Weight=@"UpperArm Readers_1".Dn
   )
   > &RigUnit_ModifyTransforms(
    ItemToModify=((Item=(Type=Bone,Name="def_arm_ba_r"),Transform=(Rotation=(X=0,Y=0,Z=0,W=1),Translation=(X=0,Y=0,Z=0),Scale3D=(X=1,Y=1,Z=1))))
    WeightMinimum=0
    Weight=@"UpperArm Readers_1".Dn
   )
   > &RigUnit_ModifyTransforms(
    ItemToModify=((Item=(Type=Bone,Name="def_arm_ba_r_end"),Transform=(Rotation=(X=0,Y=0.5556,Z=0,W=0.8315),Translation=(X=0,Y=0,Z=0),Scale3D=(X=1,Y=1,Z=1))))
    WeightMinimum=0
    Weight=@"UpperArm Readers_1".Dn
   )
   &RigUnit_ModifyTransforms(
    ItemToModify=((Item=(Type=Bone,Name="def_arm_fr_l"),Transform=(Rotation=(X=0,Y=0,Z=0,W=1),Translation=(X=-6.3509,Y=3.105,Z=0),Scale3D=(X=1,Y=1,Z=1))))
    WeightMinimum=0
    Weight=@"UpperArm Readers".Fr
    ExecutePin=@RigVMFunction_Sequence.F
   )
   > &RigUnit_ModifyTransforms(
    ItemToModify=((Item=(Type=Bone,Name="def_arm_fr_l_end"),Transform=(Rotation=(X=0,Y=0,Z=-0.866,W=-0.5),Translation=(X=0,Y=0,Z=0),Scale3D=(X=1,Y=0.5,Z=1))))
    WeightMinimum=0
    Weight=@"UpperArm Readers".Fr
   )
   > &RigUnit_ModifyTransforms(
    ItemToModify=((Item=(Type=Bone,Name="def_arm_tp_l"),Transform=(Rotation=(X=0,Y=0,Z=0.2574,W=0.9663),Translation=(X=0,Y=0,Z=0),Scale3D=(X=1,Y=1,Z=1))))
    WeightMinimum=0
    Weight=@"UpperArm Readers".Fr
   )
   > &RigUnit_ModifyTransforms(
    ItemToModify=((Item=(Type=Bone,Name="def_arm_tp_l_end"),Transform=(Rotation=(X=0,Y=0,Z=0.7071,W=0.7071),Translation=(X=0,Y=0,Z=0),Scale3D=(X=1,Y=1,Z=1))))
    WeightMinimum=0
    Weight=@"UpperArm Readers".Fr
   )
   > &RigUnit_ModifyTransforms(
    ItemToModify=((Item=(Type=Bone,Name="def_arm_bt_l"),Transform=(Rotation=(X=0,Y=0,Z=0.2574,W=0.9663),Translation=(X=0,Y=0,Z=0),Scale3D=(X=1,Y=1,Z=1))))
    WeightMinimum=0
    Weight=@"UpperArm Readers".Fr
   )
   > &RigUnit_ModifyTransforms(
    ItemToModify=((Item=(Type=Bone,Name="def_arm_bt_l_end"),Transform=(Rotation=(X=0,Y=0,Z=-0.866,W=-0.5),Translation=(X=0,Y=0,Z=0),Scale3D=(X=1,Y=1,Z=1))))
    WeightMinimum=0
    Weight=@"UpperArm Readers".Fr
   )
   > &RigUnit_ModifyTransforms(
    ItemToModify=((Item=(Type=Bone,Name="def_arm_ba_l"),Transform=(Rotation=(X=0,Y=0,Z=0,W=1),Translation=(X=2.3629,Y=-0.586,Z=0),Scale3D=(X=1,Y=1,Z=1))))
    WeightMinimum=0
    Weight=@"UpperArm Readers".Fr
   )
   > &RigUnit_ModifyTransforms(
    ItemToModify=((Item=(Type=Bone,Name="def_arm_ba_l_end"),Transform=(Rotation=(X=0,Y=0,Z=-0.2382,W=-0.9712),Translation=(X=0,Y=0,Z=0),Scale3D=(X=1,Y=1,Z=1))))
    WeightMinimum=0
    Weight=@"UpperArm Readers".Fr
   )
   > &RigUnit_ModifyTransforms(
    ItemToModify=((Item=(Type=Bone,Name="def_arm_fr_r"),Transform=(Rotation=(X=0,Y=0,Z=0,W=1),Translation=(X=6.3509,Y=-3.105,Z=0),Scale3D=(X=1,Y=1,Z=1))))
    WeightMinimum=0
    Weight=@"UpperArm Readers_1".Fr
   )
   > &RigUnit_ModifyTransforms(
    ItemToModify=((Item=(Type=Bone,Name="def_arm_fr_r_end"),Transform=(Rotation=(X=0,Y=0,Z=-0.866,W=-0.5),Translation=(X=0,Y=0,Z=0),Scale3D=(X=1,Y=0.5,Z=1))))
    WeightMinimum=0
    Weight=@"UpperArm Readers_1".Fr
   )
   > &RigUnit_ModifyTransforms(
    ItemToModify=((Item=(Type=Bone,Name="def_arm_tp_r"),Transform=(Rotation=(X=0,Y=0,Z=0.2574,W=0.9663),Translation=(X=0,Y=0,Z=0),Scale3D=(X=1,Y=1,Z=1))))
    WeightMinimum=0
    Weight=@"UpperArm Readers_1".Fr
   )
   > &RigUnit_ModifyTransforms(
    ItemToModify=((Item=(Type=Bone,Name="def_arm_tp_r_end"),Transform=(Rotation=(X=0,Y=0,Z=0.7071,W=0.7071),Translation=(X=0,Y=0,Z=0),Scale3D=(X=1,Y=1,Z=1))))
    WeightMinimum=0
    Weight=@"UpperArm Readers_1".Fr
   )
   > &RigUnit_ModifyTransforms(
    ItemToModify=((Item=(Type=Bone,Name="def_arm_bt_r"),Transform=(Rotation=(X=0,Y=0,Z=0.2574,W=0.9663),Translation=(X=0,Y=0,Z=0),Scale3D=(X=1,Y=1,Z=1))))
    WeightMinimum=0
    Weight=@"UpperArm Readers_1".Fr
   )
   > &RigUnit_ModifyTransforms(
    ItemToModify=((Item=(Type=Bone,Name="def_arm_bt_r_end"),Transform=(Rotation=(X=0,Y=0,Z=-0.866,W=-0.5),Translation=(X=0,Y=0,Z=0),Scale3D=(X=1,Y=1,Z=1))))
    WeightMinimum=0
    Weight=@"UpperArm Readers_1".Fr
   )
   > &RigUnit_ModifyTransforms(
    ItemToModify=((Item=(Type=Bone,Name="def_arm_ba_r"),Transform=(Rotation=(X=0,Y=0,Z=0,W=1),Translation=(X=-2.3629,Y=0.586,Z=0),Scale3D=(X=1,Y=1,Z=1))))
    WeightMinimum=0
    Weight=@"UpperArm Readers_1".Fr
   )
   > &RigUnit_ModifyTransforms(
    ItemToModify=((Item=(Type=Bone,Name="def_arm_ba_r_end"),Transform=(Rotation=(X=0,Y=0,Z=-0.2382,W=-0.9712),Translation=(X=0,Y=0,Z=0),Scale3D=(X=1,Y=1,Z=1))))
    WeightMinimum=0
    Weight=@"UpperArm Readers_1".Fr
   )
   &RigUnit_ModifyTransforms(
    ItemToModify=((Item=(Type=Bone,Name="def_arm_ba_l"),Transform=(Rotation=(X=0,Y=0,Z=0,W=1),Translation=(X=-9.9444,Y=-1.4272,Z=0),Scale3D=(X=1,Y=1,Z=1))))
    WeightMinimum=0
    Weight=@"UpperArm Readers".Ba
    ExecutePin=@RigVMFunction_Sequence.G
   )
   > &RigUnit_ModifyTransforms(
    ItemToModify=((Item=(Type=Bone,Name="def_arm_ba_l_end"),Transform=(Rotation=(X=0,Y=0,Z=0.1374,W=-0.9905),Translation=(X=-8.3525,Y=-7.5437,Z=0),Scale3D=(X=1,Y=0.5,Z=1))))
    WeightMinimum=0
    Weight=@"UpperArm Readers".Ba
   )
   > &RigUnit_ModifyTransforms(
    ItemToModify=((Item=(Type=Bone,Name="def_arm_tp_l"),Transform=(Rotation=(X=0,Y=0,Z=-0.2574,W=0.9663),Translation=(X=0,Y=0,Z=0),Scale3D=(X=1,Y=1,Z=1))))
    WeightMinimum=0
    Weight=@"UpperArm Readers".Ba
   )
   > &RigUnit_ModifyTransforms(
    ItemToModify=((Item=(Type=Bone,Name="def_arm_tp_l_end"),Transform=(Rotation=(X=0,Y=0,Z=-0.7071,W=0.7071),Translation=(X=0,Y=0,Z=0),Scale3D=(X=1,Y=1,Z=1))))
    WeightMinimum=0
    Weight=@"UpperArm Readers".Ba
   )
   > &RigUnit_ModifyTransforms(
    ItemToModify=((Item=(Type=Bone,Name="def_arm_bt_l"),Transform=(Rotation=(X=0,Y=0,Z=-0.2574,W=0.9663),Translation=(X=0,Y=0,Z=0),Scale3D=(X=1,Y=1,Z=1))))
    WeightMinimum=0
    Weight=@"UpperArm Readers".Ba
   )
   > &RigUnit_ModifyTransforms(
    ItemToModify=((Item=(Type=Bone,Name="def_arm_bt_l_end"),Transform=(Rotation=(X=0,Y=0,Z=-0.7071,W=0.7071),Translation=(X=0,Y=0,Z=0),Scale3D=(X=1,Y=1,Z=1))))
    WeightMinimum=0
    Weight=@"UpperArm Readers".Ba
   )
   > &RigUnit_ModifyTransforms(
    ItemToModify=((Item=(Type=Bone,Name="def_arm_fr_l"),Transform=(Rotation=(X=0,Y=0,Z=-0.1776,W=0.9841),Translation=(X=2.2284,Y=0,Z=0),Scale3D=(X=1,Y=1,Z=1))))
    WeightMinimum=0
    Weight=@"UpperArm Readers".Ba
   )
   > &RigUnit_ModifyTransforms(
    ItemToModify=((Item=(Type=Bone,Name="def_arm_fr_l_end"),Transform=(Rotation=(X=0,Y=0,Z=-0.2588,W=0.9659),Translation=(X=0,Y=0,Z=0),Scale3D=(X=1,Y=1,Z=1))))
    WeightMinimum=0
    Weight=@"UpperArm Readers".Ba
   )
   > &RigUnit_ModifyTransforms(
    ItemToModify=((Item=(Type=Bone,Name="def_arm_ba_r"),Transform=(Rotation=(X=0,Y=0,Z=0,W=1),Translation=(X=9.9444,Y=1.4272,Z=0),Scale3D=(X=1,Y=1,Z=1))))
    WeightMinimum=0
    Weight=@"UpperArm Readers_1".Ba
   )
   > &RigUnit_ModifyTransforms(
    ItemToModify=((Item=(Type=Bone,Name="def_arm_ba_r_end"),Transform=(Rotation=(X=0,Y=0,Z=0.1374,W=-0.9905),Translation=(X=8.3525,Y=7.5437,Z=0),Scale3D=(X=1,Y=0.5,Z=1))))
    WeightMinimum=0
    Weight=@"UpperArm Readers_1".Ba
   )
   > &RigUnit_ModifyTransforms(
    ItemToModify=((Item=(Type=Bone,Name="def_arm_tp_r"),Transform=(Rotation=(X=0,Y=0,Z=-0.2574,W=0.9663),Translation=(X=0,Y=0,Z=0),Scale3D=(X=1,Y=1,Z=1))))
    WeightMinimum=0
    Weight=@"UpperArm Readers_1".Ba
   )
   > &RigUnit_ModifyTransforms(
    ItemToModify=((Item=(Type=Bone,Name="def_arm_tp_r_end"),Transform=(Rotation=(X=0,Y=0,Z=-0.7071,W=0.7071),Translation=(X=0,Y=0,Z=0),Scale3D=(X=1,Y=1,Z=1))))
    WeightMinimum=0
    Weight=@"UpperArm Readers_1".Ba
   )
   > &RigUnit_ModifyTransforms(
    ItemToModify=((Item=(Type=Bone,Name="def_arm_bt_r"),Transform=(Rotation=(X=0,Y=0,Z=-0.2574,W=0.9663),Translation=(X=0,Y=0,Z=0),Scale3D=(X=1,Y=1,Z=1))))
    WeightMinimum=0
    Weight=@"UpperArm Readers_1".Ba
   )
   > &RigUnit_ModifyTransforms(
    ItemToModify=((Item=(Type=Bone,Name="def_arm_bt_r_end"),Transform=(Rotation=(X=0,Y=0,Z=-0.7071,W=0.7071),Translation=(X=0,Y=0,Z=0),Scale3D=(X=1,Y=1,Z=1))))
    WeightMinimum=0
    Weight=@"UpperArm Readers_1".Ba
   )
   > &RigUnit_ModifyTransforms(
    ItemToModify=((Item=(Type=Bone,Name="def_arm_fr_r"),Transform=(Rotation=(X=0,Y=0,Z=-0.1776,W=0.9841),Translation=(X=-2.2284,Y=0,Z=0),Scale3D=(X=1,Y=1,Z=1))))
    WeightMinimum=0
    Weight=@"UpperArm Readers_1".Ba
   )
   > &RigUnit_ModifyTransforms(
    ItemToModify=((Item=(Type=Bone,Name="def_arm_fr_r_end"),Transform=(Rotation=(X=0,Y=0,Z=-0.2588,W=0.9659),Translation=(X=0,Y=0,Z=0),Scale3D=(X=1,Y=1,Z=1))))
    WeightMinimum=0
    Weight=@"UpperArm Readers_1".Ba
   )
   n3 = &RigUnit_SphericalPoseReader(
    ActiveRegionScaleFactors=(PositiveWidth=1,NegativeWidth=1,PositiveHeight=1,NegativeHeight=1)
    ActiveRegionSize=0.001
    Debug=(bDrawDebug=false,bDraw2D=False,bDrawLocalAxes=False,DebugScale=20,DebugSegments=20,DebugThickness=0.05)
    DriverAxis=(X=1,Y=0,Z=0)
    DriverItem=(Type=Bone,Name="clavicle_l")
    FalloffRegionScaleFactors=(PositiveWidth=1,NegativeWidth=1,PositiveHeight=1,NegativeHeight=1)
    FalloffSize=0.5
    FlipHeightScaling=False
    FlipWidthScaling=False
    OptionalParentItem=(Type=Bone,Name="None")
    RotationOffset=(X=0,Y=0,Z=0)
    ExecutePin=@RigVMFunction_Sequence.H
   )
   > &RigUnit_ModifyTransforms(
    ItemToModify=((Item=(Type=Bone,Name="collar_l_end"),Transform=(Rotation=(X=0,Y=0.7716,Z=0,W=-0.6362),Translation=(X=0,Y=0,Z=0),Scale3D=(X=1,Y=1,Z=1))))
    WeightMinimum=0
    Weight=@n3.OutputParam
   )
   > &RigUnit_ModifyTransforms(
    ItemToModify=((Item=(Type=Bone,Name="collar_l"),Transform=(Rotation=(X=0,Y=0,Z=0,W=1),Translation=(X=6.533,Y=0,Z=0),Scale3D=(X=1,Y=1,Z=1))))
    WeightMinimum=0
    Weight=@n3.OutputParam
   )
   > &RigUnit_ModifyTransforms(
    ItemToModify=((Item=(Type=Bone,Name="flap_l_MSTR"),Transform=(Rotation=(X=0,Y=-0.9659,Z=0,W=0.2588),Translation=(X=0,Y=0,Z=0),Scale3D=(X=1,Y=1,Z=1))))
    WeightMinimum=0
    Weight=@n3.OutputParam
   )
   > &RigUnit_ModifyTransforms(
    ItemToModify=((Item=(Type=Bone,Name="flap_fr_l_MSTR"),Transform=(Rotation=(X=0,Y=-0.5,Z=0,W=0.866),Translation=(X=0,Y=0,Z=0),Scale3D=(X=1,Y=1,Z=1))))
    WeightMinimum=0
    Weight=@n3.OutputParam
   )
   > &RigUnit_ModifyTransforms(
    ItemToModify=((Item=(Type=Bone,Name="def_arm_bt_l"),Transform=(Rotation=(X=0,Y=0,Z=0,W=-1),Translation=(X=0.2146,Y=0,Z=27.2451),Scale3D=(X=1,Y=1,Z=1))))
    WeightMinimum=0
    Weight=@n3.OutputParam
   )
   > &RigUnit_ModifyTransforms(
    ItemToModify=((Item=(Type=Bone,Name="def_arm_bt_l_end"),Transform=(Rotation=(X=0,Y=-0.5,Z=0,W=0.866),Translation=(X=0,Y=0,Z=0),Scale3D=(X=1,Y=1,Z=1))))
    WeightMinimum=0
    Weight=@n3.OutputParam
   )
   > &RigUnit_ModifyTransforms(
    ItemToModify=((Item=(Type=Bone,Name="def_arm_fr_l"),Transform=(Rotation=(X=0,Y=0.3079,Z=0,W=0.9514),Translation=(X=0.1406,Y=0,Z=-2.3427),Scale3D=(X=1,Y=1,Z=1))))
    WeightMinimum=0
    Weight=@n3.OutputParam
   )
   > &RigUnit_ModifyTransforms(
    ItemToModify=((Item=(Type=Bone,Name="def_arm_fr_l_end"),Transform=(Rotation=(X=0,Y=0,Z=0,W=1),Translation=(X=0,Y=0,Z=0),Scale3D=(X=1,Y=1,Z=1))))
    WeightMinimum=0
    Weight=@n3.OutputParam
   )
   > &RigUnit_ModifyTransforms(
    ItemToModify=((Item=(Type=Bone,Name="def_arm_ba_l"),Transform=(Rotation=(X=0,Y=0.3989,Z=0,W=0.917),Translation=(X=-1.6254,Y=0,Z=-1.8679),Scale3D=(X=1,Y=1,Z=1))))
    WeightMinimum=0
    Weight=@n3.OutputParam
   )
   > &RigUnit_ModifyTransforms(
    ItemToModify=((Item=(Type=Bone,Name="def_arm_ba_l_end"),Transform=(Rotation=(X=0,Y=0,Z=0,W=1),Translation=(X=0,Y=0,Z=0),Scale3D=(X=1,Y=1,Z=1))))
    WeightMinimum=0
    Weight=@n3.OutputParam
   )
   > &RigUnit_ModifyTransforms(
    ItemToModify=((Item=(Type=Bone,Name="collar_ba"),Transform=(Rotation=(X=0,Y=0,Z=0,W=-1),Translation=(X=0.591,Y=0,Z=0),Scale3D=(X=1,Y=1,Z=1))))
    WeightMinimum=0
    Weight=@n3.OutputParam
   )
   > &RigUnit_ModifyTransforms(
    ItemToModify=((Item=(Type=Bone,Name="def_arm_tp_l"),Transform=(Rotation=(X=0,Y=0,Z=0,W=1),Translation=(X=3.1813,Y=0,Z=-2.674),Scale3D=(X=1,Y=1,Z=1))))
    WeightMinimum=0
    Weight=@n3.OutputParam
   )
   > &RigUnit_ModifyTransforms(
    ItemToModify=((Item=(Type=Bone,Name="def_arm_tp_l_end"),Transform=(Rotation=(X=0,Y=-0.1631,Z=0,W=-0.9866),Translation=(X=0,Y=0,Z=0),Scale3D=(X=1,Y=1,Z=1))))
    WeightMinimum=0
    Weight=@n3.OutputParam
   )
   > n4 = &RigUnit_SphericalPoseReader(
    ActiveRegionScaleFactors=(PositiveWidth=1,NegativeWidth=1,PositiveHeight=1,NegativeHeight=1)
    ActiveRegionSize=0.001
    Debug=(bDrawDebug=false,bDraw2D=False,bDrawLocalAxes=False,DebugScale=20,DebugSegments=20,DebugThickness=0.05)
    DriverAxis=(X=1,Y=0,Z=0)
    DriverItem=(Type=Bone,Name="clavicle_r")
    FalloffRegionScaleFactors=(PositiveWidth=1,NegativeWidth=1,PositiveHeight=1,NegativeHeight=1)
    FalloffSize=0.5
    FlipHeightScaling=False
    FlipWidthScaling=False
    OptionalParentItem=(Type=Bone,Name="None")
    RotationOffset=(X=0,Y=0,Z=0)
   )
   > &RigUnit_ModifyTransforms(
    ItemToModify=((Item=(Type=Bone,Name="collar_r"),Transform=(Rotation=(X=0,Y=0,Z=0,W=1),Translation=(X=4.5404,Y=0,Z=0),Scale3D=(X=1,Y=1,Z=1))))
    WeightMinimum=0
    Weight=@n4.OutputParam
   )
   > &RigUnit_ModifyTransforms(
    ItemToModify=((Item=(Type=Bone,Name="collar_r_end"),Transform=(Rotation=(X=0,Y=0.7406,Z=0,W=0.672),Translation=(X=0,Y=0,Z=0),Scale3D=(X=1,Y=1,Z=1))))
    WeightMinimum=0
    Weight=@n4.OutputParam
   )
   > &RigUnit_ModifyTransforms(
    ItemToModify=((Item=(Type=Bone,Name="flap_r_MSTR"),Transform=(Rotation=(X=0,Y=0.9659,Z=0,W=0.2588),Translation=(X=0,Y=0,Z=0),Scale3D=(X=1,Y=1,Z=1))))
    WeightMinimum=0
    Weight=@n4.OutputParam
   )
   > &RigUnit_ModifyTransforms(
    ItemToModify=((Item=(Type=Bone,Name="flap_fr_r_MSTR"),Transform=(Rotation=(X=0,Y=0.9659,Z=0,W=0.2588),Translation=(X=0,Y=0,Z=0),Scale3D=(X=1,Y=1,Z=1))))
    WeightMinimum=0
    Weight=@n4.OutputParam
   )
   > &RigUnit_ModifyTransforms(
    ItemToModify=((Item=(Type=Bone,Name="def_arm_bt_r"),Transform=(Rotation=(X=0,Y=0,Z=0,W=-1),Translation=(X=1.06,Y=0,Z=-25.722),Scale3D=(X=1,Y=1,Z=1))))
    WeightMinimum=0
    Weight=@n4.OutputParam
   )
   > &RigUnit_ModifyTransforms(
    ItemToModify=((Item=(Type=Bone,Name="def_arm_bt_r_end"),Transform=(Rotation=(X=0,Y=-0.5,Z=0,W=0.866),Translation=(X=0,Y=0,Z=0),Scale3D=(X=1,Y=1,Z=1))))
    WeightMinimum=0
    Weight=@n4.OutputParam
   )
   > &RigUnit_ModifyTransforms(
    ItemToModify=((Item=(Type=Bone,Name="def_arm_fr_r"),Transform=(Rotation=(X=0,Y=-0.2544,Z=0,W=-0.9671),Translation=(X=-0.0133,Y=0,Z=0.1892),Scale3D=(X=1,Y=1,Z=1))))
    WeightMinimum=0
    Weight=@n4.OutputParam
   )
   > &RigUnit_ModifyTransforms(
    ItemToModify=((Item=(Type=Bone,Name="def_arm_fr_r_end"),Transform=(Rotation=(X=0,Y=0,Z=0,W=1),Translation=(X=0,Y=0,Z=0),Scale3D=(X=1,Y=1,Z=1))))
    WeightMinimum=0
    Weight=@n4.OutputParam
   )
   > &RigUnit_ModifyTransforms(
    ItemToModify=((Item=(Type=Bone,Name="def_arm_ba_r"),Transform=(Rotation=(X=0,Y=0.2588,Z=0,W=0.9659),Translation=(X=2.755,Y=0,Z=-1.6403),Scale3D=(X=1,Y=1,Z=1))))
    WeightMinimum=0
    Weight=@n4.OutputParam
   )
   > &RigUnit_ModifyTransforms(
    ItemToModify=((Item=(Type=Bone,Name="def_arm_ba_r_end"),Transform=(Rotation=(X=0,Y=0,Z=0,W=1),Translation=(X=0,Y=0,Z=0),Scale3D=(X=1,Y=1,Z=1))))
    WeightMinimum=0
    Weight=@n4.OutputParam
   )
   > &RigUnit_ModifyTransforms(
    ItemToModify=((Item=(Type=Bone,Name="collar_ba"),Transform=(Rotation=(X=0,Y=0,Z=0,W=-1),Translation=(X=0.6,Y=0,Z=0),Scale3D=(X=1,Y=1,Z=1))))
    WeightMinimum=0
    Weight=@n4.OutputParam
   )
   > &RigUnit_ModifyTransforms(
    ItemToModify=((Item=(Type=Bone,Name="def_arm_tp_r"),Transform=(Rotation=(X=0,Y=0,Z=0,W=-1),Translation=(X=-5.1286,Y=0,Z=3.185),Scale3D=(X=1,Y=1,Z=1))))
    WeightMinimum=0
    Weight=@n4.OutputParam
   )
   > &RigUnit_ModifyTransforms(
    ItemToModify=((Item=(Type=Bone,Name="def_arm_tp_r_end"),Transform=(Rotation=(X=0,Y=-0.1631,Z=0,W=-0.9866),Translation=(X=0,Y=0,Z=0),Scale3D=(X=1,Y=1,Z=1))))
    WeightMinimum=0
    Weight=@n4.OutputParam
   )
   n5 = &RigUnit_SphericalPoseReader(
    ActiveRegionScaleFactors=(PositiveWidth=1,NegativeWidth=1,PositiveHeight=1,NegativeHeight=1)
    ActiveRegionSize=0.001
    Debug=(bDrawDebug=false,bDraw2D=False,bDrawLocalAxes=False,DebugScale=20,DebugSegments=20,DebugThickness=0.05)
    DriverAxis=(X=1,Y=0,Z=0)
    DriverItem=(Type=Bone,Name="clavicle_l")
    FalloffRegionScaleFactors=(PositiveWidth=1,NegativeWidth=1,PositiveHeight=1,NegativeHeight=1)
    FalloffSize=0.5
    FlipHeightScaling=False
    FlipWidthScaling=False
    OptionalParentItem=(Type=Bone,Name="None")
    RotationOffset=(X=90,Y=0,Z=0)
    ExecutePin=@RigVMFunction_Sequence.I
   )
   > &RigUnit_ModifyTransforms(
    ItemToModify=((Item=(Type=Bone,Name="collar_l"),Transform=(Rotation=(X=-0.2562,Y=0,Z=0,W=-0.9666),Translation=(X=0,Y=0.4746,Z=1.1463),Scale3D=(X=1,Y=1,Z=1))))
    WeightMinimum=0
    Weight=@n5.OutputParam
   )
   > &RigUnit_ModifyTransforms(
    ItemToModify=((Item=(Type=Bone,Name="collar_l_end"),Transform=(Rotation=(X=0.526,Y=0,Z=0,W=0.8505),Translation=(X=0,Y=0,Z=0),Scale3D=(X=1,Y=1,Z=1))))
    WeightMinimum=0
    Weight=@n5.OutputParam
   )
   > &RigUnit_ModifyTransforms(
    ItemToModify=((Item=(Type=Bone,Name="lapel_l"),Transform=(Rotation=(X=0.0469,Y=0,Z=0,W=0.9989),Translation=(X=0,Y=2.7186,Z=10.2693),Scale3D=(X=1,Y=1,Z=1))))
    WeightMinimum=0
    Weight=@n5.OutputParam
   )
   > &RigUnit_ModifyTransforms(
    ItemToModify=((Item=(Type=Bone,Name="def_arm_bt_l"),Transform=(Rotation=(X=0,Y=0,Z=-0.3575,W=-0.9339),Translation=(X=-6.7269,Y=13.1661,Z=0),Scale3D=(X=1,Y=1,Z=1))))
    WeightMinimum=0
    Weight=@n5.OutputParam
   )
   > n6 = &RigUnit_SphericalPoseReader(
    ActiveRegionScaleFactors=(PositiveWidth=1,NegativeWidth=1,PositiveHeight=1,NegativeHeight=1)
    ActiveRegionSize=0.001
    Debug=(bDrawDebug=false,bDraw2D=False,bDrawLocalAxes=False,DebugScale=20,DebugSegments=20,DebugThickness=0.05)
    DriverAxis=(X=1,Y=0,Z=0)
    DriverItem=(Type=Bone,Name="clavicle_r")
    FalloffRegionScaleFactors=(PositiveWidth=1,NegativeWidth=1,PositiveHeight=1,NegativeHeight=1)
    FalloffSize=0.5
    FlipHeightScaling=False
    FlipWidthScaling=False
    OptionalParentItem=(Type=Bone,Name="None")
    RotationOffset=(X=90,Y=0,Z=0)
   )
   > &RigUnit_ModifyTransforms(
    ItemToModify=((Item=(Type=Bone,Name="collar_r"),Transform=(Rotation=(X=-0.2562,Y=0,Z=0,W=0.9666),Translation=(X=0,Y=-0.4746,Z=-1.1463),Scale3D=(X=1,Y=1,Z=1))))
    WeightMinimum=0
    Weight=@n6.OutputParam
   )
   > &RigUnit_ModifyTransforms(
    ItemToModify=((Item=(Type=Bone,Name="collar_r_end"),Transform=(Rotation=(X=-0.396,Y=0,Z=0,W=0.9183),Translation=(X=0,Y=0,Z=0),Scale3D=(X=1,Y=1,Z=1))))
    WeightMinimum=0
    Weight=@n6.OutputParam
   )
   > &RigUnit_ModifyTransforms(
    ItemToModify=((Item=(Type=Bone,Name="lapel_r"),Transform=(Rotation=(X=-0.2588,Y=0,Z=0,W=0.9659),Translation=(X=0,Y=-0.532,Z=-8.4548),Scale3D=(X=1,Y=1,Z=1))))
    WeightMinimum=0
    Weight=@n6.OutputParam
   )
   > &RigUnit_ModifyTransforms(
    ItemToModify=((Item=(Type=Bone,Name="def_arm_bt_r"),Transform=(Rotation=(X=0,Y=0,Z=-0.4587,W=-0.8886),Translation=(X=6.7267,Y=-13.1655,Z=0),Scale3D=(X=1,Y=1,Z=1))))
    WeightMinimum=0
    Weight=@n6.OutputParam
   )
   > &RigUnit_ModifyTransforms(
    ItemToModify=((Item=(Type=Bone,Name="flap_fr_r_MSTR"),Transform=(Rotation=(X=0,Y=0.4789,Z=0,W=0.8779),Translation=(X=0,Y=0,Z=0),Scale3D=(X=1,Y=1,Z=1))))
    WeightMinimum=0
    Weight=@n6.OutputParam
   )
   > n7 = &RigUnit_SphericalPoseReader(
    ActiveRegionScaleFactors=(PositiveWidth=1,NegativeWidth=1,PositiveHeight=1,NegativeHeight=1)
    ActiveRegionSize=0.001
    Debug=(bDrawDebug=false,bDraw2D=False,bDrawLocalAxes=False,DebugScale=20,DebugSegments=20,DebugThickness=0.05)
    DriverAxis=(X=1,Y=0,Z=0)
    DriverItem=(Type=Bone,Name="clavicle_l")
    FalloffRegionScaleFactors=(PositiveWidth=1,NegativeWidth=1,PositiveHeight=1,NegativeHeight=1)
    FalloffSize=0.5
    FlipHeightScaling=False
    FlipWidthScaling=False
    OptionalParentItem=(Type=Bone,Name="None")
    RotationOffset=(X=-90,Y=0,Z=0)
   )
   > &RigUnit_ModifyTransforms(
    ItemToModify=((Item=(Type=Bone,Name="collar_l"),Transform=(Rotation=(X=-0.1515,Y=0,Z=0,W=0.9885),Translation=(X=0,Y=-0.8084,Z=-0.6158),Scale3D=(X=1,Y=1,Z=1))))
    WeightMinimum=0
    Weight=@n7.OutputParam
   )
   > &RigUnit_ModifyTransforms(
    ItemToModify=((Item=(Type=Bone,Name="collar_l_end"),Transform=(Rotation=(X=-0.5,Y=0,Z=0,W=0.866),Translation=(X=0,Y=0,Z=0),Scale3D=(X=1,Y=1,Z=1))))
    WeightMinimum=0
    Weight=@n7.OutputParam
   )
   > &RigUnit_ModifyTransforms(
    ItemToModify=((Item=(Type=Bone,Name="lapel_l"),Transform=(Rotation=(X=-0.1086,Y=0,Z=0,W=0.9941),Translation=(X=0,Y=0,Z=-1.866),Scale3D=(X=1,Y=1,Z=1))))
    WeightMinimum=0
    Weight=@n7.OutputParam
   )
   > &RigUnit_ModifyTransforms(
    ItemToModify=((Item=(Type=Bone,Name="def_arm_bt_l"),Transform=(Rotation=(X=0,Y=0,Z=-0.7042,W=0.71),Translation=(X=-5.331,Y=-11.8085,Z=0),Scale3D=(X=1,Y=1,Z=1))))
    WeightMinimum=0
    Weight=@n7.OutputParam
   )
   > n8 = &RigUnit_SphericalPoseReader(
    ActiveRegionScaleFactors=(PositiveWidth=1,NegativeWidth=1,PositiveHeight=1,NegativeHeight=1)
    ActiveRegionSize=0.001
    Debug=(bDrawDebug=false,bDraw2D=False,bDrawLocalAxes=False,DebugScale=20,DebugSegments=20,DebugThickness=0.05)
    DriverAxis=(X=1,Y=0,Z=0)
    DriverItem=(Type=Bone,Name="clavicle_r")
    FalloffRegionScaleFactors=(PositiveWidth=1,NegativeWidth=1,PositiveHeight=1,NegativeHeight=1)
    FalloffSize=0.5
    FlipHeightScaling=False
    FlipWidthScaling=False
    OptionalParentItem=(Type=Bone,Name="None")
    RotationOffset=(X=-90,Y=0,Z=0)
   )
   > &RigUnit_ModifyTransforms(
    ItemToModify=((Item=(Type=Bone,Name="collar_r"),Transform=(Rotation=(X=0.1515,Y=0,Z=0,W=0.9885),Translation=(X=0,Y=-0.8084,Z=0.6158),Scale3D=(X=1,Y=1,Z=1))))
    WeightMinimum=0
    Weight=@n8.OutputParam
   )
   > &RigUnit_ModifyTransforms(
    ItemToModify=((Item=(Type=Bone,Name="collar_r_end"),Transform=(Rotation=(X=0.5,Y=0,Z=0,W=0.866),Translation=(X=0,Y=0,Z=0),Scale3D=(X=1,Y=1,Z=1))))
    WeightMinimum=0
    Weight=@n8.OutputParam
   )
   > &RigUnit_ModifyTransforms(
    ItemToModify=((Item=(Type=Bone,Name="lapel_r"),Transform=(Rotation=(X=-0.1086,Y=0,Z=0,W=0.9941),Translation=(X=0,Y=0,Z=1.866),Scale3D=(X=1,Y=1,Z=1))))
    WeightMinimum=0
    Weight=@n8.OutputParam
   )
   > &RigUnit_ModifyTransforms(
    ItemToModify=((Item=(Type=Bone,Name="def_arm_bt_r"),Transform=(Rotation=(X=0,Y=0,Z=0.665,W=-0.7468),Translation=(X=4.6509,Y=8.996,Z=0),Scale3D=(X=1,Y=1,Z=1))))
    WeightMinimum=0
    Weight=@n8.OutputParam
   )
   Thigh Readers = RigVMFunctionReferenceNode(
    Ba Post Viz=false
    Ba Pre Viz=false
    Ba Viz=false
    Driver Item=(Type=Bone,Name="thigh_l")
    Fr Post Viz=false
    Fr Pre Viz=false
    Fr Viz=false
    In Post Viz=false
    In Pre Viz=false
    In Viz=false
    Ot Post Viz=false
    Ot Pre Viz=false
    Ot Viz=false
    ExecuteContext=@RigVMFunction_Sequence.J
   )
   > &RigUnit_ModifyTransforms(
    ItemToModify="((Item=(Type=Bone,Name=\"leg_crease_fr_l\"),Transform=(Rotation=(X=0,Y=0,Z=-0.5696,W=0.8219),Translation=(X=23.841,Y=5.7143,Z=0),Scale3D=(X=1,Y=1,Z=1))),(Item=(Type=Bone,Name=\"belt_fr_l\"),Transform=(Rotation=(X=0,Y=0,Z=0,W=-1),Translation=(X=8.0514,Y=0,Z=0), ...(truncated)"
    WeightMinimum=0
    Weight=@"Thigh Readers".Fr
   )
   > &RigUnit_ModifyTransforms(
    ItemToModify="((Item=(Type=Bone,Name=\"fold_02_fr_l\"),Transform=(Rotation=(X=0,Y=0,Z=-0.309,W=0.9511),Translation=(X=0,Y=0,Z=0),Scale3D=(X=1,Y=1,Z=1))),(Item=(Type=Bone,Name=\"fold_03_fr_l\"),Transform=(Rotation=(X=0,Y=0,Z=-0.309,W=0.9511),Translation=(X=0,Y=0,Z=0),Scale3D ...(truncated)"
    WeightMinimum=0
    Weight=@"Thigh Readers".Fr
   )
   > &RigUnit_ModifyTransforms(
    ItemToModify=((Item=(Type=Bone,Name="glute_l"),Transform=(Rotation=(X=0,Y=0,Z=-0.2908,W=0.9568),Translation=(X=-4.7598,Y=1.2462,Z=0),Scale3D=(X=1,Y=1,Z=1))))
    WeightMinimum=0
    Weight=@"Thigh Readers".Fr
   )
   > &RigUnit_ModifyTransforms(
    ItemToModify="((Item=(Type=Bone,Name=\"leg_crease_fr_l\"),Transform=(Rotation=(X=0,Y=0,Z=-0.5696,W=0.8219),Translation=(X=23.841,Y=5.7143,Z=0),Scale3D=(X=1,Y=1,Z=1))),(Item=(Type=Bone,Name=\"belt_fr_l\"),Transform=(Rotation=(X=0,Y=0,Z=0,W=-1),Translation=(X=8.0514,Y=0,Z=0), ...(truncated)"
    WeightMinimum=0
    Weight=@"Thigh Readers".Fr Post
   )
   > &RigUnit_ModifyTransforms(
    ItemToModify="((Item=(Type=Bone,Name=\"fold_02_fr_l\"),Transform=(Rotation=(X=0,Y=0,Z=-0.5,W=0.866),Translation=(X=0,Y=0,Z=0),Scale3D=(X=1,Y=1,Z=1))),(Item=(Type=Bone,Name=\"fold_02_diag_fr_l\"),Transform=(Rotation=(X=0,Y=0,Z=-0.4617,W=0.887),Translation=(X=0,Y=0,Z=0),Scale ...(truncated)"
    WeightMinimum=0
    Weight=@"Thigh Readers".Fr Post
   )
   > &RigUnit_ModifyTransforms(
    ItemToModify="((Item=(Type=Bone,Name=\"belt_ba_l\"),Transform=(Rotation=(X=0,Y=0,Z=0,W=-1),Translation=(X=7.3326,Y=0,Z=0),Scale3D=(X=1,Y=1,Z=1))),(Item=(Type=Bone,Name=\"glute_l\"),Transform=(Rotation=(X=0,Y=0,Z=0.4535,W=0.8912),Translation=(X=12.1561,Y=-4.1782,Z=0),Scale3D ...(truncated)"
    WeightMinimum=0
    Weight=@"Thigh Readers".Ba
   )
   > &RigUnit_ModifyTransforms(
    ItemToModify="((Item=(Type=Bone,Name=\"fold_02_leg_ba_l\"),Transform=(Rotation=(X=0,Y=0,Z=-0.309,W=0.9511),Translation=(X=0,Y=0,Z=0),Scale3D=(X=1,Y=1,Z=1))),(Item=(Type=Bone,Name=\"fold_03_leg_ba_l\"),Transform=(Rotation=(X=0,Y=0,Z=-0.309,W=0.9511),Translation=(X=0,Y=0,Z=0) ...(truncated)"
    WeightMinimum=0
    Weight=@"Thigh Readers".Ba
   )
   > &RigUnit_ModifyTransforms(
    ItemToModify="((Item=(Type=Bone,Name=\"belt_ba_l\"),Transform=(Rotation=(X=0,Y=0,Z=0,W=-1),Translation=(X=7.3326,Y=0,Z=0),Scale3D=(X=1,Y=1,Z=1))),(Item=(Type=Bone,Name=\"glute_l\"),Transform=(Rotation=(X=0,Y=0,Z=0.4535,W=0.8912),Translation=(X=17.7166,Y=-8.0449,Z=0.0001),Sc ...(truncated)"
    WeightMinimum=0
    Weight=@"Thigh Readers".Ba Post
   )
   > &RigUnit_ModifyTransforms(
    ItemToModify="((Item=(Type=Bone,Name=\"fold_01_leg_ba_r\"),Transform=(Rotation=(X=0,Y=0,Z=0,W=1),Translation=(X=0,Y=0,Z=0),Scale3D=(X=1,Y=1,Z=1))),(Item=(Type=Bone,Name=\"fold_01_ba_ct\"),Transform=(Rotation=(X=0,Y=0,Z=0,W=1),Translation=(X=0,Y=0,Z=0),Scale3D=(X=1,Y=1,Z=1)) ...(truncated)"
    WeightMinimum=0
    Weight=@"Thigh Readers".Ba Post
   )
   > &RigUnit_ModifyTransforms(
    ItemToModify="((Item=(Type=Bone,Name=\"belt_l\"),Transform=(Rotation=(X=0,Y=0,Z=0,W=-1),Translation=(X=7.914,Y=0,Z=0),Scale3D=(X=1,Y=1,Z=1))),(Item=(Type=Bone,Name=\"belt_fr_l\"),Transform=(Rotation=(X=0,Y=0.1456,Z=0,W=-0.9893),Translation=(X=2.9141,Y=0,Z=0),Scale3D=(X=1,Y= ...(truncated)"
    WeightMinimum=0
    Weight=@"Thigh Readers".Ot
   )
   > &RigUnit_ModifyTransforms(
    ItemToModify="((Item=(Type=Bone,Name=\"fold_01_side_l\"),Transform=(Rotation=(X=0,Y=0,Z=-0.1305,W=0.9914),Translation=(X=0,Y=0,Z=0),Scale3D=(X=1,Y=1,Z=1))),(Item=(Type=Bone,Name=\"fold_02_side_l\"),Transform=(Rotation=(X=0,Y=0,Z=-0.3827,W=0.9239),Translation=(X=0,Y=0,Z=0),S ...(truncated)"
    WeightMinimum=0
    Weight=@"Thigh Readers".Ot
   )
   > &RigUnit_ModifyTransforms(
    ItemToModify="((Item=(Type=Bone,Name=\"belt_l\"),Transform=(Rotation=(X=0,Y=0,Z=0,W=1),Translation=(X=12.3831,Y=0,Z=-0.2574),Scale3D=(X=1,Y=1,Z=1))),(Item=(Type=Bone,Name=\"belt_fr_l\"),Transform=(Rotation=(X=0,Y=0.1456,Z=0,W=-0.9893),Translation=(X=2.9141,Y=0,Z=0),Scale3D= ...(truncated)"
    WeightMinimum=0
    Weight=@"Thigh Readers".Ot Post
   )
   > &RigUnit_ModifyTransforms(
    ItemToModify="((Item=(Type=Bone,Name=\"fold_01_side_l\"),Transform=(Rotation=(X=0,Y=0,Z=0.7841,W=-0.6206),Translation=(X=0,Y=0,Z=0),Scale3D=(X=1,Y=1,Z=1))),(Item=(Type=Bone,Name=\"fold_02_side_l\"),Transform=(Rotation=(X=0,Y=0,Z=-0.2284,W=0.9736),Translation=(X=0,Y=0,Z=0),S ...(truncated)"
    WeightMinimum=0
    Weight=@"Thigh Readers".Ot Post
   )
   > Thigh Readers_1 = RigVMFunctionReferenceNode(
    Ba Post Viz=false
    Ba Pre Viz=false
    Ba Viz=false
    Driver Item=(Type=Bone,Name="thigh_r")
    Fr Post Viz=false
    Fr Pre Viz=false
    Fr Viz=false
    In Post Viz=false
    In Pre Viz=false
    In Viz=false
    Ot Post Viz=false
    Ot Pre Viz=false
    Ot Viz=false
   )
   > &RigUnit_ModifyTransforms(
    ItemToModify="((Item=(Type=Bone,Name=\"leg_crease_fr_r\"),Transform=(Rotation=(X=0,Y=0,Z=-0.5696,W=0.8219),Translation=(X=23.841,Y=5.7143,Z=0),Scale3D=(X=1,Y=1,Z=1))),(Item=(Type=Bone,Name=\"belt_fr_r\"),Transform=(Rotation=(X=0,Y=0,Z=0,W=-1),Translation=(X=8.0514,Y=0,Z=0), ...(truncated)"
    WeightMinimum=0
    Weight=@"Thigh Readers_1".Fr
   )
   > &RigUnit_ModifyTransforms(
    ItemToModify="((Item=(Type=Bone,Name=\"fold_02_fr_r\"),Transform=(Rotation=(X=0,Y=0,Z=-0.309,W=0.9511),Translation=(X=0,Y=0,Z=0),Scale3D=(X=1,Y=1,Z=1))),(Item=(Type=Bone,Name=\"fold_03_fr_r\"),Transform=(Rotation=(X=0,Y=0,Z=-0.309,W=0.9511),Translation=(X=0,Y=0,Z=0),Scale3D ...(truncated)"
    WeightMinimum=0
    Weight=@"Thigh Readers_1".Fr
   )
   > &RigUnit_ModifyTransforms(
    ItemToModify=((Item=(Type=Bone,Name="glute_r"),Transform=(Rotation=(X=0,Y=0,Z=-0.2908,W=0.9568),Translation=(X=-4.7598,Y=1.2462,Z=0),Scale3D=(X=1,Y=1,Z=1))))
    WeightMinimum=0
    Weight=@"Thigh Readers_1".Fr
   )
   > &RigUnit_ModifyTransforms(
    ItemToModify="((Item=(Type=Bone,Name=\"leg_crease_fr_r\"),Transform=(Rotation=(X=0,Y=0,Z=-0.5696,W=0.8219),Translation=(X=23.841,Y=5.7143,Z=0),Scale3D=(X=1,Y=1,Z=1))),(Item=(Type=Bone,Name=\"belt_fr_r\"),Transform=(Rotation=(X=0,Y=0,Z=0,W=-1),Translation=(X=8.0514,Y=0,Z=0), ...(truncated)"
    WeightMinimum=0
    Weight=@"Thigh Readers_1".Fr Post
   )
   > &RigUnit_ModifyTransforms(
    ItemToModify="((Item=(Type=Bone,Name=\"fold_02_fr_r\"),Transform=(Rotation=(X=0,Y=0,Z=-0.5,W=0.866),Translation=(X=0,Y=0,Z=0),Scale3D=(X=1,Y=1,Z=1))),(Item=(Type=Bone,Name=\"fold_02_diag_fr_r\"),Transform=(Rotation=(X=0,Y=0,Z=-0.4617,W=0.887),Translation=(X=0,Y=0,Z=0),Scale ...(truncated)"
    WeightMinimum=0
    Weight=@"Thigh Readers_1".Fr Post
   )
   > &RigUnit_ModifyTransforms(
    ItemToModify="((Item=(Type=Bone,Name=\"belt_ba_r\"),Transform=(Rotation=(X=0,Y=0,Z=0,W=-1),Translation=(X=7.3326,Y=0,Z=0),Scale3D=(X=1,Y=1,Z=1))),(Item=(Type=Bone,Name=\"glute_r\"),Transform=(Rotation=(X=0,Y=0,Z=0.4535,W=0.8912),Translation=(X=12.1561,Y=-4.1782,Z=0),Scale3D ...(truncated)"
    WeightMinimum=0
    Weight=@"Thigh Readers_1".Ba
   )
   > &RigUnit_ModifyTransforms(
    ItemToModify="((Item=(Type=Bone,Name=\"fold_02_leg_ba_r\"),Transform=(Rotation=(X=0,Y=0,Z=-0.309,W=0.9511),Translation=(X=0,Y=0,Z=0),Scale3D=(X=1,Y=1,Z=1))),(Item=(Type=Bone,Name=\"fold_03_leg_ba_r\"),Transform=(Rotation=(X=0,Y=0,Z=-0.309,W=0.9511),Translation=(X=0,Y=0,Z=0) ...(truncated)"
    WeightMinimum=0
    Weight=@"Thigh Readers_1".Ba
   )
   > &RigUnit_ModifyTransforms(
    ItemToModify="((Item=(Type=Bone,Name=\"belt_ba_r\"),Transform=(Rotation=(X=0,Y=0,Z=0,W=-1),Translation=(X=7.3326,Y=0,Z=0),Scale3D=(X=1,Y=1,Z=1))),(Item=(Type=Bone,Name=\"glute_r\"),Transform=(Rotation=(X=0,Y=0,Z=0.4535,W=0.8912),Translation=(X=22.5744,Y=-9.3688,Z=0.0002),Sc ...(truncated)"
    WeightMinimum=0
    Weight=@"Thigh Readers_1".Ba Post
   )
   > &RigUnit_ModifyTransforms(
    ItemToModify="((Item=(Type=Bone,Name=\"fold_01_leg_ba_r\"),Transform=(Rotation=(X=0,Y=0,Z=0.9954,W=-0.0955),Translation=(X=0,Y=0,Z=0),Scale3D=(X=1,Y=1,Z=1))),(Item=(Type=Bone,Name=\"fold_02_ba_ct\"),Transform=(Rotation=(X=0,Y=0,Z=0.5125,W=-0.8587),Translation=(X=0,Y=0,Z=0), ...(truncated)"
    WeightMinimum=0
    Weight=@"Thigh Readers_1".Ba Post
   )
   > &RigUnit_ModifyTransforms(
    ItemToModify="((Item=(Type=Bone,Name=\"belt_r\"),Transform=(Rotation=(X=0,Y=0,Z=0,W=-1),Translation=(X=7.914,Y=0,Z=0),Scale3D=(X=1,Y=1,Z=1))),(Item=(Type=Bone,Name=\"belt_fr_r\"),Transform=(Rotation=(X=0,Y=0.1456,Z=0,W=0.9893),Translation=(X=2.9141,Y=0,Z=0),Scale3D=(X=1,Y=1 ...(truncated)"
    WeightMinimum=0
    Weight=@"Thigh Readers_1".Ot
   )
   > &RigUnit_ModifyTransforms(
    ItemToModify="((Item=(Type=Bone,Name=\"fold_01_side_r\"),Transform=(Rotation=(X=0,Y=0,Z=-0.1305,W=0.9914),Translation=(X=0,Y=0,Z=0),Scale3D=(X=1,Y=1,Z=1))),(Item=(Type=Bone,Name=\"fold_02_side_r\"),Transform=(Rotation=(X=0,Y=0,Z=-0.3827,W=0.9239),Translation=(X=0,Y=0,Z=0),S ...(truncated)"
    WeightMinimum=0
    Weight=@"Thigh Readers_1".Ot
   )
   > &RigUnit_ModifyTransforms(
    ItemToModify="((Item=(Type=Bone,Name=\"belt_r\"),Transform=(Rotation=(X=0,Y=0,Z=0,W=-1),Translation=(X=22.1256,Y=0,Z=1.0348),Scale3D=(X=1,Y=1,Z=1))),(Item=(Type=Bone,Name=\"belt_fr_r\"),Transform=(Rotation=(X=0,Y=0.1456,Z=0,W=0.9893),Translation=(X=2.9141,Y=0,Z=0),Scale3D=( ...(truncated)"
    WeightMinimum=0
    Weight=@"Thigh Readers_1".Ot Post
   )
   > &RigUnit_ModifyTransforms(
    ItemToModify="((Item=(Type=Bone,Name=\"fold_01_side_r\"),Transform=(Rotation=(X=0,Y=0,Z=0.7841,W=-0.6206),Translation=(X=0,Y=0,Z=0),Scale3D=(X=1,Y=1,Z=1))),(Item=(Type=Bone,Name=\"fold_02_side_r\"),Transform=(Rotation=(X=0,Y=0,Z=-0.2284,W=0.9736),Translation=(X=0,Y=0,Z=0),S ...(truncated)"
    WeightMinimum=0
    Weight=@"Thigh Readers_1".Ot Post
   )
   &RigUnit_SetTransform(
    Item=(Type=Bone,Name="knee_l")
    Space=LocalSpace
    Weight=1
    bInitial=False
    Value=&RigVMFunction_MathTransformLerp(
     T=0.5
     A=&RigUnit_GetTransform(Item=(Type=Bone,Name="calf_l"), Space=LocalSpace, bInitial=true).Transform
     B=&RigUnit_GetTransform(Item=(Type=Bone,Name="calf_l"), Space=LocalSpace, bInitial=False).Transform
    ) {template="Interpolate::Execute(in A,in B,in T,out Result)"}.Result
    ExecutePin=@RigVMFunction_Sequence.K
   )
   > Elbow Readers_2 = RigVMFunctionReferenceNode(driver item=(Type=Bone,Name="calf_l"), viz bend=false)
   > &RigUnit_ModifyTransforms(
    ItemToModify=((Item=(Type=Bone,Name="knee_ba_l"),Transform=(Rotation=(X=0,Y=0,Z=-0.9492,W=-0.3145),Translation=(X=-0.5557,Y=-2.5933,Z=0),Scale3D=(X=0.3783,Y=1,Z=1))))
    WeightMinimum=0
    Weight=@"Elbow Readers_2".Bend
   )
   > &RigUnit_ModifyTransforms(
    ItemToModify=((Item=(Type=Bone,Name="knee_fr_bt_l"),Transform=(Rotation=(X=0,Y=0,Z=0.4592,W=0.8883),Translation=(X=1.6182,Y=2.0891,Z=0),Scale3D=(X=0.3783,Y=1,Z=1))))
    WeightMinimum=0
    Weight=@"Elbow Readers_2".Bend
   )
   > &RigUnit_ModifyTransforms(
    ItemToModify=((Item=(Type=Bone,Name="knee_fr_tp_l"),Transform=(Rotation=(X=0,Y=0,Z=0,W=1),Translation=(X=1.9506,Y=0.744,Z=0),Scale3D=(X=0.3783,Y=1,Z=1))))
    WeightMinimum=0
    Weight=@"Elbow Readers_2".Bend
   )
   > &RigUnit_SetTransform(
    Item=(Type=Bone,Name="knee_r")
    Space=LocalSpace
    Weight=1
    bInitial=False
    Value=&RigVMFunction_MathTransformLerp(
     T=0.667
     A=&RigUnit_GetTransform(Item=(Type=Bone,Name="calf_r"), Space=LocalSpace, bInitial=true).Transform
     B=&RigUnit_GetTransform(Item=(Type=Bone,Name="calf_r"), Space=LocalSpace, bInitial=False).Transform
    ) {template="Interpolate::Execute(in A,in B,in T,out Result)"}.Result
   )
   > Elbow Readers_3 = RigVMFunctionReferenceNode(driver item=(Type=Bone,Name="calf_r"), viz bend=false)
   > &RigUnit_ModifyTransforms(
    ItemToModify=((Item=(Type=Bone,Name="knee_ba_r"),Transform=(Rotation=(X=0,Y=0,Z=0.3793,W=0.9253),Translation=(X=-2.4377,Y=-0.863,Z=-0.0792),Scale3D=(X=0.3783,Y=1,Z=1))))
    WeightMinimum=0
    Weight=@"Elbow Readers_3".Bend
   )
   > &RigUnit_ModifyTransforms(
    ItemToModify=((Item=(Type=Bone,Name="knee_fr_bt_r"),Transform=(Rotation=(X=0,Y=0,Z=0.4698,W=0.8828),Translation=(X=0.7183,Y=-0.1273,Z=0),Scale3D=(X=0.3783,Y=1,Z=1))))
    WeightMinimum=0
    Weight=@"Elbow Readers_3".Bend
   )
   > &RigUnit_ModifyTransforms(
    ItemToModify=((Item=(Type=Bone,Name="knee_fr_tp_r"),Transform=(Rotation=(X=0,Y=0,Z=-0.2938,W=0.9559),Translation=(X=1.9504,Y=0.7439,Z=-0.0001),Scale3D=(X=0.3783,Y=1,Z=1))))
    WeightMinimum=0
    Weight=@"Elbow Readers_3".Bend
   )
   ItemArray_2 = RigVMDispatchNode(Value=((Type=Null,Name="null_A"),(Type=Null,Name="null_b"),(Type=Null,Name="null_c"),(Type=Null,Name="null_d"),(Type=Null,Name="null_e"))) {template="DISPATCH_RigVMDispatch_Constant(io Value)"}
   /Script/ControlRigSpline.RigUnit_DrawControlRigSpline(
    Color=(R=1,G=0,B=0,A=1)
    Detail=16
    ExecutePin=()
    Thickness=0.5
    Spline=/Script/ControlRigSpline.RigUnit_ControlRigSplineFromTransforms(
     Compression=0
     SamplesPerSegment=16
     SplineMode=BSpline
     Stretch=0
     bClosed=False
     Transforms=/Script/ControlRig.RigUnit_GetTransformItemArray(Space=GlobalSpace, bInitial=False, Items=@ItemArray_2.Value) {is=struct RigVMUnitNode}.Transforms
    ) {is=struct RigVMUnitNode}.Spline
   ) {is=struct RigVMUnitNode}
   RigVMRerouteNode(Value=())
   RigVMRerouteNode(Value=())
  ]}
 RigVMModel Rig {nodes [
   Get Delta Ref_5 = RigVMFunctionReferenceNode(ExecuteContext=(), Item=(Type=Bone,Name="thigh_l"))
   > &RigUnit_AimConstraintLocalSpaceOffset(
    Child=(Type=Bone,Name="fold_01_fr_l_MSTR")
    Filter=(bX=true,bY=true,bZ=true)
    Parents=((Item=(Type=Bone,Name="deform_knee_in_base_l"),Weight=1))
    UpAxis=(X=0,Y=-1,Z=0)
    Weight=1
    WorldUp=(Target=(X=0,Y=0,Z=1),Kind=Location,Space=(Type=Bone,Name="fold_01_fr_l_MSTR"))
    bMaintainOffset=true
   )
   > &RigUnit_RotationConstraintLocalSpaceOffset(
    AdvancedSettings=(InterpolationType=Average,RotationOrderForFilter=XZY)
    Child=(Type=Bone,Name="fold_01_side_l_MSTR")
    Filter=(bX=true,bY=true,bZ=true)
    Parents=((Item=(Type=Bone,Name="fold_01_diag_fr_l_MSTR"),Weight=0.8),(Item=(Type=Bone,Name="fold_01_fr_l_MSTR"),Weight=0.2))
    bMaintainOffset=false
    Weight=&RigVMFunction_MathDoubleRemap(SourceMaximum=20, SourceMinimum=0, TargetMaximum=1, TargetMinimum=0, Value=@"Get Delta Ref_5".Result.Rotation.Pitch).Result
   )
   &RigUnit_HierarchyAddNull(
    ExecutePin=()
    Name=null_A
    Parent=(Type=Bone,Name="fold_01_fr_l_MSTR")
    Space=LocalSpace
    Transform=(Rotation=(X=0,Y=0,Z=0,W=1),Translation=(X=0,Y=0,Z=0),Scale3D=(X=1,Y=1,Z=1))
   )
   > &RigUnit_HierarchyAddNull(
    Name=null_b
    Parent=(Type=Bone,Name="fold_01_fr_l_MSTR")
    Space=LocalSpace
    Transform=(Rotation=(X=0,Y=0,Z=0,W=1),Translation=(X=6,Y=0,Z=0),Scale3D=(X=1,Y=1,Z=1))
   )
   > &RigUnit_HierarchyAddNull(
    Name=null_c
    Parent=(Type=Bone,Name="fold_01_fr_l_MSTR")
    Space=LocalSpace
    Transform=(Rotation=(X=0,Y=0,Z=0,W=1),Translation=(X=17.5,Y=-3,Z=2.5),Scale3D=(X=1,Y=1,Z=1))
   )
   > &RigUnit_HierarchyAddNull(
    Name=null_d
    Parent=(Type=Bone,Name="thigh_twist_01_l")
    Space=LocalSpace
    Transform=(Rotation=(X=0,Y=0,Z=0,W=1),Translation=(X=0,Y=0,Z=-10),Scale3D=(X=1,Y=1,Z=1))
   )
   > &RigUnit_HierarchyAddNull(
    Name=null_e
    Parent=(Type=Bone,Name="thigh_twist_01_l")
    Space=LocalSpace
    Transform=(Rotation=(X=0,Y=0,Z=0,W=1),Translation=(X=-16,Y=0,Z=-10),Scale3D=(X=1,Y=1,Z=1))
   )
   n1 = &RigUnit_PrepareForExecution()
   FN_DeformRig_FNC = RigVMFunctionReferenceNode(
    DeformRig_ Curve="(deformRig_l_upperarm_up_alpha_72_#1=False,deformRig_r_upperarm_up_alpha_73_#2=False,deformRig_l_upperarm_fr_alpha_74_#3=False,deformRig_r_upperarm_fr_alpha_75_CB51B8 ...(truncated)"
    DeformRig_Bool="(upperarm_1_#4=True,elbow_3_#5=True,wrist_7_#6=True,groin_8_#7=True,glute_17_#8=True,knee_9_40C408324CA261 ...(truncated)"
    DeformRig_Debug=(upperarm_debug_18_#4=False,wrist_debug_20_#6=False,lat_debug_26_#9=False,groinGlute_debug_27_#7=False)
    DeformRig_Value="(upperarm_up_18_#10=(Rotation=(X=0,Y=0,Z=0,W=1),Translation=(X=0,Y=0,Z=8),Scale3D=(X=0.7,Y=1,Z=1)),upperarm_fwd_fr_19_#11=(Rotation=(X=0,Y=0,Z=0,W=1),Translation=(X=0,Y=10,Z=0),Scale3D=(X=0.5,Y=1,Z= ...(truncated)"
    ExecuteContext=@n1.ExecutePin
   )
   RigVMFunction_Sequence = &RigVMFunction_Sequence(ExecuteContext=@FN_DeformRig_FNC.ExecuteContext)
   &RigUnit_ModifyTransforms(
    ItemToModify=((Item=(Type=Bone,Name="fold_01_fr_l_MSTR"),Transform=(Rotation=(X=0,Y=0,Z=0.7071,W=0.7071),Translation=(X=0,Y=0,Z=0),Scale3D=(X=1,Y=1,Z=1))))
    Weight=1
    WeightMinimum=0
    ExecutePin=@RigVMFunction_Sequence.A
   )
   &RigUnit_SphericalPoseReader(
    ActiveRegionScaleFactors=(PositiveWidth=3.5,NegativeWidth=0.5,PositiveHeight=2,NegativeHeight=2)
    ActiveRegionSize=0.1
    Debug=(bDrawDebug=true,DebugScale=25,DebugSegments=20,DebugThickness=0.25)
    DriverAxis=(X=-1,Y=0,Z=0)
    DriverItem=(Type=Bone,Name="thigh_l")
    FalloffRegionScaleFactors=(PositiveWidth=1,NegativeWidth=1,PositiveHeight=1,NegativeHeight=1)
    FalloffSize=0.45
    OptionalParentItem=(Type=Bone)
    RotationOffset=(X=90,Y=0,Z=-30)
    ExecutePin=@RigVMFunction_Sequence.B
   )
   > &RigUnit_SphericalPoseReader(
    ActiveRegionScaleFactors=(PositiveWidth=3.5,NegativeWidth=0.5,PositiveHeight=2,NegativeHeight=2)
    ActiveRegionSize=0.1
    Debug=(bDrawDebug=true,DebugScale=25,DebugSegments=20,DebugThickness=0.25)
    DriverAxis=(X=1,Y=0,Z=0)
    DriverItem=(Type=Bone,Name="thigh_r")
    FalloffRegionScaleFactors=(PositiveWidth=1,NegativeWidth=1,PositiveHeight=1,NegativeHeight=1)
    FalloffSize=0.45
    FlipHeightScaling=false
    FlipWidthScaling=true
    OptionalParentItem=(Type=Bone)
    RotationOffset=(X=-90,Y=0,Z=-30)
   )
   RigVMFunction_ControlFlowBranch = &RigVMFunction_ControlFlowBranch(ExecuteContext=(), Condition=RigVMVariableNode() {variable=Debug Controls On}.Value)
   RigVMDispatchNode(ExecuteContext=@RigVMFunction_ControlFlowBranch.True, Array=RigVMVariableNode() {variable=Debug Ctrls}.Value) {template="DISPATCH_RigVMDispatch_ArrayReset(io Array)"}
   ItemArray_1 = RigVMDispatchNode(Value="((Type=Bone,Name=\"calf_l\"),(Type=Bone,Name=\"calf_r\"),(Type=Bone,Name=\"clavicle_l\"),(Type=Bone,Name=\"clavicle_r\"),(Type=Bone,Name=\"upperarm_l\"),(Type=Bone,Name=\"upperarm_r\"),(Type=Bone,Name=\"lowerarm_l\"),(Type=Bone,Name=\"lowerarm_r\"),(Type=Bone,Name=\"thigh_ ...(truncated)") {template="DISPATCH_RigVMDispatch_Constant(io Value)"}
   > RigVMVariableNode(Value=@ItemArray_1.Value) {variable=Debug Bones}
   > For_Each = RigVMDispatchNode(Array=@ItemArray_1.Value) {template="DISPATCH_RigVMDispatch_ArrayIterator(in Array,out Element,out Index,out Count,out Ratio)"}
   Concat = &RigVMFunction_NameConcat(B=" Ctrl", A=@For_Each.Element.Name) {template="Concat::Execute(in A,in B,out Result)"}
   > n2 = /Script/ControlRig.RigUnit_HierarchyAddControlTransform(
    InitialValue=(Rotation=(X=0,Y=0,Z=0,W=1),Translation=(X=0,Y=0,Z=0),Scale3D=(X=1,Y=1,Z=1))
    OffsetSpace=GlobalSpace
    Settings="(InitialSpace=LocalSpace,bUsePreferredRotationOrder=False,PreferredRotationOrder=YZX,Limits=(LimitTranslationX=(bMinimum=False,bMaximum=False),LimitTranslationY=(bMinimum=False,bMaximum=False),LimitTranslationZ=(bMinimum=False,bMaximum=False),LimitPitch=(b ...(truncated)"
    Name=@Concat.Result
    Parent=&RigUnit_HierarchyGetParent(bDefaultParent=True, Child=@For_Each.Element).Parent
    OffsetTransform=&RigUnit_GetTransform(bInitial=False, Item=@For_Each.Element).Transform
   ) {
    is=struct RigVMUnitNode
    template="SpawnControl::Execute(in InitialValue,in Settings,in OffsetTransform,in OffsetSpace,in Parent,in Name,out Item)"
   }
   > RigVMDispatchNode(Array=RigVMVariableNode() {variable=Debug Ctrls}.Value, Element=@n2.Item) {template="DISPATCH_RigVMDispatch_ArrayAdd(io Array,in Element,out Index)"}
   &RigUnit_ModifyTransforms(
    ExecutePin=()
    ItemToModify="((Item=(Type=Bone,Name=\"fold_01_fr_l_MSTR\"),Transform=(Rotation=(X=0.1677,Y=-0.2549,Z=0.0449,W=0.9513),Translation=(X=0,Y=0,Z=2),Scale3D=(X=1,Y=1,Z=1))),(Item=(Type=Bone,Name=\"deform_skirt_m_r\"),Transform=(Rotation=(X=0,Y=0,Z=-0.2588,W=0.9659),Translation= ...(truncated)"
    WeightMinimum=0
    Weight=&RigVMFunction_MathDoubleRemap(SourceMaximum=-90, SourceMinimum=0, TargetMaximum=1, TargetMinimum=0, Value=@"Get Delta Ref_5".Result.Rotation.Yaw).Result
   )
  ]}
]
hierarchy [
 root {loc=<0 0 0>, rot=<0 0 0 1>}
 pelvis {parent=root, loc=<0 -0.7071 93.1191>, rot=<0.0127 0.707 -0.0127 -0.707>}
 spine_01 {parent=pelvis, loc=<6.7494 -0.6388 0>, rot=<0 0 0.034 -0.9994>}
 spine_02 {parent=spine_01, loc=<7.2518 -0.035 0>, rot=<0 0 0.0238 -0.9997>}
 spine_03 {parent=spine_02, loc=<6.0104 0 0>, rot=<0 0 0.0149 -0.9999>}
 spine_04 {parent=spine_03, loc=<8.1024 0 0>, rot=<0 0 -0.0132 -0.9999>}
 spine_05 {parent=spine_04, loc=<10.0146 0 0>, rot=<0 0 -0.0015 -1>}
 neck_01 {parent=spine_05, loc=<11.3138 0.1529 0>, rot=<0 0 -0.2003 -0.9797>}
 neck_02 {parent=neck_01, loc=<4.5627 0 0>, rot=<0 0 0.0504 -0.9987>}
 head {parent=neck_02, loc=<5.3914 0 0>, rot=<0 0 0.115 -0.9934>}
 attach_cape {parent=spine_05, loc=<9.017 -11.714 0>, rot=<0.0616 -0.7044 -0.0616 -0.7044>}
 cape {parent=attach_cape, type=Null, loc=<0 0 0>, rot=<0 0 0 1>}
 attach_backpack {parent=spine_05, loc=<-0.123 -13.473 0>, rot=<-0.0093 -0.707 0.0093 -0.707>}
 Backpack_BR {parent=attach_backpack, type=Null, loc=<0 0 0>, rot=<0 0 0 1>}
 clavicle_l {parent=spine_05, loc=<4.3575 -0.1466 -2.3223>, rot=<0.0465 -0.7289 -0.009 -0.6829>}
 upperarm_l {parent=clavicle_l, loc=<15.5111 0 0>, rot=<0.0226 -0.3234 -0.0088 -0.9459>}
 lowerarm_l {parent=upperarm_l, loc=<26.6514 0 0>, rot=<0.0391 -0.0099 -0.2453 -0.9686>}
 hand_l {parent=lowerarm_l, loc=<22.3491 0 0>, rot=<-0.7081 -0.0507 0.0975 -0.6975>}
 thumb_01_l {parent=hand_l, loc=<1.3118 -0.3191 -2.459>, rot=<0.6546 -0.4042 0.0664 -0.6354>}
 thumb_02_l {parent=thumb_01_l, loc=<5.9336 -0.0448 0.0401>, rot=<-0.0037 0.0213 0.298 -0.9543>}
 thumb_03_l {parent=thumb_02_l, loc=<2.2223 0.3682 -0.0907>, rot=<0.0118 -0.013 0.0066 -0.9998>}
 index_metacarpal_l {parent=hand_l, loc=<1.6535 0.0183 -2.96>, rot=<0.0158 -0.0124 -0.0173 -0.9996>}
 index_01_l {parent=index_metacarpal_l, loc=<8.3008 0.0357 -0.0524>, rot=<-0.0044 0.0485 0.0924 -0.9945>}
 index_02_l {parent=index_01_l, loc=<3.3587 -0.3004 0.0452>, rot=<-0.002 0.0027 0.3225 -0.9466>}
 index_03_l {parent=index_02_l, loc=<2.5578 -0.123 -0.0252>, rot=<-0.002 0.0067 0.2558 -0.9667>}
 middle_metacarpal_l {parent=hand_l, loc=<2.1856 -0.1725 -0.5984>, rot=<0.006 0.0168 -0.0329 -0.9993>}
 middle_01_l {parent=middle_metacarpal_l, loc=<7.3275 -0.2447 0.0129>, rot=<-0.0208 0.0042 0.0858 -0.9961>}
 middle_02_l {parent=middle_01_l, loc=<4.0123 0.1401 0.1175>, rot=<-0.0218 0.0258 0.3512 -0.9357>}
 middle_03_l {parent=middle_02_l, loc=<2.6077 0.0534 0.1373>, rot=<0.0082 0.0188 0.1708 -0.9851>}
 ring_metacarpal_l {parent=hand_l, loc=<2.2074 -0.3625 1.3638>, rot=<-0.0284 0.0722 -0.0458 -0.9959>}
 ring_01_l {parent=ring_metacarpal_l, loc=<7.0295 -0.6197 -0.1446>, rot=<-0.0489 -0.0252 0.1892 -0.9804>}
 ring_02_l {parent=ring_01_l, loc=<3.6174 0.4647 -0.1027>, rot=<0 0 0.2806 -0.9598>}
 ring_03_l {parent=ring_02_l, loc=<2.4152 0.114 0.0199>, rot=<0 0 0.1189 -0.9929>}
 pinky_metacarpal_l {parent=hand_l, loc=<1.7616 -0.9753 3.3814>, rot=<-0.0397 0.0814 -0.069 -0.9935>}
 pinky_01_l {parent=pinky_metacarpal_l, loc=<7.1709 -0.4831 -0.1094>, rot=<-0.0654 -0.0035 0.2242 -0.9723>}
 pinky_02_l {parent=pinky_01_l, loc=<2.8274 0.0893 -0.1043>, rot=<0.0037 0.0011 0.2854 -0.9584>}
 pinky_03_l {parent=pinky_02_l, loc=<2.2046 0.0307 -0.0078>, rot=<-0.0002 0.0098 0.1445 -0.9895>}
 hand_attach_l {parent=hand_l, loc=<8.5 -3 0>, rot=<0.7071 0.7071 0 0>}
 deform_wrist_dn_base_l {parent=hand_l, loc=<0 -2.2342 0>, rot=<0 0 0 -1>}
 deform_wrist_dn_l {parent=deform_wrist_dn_base_l, loc=<0 0 0>, rot=<0 0 0 1>}
 deform_wrist_up_base_l {parent=hand_l, loc=<0 2.2682 0>, rot=<0 0 0 -1>}
 deform_wrist_up_l {parent=deform_wrist_up_base_l, loc=<0 0 0>, rot=<0 0 0 1>}
 lowerarm_twist_02_l {parent=lowerarm_l, loc=<9.6199 0 0>, rot=<0 0 0 -1>}
 sleeve_l {parent=lowerarm_twist_02_l, loc=<0 0 0>, rot=<0 0 0 1>}
 sleeve_l_end {parent=sleeve_l, loc=<13.7271 0 0>, rot=<0 0 0 1>}
 deform_sleeve_l {parent=lowerarm_twist_02_l, loc=<0 0 0>, rot=<0 0 0 1>}
 dyn_sleeve_l {parent=deform_sleeve_l, loc=<13.7271 0 0>, rot=<0 0 0 1>}
 lowerarm_twist_01_l {parent=lowerarm_l, loc=<17.8549 0 0>, rot=<0 0 0 1>}
 deform_elbow_in_base_l {parent=lowerarm_l, loc=<0.2112 5.259 0.8274>, rot=<0 0 0 -1>}
 deform_elbow_in_l {parent=deform_elbow_in_base_l, loc=<0 0 0>, rot=<0 0 0 1>}
 deform_elbow_out_base_l {parent=lowerarm_l, loc=<0 -4 0>, rot=<0 0 0 -1>}
 deform_elbow_out_l {parent=deform_elbow_out_base_l, loc=<0 0 0>, rot=<0 0 0 1>}
 upperarm_twist_01_l {parent=upperarm_l, loc=<0.0041 0 0>, rot=<0 0 0 1>}
 deform_upperarm_bk_base_l {parent=upperarm_twist_01_l, loc=<1.0813 -7.2193 -1.9453>, rot=<0 0 0 1>}
 deform_upperarm_bk_l {parent=deform_upperarm_bk_base_l, loc=<0 0 0>, rot=<0 0 0 1>}
 deform_upperarm_fr_base_l {parent=upperarm_twist_01_l, loc=<-0.3691 7.0506 -2.765>, rot=<0 0 0 1>}
 deform_upperarm_fr_l {parent=deform_upperarm_fr_base_l, loc=<0 0 0>, rot=<0 0 0 1>}
 deform_upperarm_up_base_l {parent=upperarm_twist_01_l, loc=<-3.2216 -1.1391 3.3311>, rot=<0 0 0 1>}
 deform_upperarm_up_l {parent=deform_upperarm_up_base_l, loc=<0 0 0>, rot=<0 0 0 1>}
 upperarm_twist_02_l {parent=upperarm_l, loc=<13.7941 0 0>, rot=<0 0 0 1>}
 tricep_l {parent=upperarm_twist_02_l, loc=<0 -8.019 0>, rot=<0 0 0 1>}
 bicep_l {parent=upperarm_twist_02_l, loc=<0 6.7438 0>, rot=<0 0 0 1>}
 elbow_l {parent=upperarm_l, loc=<26.6514 0 0>, rot=<0.0391 -0.0099 -0.2453 -0.9686>}
 elbow_fr_l {parent=elbow_l, loc=<0.3197 7.3801 0>, rot=<0 0 0 1>}
 elbow_ba_l {parent=elbow_l, loc=<-1.524 -4.717 0>, rot=<0 0 0 1>}
 def_arm_tp_l {parent=clavicle_l, loc=<17.8594 -1.0553 6.8356>, rot=<0 0 0 1>}
 def_arm_tp_l_end {parent=def_arm_tp_l, loc=<0 0 0>, rot=<0 0 0 1>}
 def_arm_fr_l {parent=clavicle_l, loc=<16.8097 7.219 -2.9938>, rot=<0 0 0 1>}
 def_arm_fr_l_end {parent=def_arm_fr_l, loc=<0 0 0>, rot=<0 0 0 1>}
 def_arm_ba_l {parent=clavicle_l, loc=<17.803 -9.1305 -2.9297>, rot=<0 0 0 1>}
 def_arm_ba_l_end {parent=def_arm_ba_l, loc=<0 0 0>, rot=<0 0 0 1>}
 clavicle_r {parent=spine_05, loc=<4.3575 -0.1466 2.3223>, rot=<0.7289 0.0465 -0.6829 0.009>}
 upperarm_r {parent=clavicle_r, loc=<-15.5111 0 0>, rot=<0.0226 -0.3234 -0.0088 -0.9459>}
 lowerarm_r {parent=upperarm_r, loc=<-26.6514 0 0>, rot=<0.0391 -0.0099 -0.2453 -0.9686>}
 hand_r {parent=lowerarm_r, loc=<-22.3491 0 0>, rot=<-0.7081 -0.0507 0.0975 -0.6975>}
 thumb_01_r {parent=hand_r, loc=<-1.3118 0.3191 2.459>, rot=<0.6546 -0.4042 0.0664 -0.6354>}
 thumb_02_r {parent=thumb_01_r, loc=<-5.9336 0.0448 -0.0401>, rot=<-0.0037 0.0213 0.298 -0.9543>}
 thumb_03_r {parent=thumb_02_r, loc=<-2.2223 -0.3682 0.0907>, rot=<0.0118 -0.013 0.0066 -0.9998>}
 index_metacarpal_r {parent=hand_r, loc=<-1.6535 -0.0183 2.96>, rot=<0.0158 -0.0124 -0.0173 -0.9996>}
 index_01_r {parent=index_metacarpal_r, loc=<-8.3008 -0.0357 0.0524>, rot=<-0.0044 0.0485 0.0924 -0.9945>}
 index_02_r {parent=index_01_r, loc=<-3.3587 0.3004 -0.0452>, rot=<-0.002 0.0027 0.3225 -0.9466>}
 index_03_r {parent=index_02_r, loc=<-2.5578 0.123 0.0252>, rot=<-0.002 0.0067 0.2558 -0.9667>}
 middle_metacarpal_r {parent=hand_r, loc=<-2.1856 0.1725 0.5984>, rot=<0.006 0.0168 -0.0329 -0.9993>}
 middle_01_r {parent=middle_metacarpal_r, loc=<-7.3275 0.2447 -0.0129>, rot=<-0.0208 0.0042 0.0858 -0.9961>}
 middle_02_r {parent=middle_01_r, loc=<-4.0123 -0.1401 -0.1175>, rot=<-0.0218 0.0258 0.3512 -0.9357>}
 middle_03_r {parent=middle_02_r, loc=<-2.6077 -0.0534 -0.1373>, rot=<0.0082 0.0188 0.1708 -0.9851>}
 ring_metacarpal_r {parent=hand_r, loc=<-2.2074 0.3625 -1.3638>, rot=<-0.0284 0.0722 -0.0458 -0.9959>}
 ring_01_r {parent=ring_metacarpal_r, loc=<-7.0295 0.6197 0.1446>, rot=<-0.0489 -0.0252 0.1892 -0.9804>}
 ring_02_r {parent=ring_01_r, loc=<-3.6174 -0.4647 0.1027>, rot=<0 0 0.2806 -0.9598>}
 ring_03_r {parent=ring_02_r, loc=<-2.4152 -0.114 -0.0199>, rot=<0 0 0.1189 -0.9929>}
 pinky_metacarpal_r {parent=hand_r, loc=<-1.7616 0.9753 -3.3814>, rot=<-0.0397 0.0814 -0.069 -0.9935>}
 pinky_01_r {parent=pinky_metacarpal_r, loc=<-7.1709 0.4831 0.1094>, rot=<-0.0654 -0.0035 0.2242 -0.9723>}
 pinky_02_r {parent=pinky_01_r, loc=<-2.8274 -0.0893 0.1043>, rot=<0.0037 0.0011 0.2854 -0.9584>}
 pinky_03_r {parent=pinky_02_r, loc=<-2.2046 -0.0307 0.0078>, rot=<-0.0002 0.0098 0.1445 -0.9895>}
 hand_attach_r {parent=hand_r, loc=<-8.5001 2.9999 0>, rot=<0 0 -0.7071 -0.7071>}
 deform_wrist_dn_base_r {parent=hand_r, loc=<-0.0004 2.2345 -0.0001>, rot=<0 0 0 -1>}
 deform_wrist_dn_r {parent=deform_wrist_dn_base_r, loc=<0 0 0>, rot=<0 0 0 1>}
 deform_wrist_up_base_r {parent=hand_r, loc=<0.0003 -2.2684 0.0001>, rot=<0 0 0 -1>}
 deform_wrist_up_r {parent=deform_wrist_up_base_r, loc=<0 0 0>, rot=<0 0 0 1>}
 lowerarm_twist_02_r {parent=lowerarm_r, loc=<-9.6199 0 0>, rot=<0 0 0 1>}
 sleeve_r {parent=lowerarm_twist_02_r, loc=<0 0 0>, rot=<0 0 0 1>}
 sleeve_r_end {parent=sleeve_r, loc=<-13.727 0 0>, rot=<0 0 0 1>}
 deform_sleeve_r {parent=lowerarm_twist_02_r, loc=<0 0 0>, rot=<0 0 0 1>}
 dyn_sleeve_r {parent=deform_sleeve_r, loc=<-13.727 0 0>, rot=<0 0 0 1>}
 lowerarm_twist_01_r {parent=lowerarm_r, loc=<-17.8549 0 0>, rot=<0 0 0 1>}
 deform_elbow_in_base_r {parent=lowerarm_r, loc=<-0.2111 -5.2591 -0.8275>, rot=<0 0 0 -1>}
 deform_elbow_in_r {parent=deform_elbow_in_base_r, loc=<-0.0001 0 0.0002>, rot=<0 0 0 1>}
 deform_elbow_out_base_r {parent=lowerarm_r, loc=<0 4 -0.0001>, rot=<0 0 0 -1>}
 deform_elbow_out_r {parent=deform_elbow_out_base_r, loc=<0 0 0>, rot=<0 0 0 1>}
 upperarm_twist_01_r {parent=upperarm_r, loc=<-0.0041 0 0>, rot=<0 0 0 1>}
 deform_upperarm_bk_base_r {parent=upperarm_twist_01_r, loc=<-1.0814 7.2192 1.9455>, rot=<0 0 0 1>}
 deform_upperarm_bk_r {parent=deform_upperarm_bk_base_r, loc=<0 0 0>, rot=<0 0 0 1>}
 deform_upperarm_fr_base_r {parent=upperarm_twist_01_r, loc=<0.3689 -7.0506 2.7652>, rot=<0 0 0 1>}
 deform_upperarm_fr_r {parent=deform_upperarm_fr_base_r, loc=<0 0 0>, rot=<0 0 0 1>}
 deform_upperarm_up_base_r {parent=upperarm_twist_01_r, loc=<3.2214 1.1391 -3.3308>, rot=<0 0 0 1>}
 deform_upperarm_up_r {parent=deform_upperarm_up_base_r, loc=<0 0 0>, rot=<0 0 0 -1>}
 upperarm_twist_02_r {parent=upperarm_r, loc=<-13.7941 0 0>, rot=<0 0 0 1>}
 bicep_r {parent=upperarm_twist_02_r, loc=<0 -6.744 0>, rot=<0 0 0 1>}
 tricep_r {parent=upperarm_twist_02_r, loc=<0 8.019 0>, rot=<0 0 0 1>}
 elbow_r {parent=upperarm_r, loc=<-26.6514 0 0>, rot=<0.0391 -0.0099 -0.2453 -0.9686>}
 elbow_fr_r {parent=elbow_r, loc=<-0.32 -7.38 0>, rot=<0 0 0 1>}
 elbow_ba_r {parent=elbow_r, loc=<1.524 4.717 0>, rot=<0 0 0 1>}
 def_arm_tp_r {parent=clavicle_r, loc=<-17.859 1.055 -6.836>, rot=<0 0 0 1>}
 def_arm_tp_r_end {parent=def_arm_tp_r, loc=<0 0 0>, rot=<0 0 0 1>}
 def_arm_fr_r {parent=clavicle_r, loc=<-16.81 -7.219 2.994>, rot=<0 0 0 1>}
 def_arm_fr_r_end {parent=def_arm_fr_r, loc=<0 0 0>, rot=<0 0 0 1>}
 def_arm_ba_r {parent=clavicle_r, loc=<-17.803 9.131 2.93>, rot=<0 0 0 1>}
 def_arm_ba_r_end {parent=def_arm_ba_r, loc=<0 0 0>, rot=<0 0 0 1>}
 collar_l {parent=spine_05, loc=<13.0069 2.1738 -7.5576>, rot=<0 0 -0.2003 -0.9797>}
 collar_l_end {parent=collar_l, loc=<0 0 0>, rot=<0 0 0 1>}
 flap_fr_l_MSTR {parent=collar_l, loc=<0.8856 7.2152 -1.3091>, rot=<0.0443 -0.7514 0.0427 -0.657>}
 flap_fr_l {parent=flap_fr_l_MSTR, loc=<0 0 0>, rot=<0 0 0 1>}
 flap_fr_l_end {parent=flap_fr_l, loc=<2.562 0 0>, rot=<0 0.2228 0 -0.9749>}
 flap_l_MSTR {parent=collar_l, loc=<1.0871 -0.7137 -0.3885>, rot=<0 -0.8235 0 -0.5673>}
 flap_l {parent=flap_l_MSTR, loc=<0 0 0>, rot=<0 0 0 1>}
 collar_r {parent=spine_05, loc=<13.0069 2.1738 7.558>, rot=<0 0 -0.2003 -0.9797>}
 collar_r_end {parent=collar_r, loc=<0 0 0>, rot=<0 0 0 1>}
 flap_fr_r_MSTR {parent=collar_r, loc=<0.9054 6.8336 1.4519>, rot=<0.0747 -0.9056 -0.0373 0.4157>}
 flap_fr_r {parent=flap_fr_r_MSTR, loc=<0 0 0>, rot=<0 0 0 1>}
 flap_fr_r_end {parent=flap_fr_r, loc=<3.9967 0 0>, rot=<0 -0.0857 0 -0.9963>}
 flap_r_MSTR {parent=collar_r, loc=<1.6664 -0.9642 -0.0424>, rot=<0 0.8372 0 -0.547>}
 flap_r {parent=flap_r_MSTR, loc=<0 0 0>, rot=<0 0 0 1>}
 collar_ba {parent=spine_05, loc=<16.0415 -5.1253 0>, rot=<0 0 -0.2003 -0.9797>}
 def_arm_bt_l {parent=spine_05, loc=<-7.569 -2.4642 -18.1899>, rot=<0.0465 -0.7289 -0.009 -0.6829>}
 def_arm_bt_l_end {parent=def_arm_bt_l, loc=<0 0 0>, rot=<0 0 0 1>}
 def_arm_bt_r {parent=spine_05, loc=<-7.5691 -2.4638 18.1901>, rot=<0.7289 0.0465 -0.6829 0.009>}
 def_arm_bt_r_end {parent=def_arm_bt_r, loc=<0 0 0>, rot=<0 0 0 1>}
 lapel_r {parent=spine_05, loc=<1.4269 13.8503 6.61>, rot=<0 0 -0.0401 -0.9992>}
 lapel_l {parent=spine_05, loc=<1.4269 13.8503 -6.61>, rot=<0 0 -0.0401 -0.9992>}
 deform_lapel_l {parent=spine_05, loc=<3.4312 9.6078 -9.8701>, rot=<0.6439 0.0445 0.682 0.344>}
 deform_collar_l {parent=spine_05, loc=<13.1929 4.3403 -7.7495>, rot=<-0.648 -0.2772 -0.1025 -0.702>}
 dyn_collar_l {parent=deform_collar_l, loc=<0 0 0>, rot=<0 0 0 1>}
 dyn_collar_1_l {parent=dyn_collar_l, loc=<0.4733 -3.1657 -3.8072>, rot=<0.0461 0.2215 -0.2622 -0.9381>}
 deform_collar_r {parent=spine_05, loc=<13.1926 4.3403 7.7495>, rot=<-0.2772 0.648 0.702 -0.1025>}
 dyn_collar_r {parent=deform_collar_r, loc=<0 0 0>, rot=<0 0 0 1>}
 deform_lapel_r {parent=spine_05, loc=<3.4841 9.7278 10.028>, rot=<0.6678 0.2095 -0.5976 -0.3911>}
 jacket {parent=spine_04, loc=<10.0146 0 0>, rot=<0 0 -0.0015 -1>}
 deform_pec_aim_base_l {parent=spine_04, loc=<7.8732 12.6388 -2.4123>, rot=<0.0102 -0.4602 0.3979 -0.7936>}
 deform_pec_aim_l {parent=deform_pec_aim_base_l, loc=<0 0 0>, rot=<0 0 0 1>}
 deform_pec_base_l {parent=deform_pec_aim_l, loc=<10.766 0 0>, rot=<0 0 0 1>}
 deform_pec_l {parent=deform_pec_base_l, loc=<0.055 0 0>, rot=<0 0 0 1>}
 deform_pec_aim_base_r {parent=spine_04, loc=<7.8727 12.6388 2.4123>, rot=<-0.4602 -0.0102 0.7936 0.3979>}
 deform_pec_aim_r {parent=deform_pec_aim_base_r, loc=<0 0 0>, rot=<0 0 0 1>}
 deform_pec_base_r {parent=deform_pec_aim_r, loc=<-10.7662 -0.0006 -0.0007>, rot=<0 0 0 1>}
 deform_pec_r {parent=deform_pec_base_r, loc=<0 0 0>, rot=<0 0 0 1>}
 fold_01_fr_l_MSTR {parent=spine_03, loc=<-1.0505 12.5735 -7.7969>, rot=<0 0 0.9985 -0.0547>}
 fold_01_fr_l {parent=fold_01_fr_l_MSTR, loc=<0 0 0>, rot=<0 0 0 1>}
 fold_02_fr_l {parent=fold_01_fr_l, loc=<6.0078 0 0>, rot=<0 0 0 1>}
 fold_03_fr_l {parent=fold_02_fr_l, loc=<7.2452 0 0>, rot=<0 0 0 1>}
 fold_04_fr_l {parent=fold_03_fr_l, loc=<6.7772 0 0>, rot=<0 0 0 1>}
 fold_01_fr_r_MSTR {parent=spine_03, loc=<-1.0505 12.5735 7.797>, rot=<0 0 0.9985 -0.0547>}
 fold_01_fr_r {parent=fold_01_fr_r_MSTR, loc=<0 0 0>, rot=<0 0 0 1>}
 fold_02_fr_r {parent=fold_01_fr_r, loc=<6.0078 0 0>, rot=<0 0 0 1>}
 fold_03_fr_r {parent=fold_02_fr_r, loc=<7.2452 0 0>, rot=<0 0 0 1>}
 fold_04_fr_r {parent=fold_03_fr_r, loc=<6.7772 0 0>, rot=<0 0 0 1>}
 fold_01_side_l_MSTR {parent=spine_03, loc=<-0.0211 0.7385 -17.1498>, rot=<-0.0387 0.706 0.706 -0.0387>}
 fold_01_side_l {parent=fold_01_side_l_MSTR, loc=<0 0 0>, rot=<0 0 0 1>}
 fold_02_side_l {parent=fold_01_side_l, loc=<6.0078 0 0>, rot=<0 0 0 1>}
 fold_03_side_l {parent=fold_02_side_l, loc=<7.2452 0 0>, rot=<0 0 0 1>}
 fold_04_side_l {parent=fold_03_side_l, loc=<6.7772 0 0>, rot=<0 0 0 1>}
 fold_01_diag_fr_l_MSTR {parent=spine_03, loc=<-0.6589 9.0118 -13.1444>, rot=<-0.0176 0.321 0.9455 -0.0518>}
 fold_01_diag_fr_l {parent=fold_01_diag_fr_l_MSTR, loc=<0 0 0>, rot=<0 0 0 1>}
 fold_02_diag_fr_l {parent=fold_01_diag_fr_l, loc=<6.0078 0 0>, rot=<0 0 0 1>}
 fold_03_diag_fr_l {parent=fold_02_diag_fr_l, loc=<7.2452 0 0>, rot=<0 0 0 1>}
 fold_04_diag_fr_l {parent=fold_03_diag_fr_l, loc=<6.7772 0 0>, rot=<0 0 0 1>}
 fold_01_ba_ct_MSTR {parent=spine_03, loc=<1.2158 -14.0855 0>, rot=<0.0547 -0.9985 0 0>}
 fold_01_ba_ct {parent=fold_01_ba_ct_MSTR, loc=<0 0 0>, rot=<0 0 0 1>}
 fold_02_ba_ct {parent=fold_01_ba_ct, loc=<6.0078 0 0>, rot=<0 0 0 1>}
 fold_03_ba_ct {parent=fold_02_ba_ct, loc=<7.2452 0 0>, rot=<0 0 0 1>}
 fold_04_ba_ct {parent=fold_03_ba_ct, loc=<6.7772 0 0>, rot=<0 0 0 1>}
 fold_01_leg_ba_l_MSTR {parent=spine_03, loc=<1.2121 -13.2719 -9.0047>, rot=<0.0547 -0.9985 0 0>}
 fold_01_leg_ba_l {parent=fold_01_leg_ba_l_MSTR, loc=<0 0 0>, rot=<0 0 0 1>}
 fold_02_leg_ba_l {parent=fold_01_leg_ba_l, loc=<6.0078 0 0>, rot=<0 0 0 1>}
 fold_03_leg_ba_l {parent=fold_02_leg_ba_l, loc=<7.2452 0 0>, rot=<0 0 0 1>}
 fold_04_leg_ba_l {parent=fold_03_leg_ba_l, loc=<6.7772 0 0>, rot=<0 0 0 1>}
 fold_01_leg_ba_r_MSTR {parent=spine_03, loc=<1.2121 -13.2719 9.005>, rot=<0.0547 -0.9985 0 0>}
 fold_01_leg_ba_r {parent=fold_01_leg_ba_r_MSTR, loc=<0 0 0>, rot=<0 0 0 1>}
 fold_02_leg_ba_r {parent=fold_01_leg_ba_r, loc=<6.0078 0 0>, rot=<0 0 0 1>}
 fold_03_leg_ba_r {parent=fold_02_leg_ba_r, loc=<7.2452 0 0>, rot=<0 0 0 1>}
 fold_04_leg_ba_r {parent=fold_03_leg_ba_r, loc=<6.7772 0 0>, rot=<0 0 0 1>}
 fold_01_diag_ba_l_MSTR {parent=spine_03, loc=<0.5121 -7.4862 -16.38>, rot=<0.0491 -0.8955 -0.4416 0.0242>}
 fold_01_diag_ba_l {parent=fold_01_diag_ba_l_MSTR, loc=<0 0 0>, rot=<0 0 0 1>}
 fold_02_diag_ba_l {parent=fold_01_diag_ba_l, loc=<6.0078 0 0>, rot=<0 0 0 1>}
 fold_03_diag_ba_l {parent=fold_02_diag_ba_l, loc=<7.2452 0 0>, rot=<0 0 0 1>}
 fold_04_diag_ba_l {parent=fold_03_diag_ba_l, loc=<6.7772 0 0>, rot=<0 0 0 1>}
 fold_01_diag_ba_r_MSTR {parent=spine_03, loc=<0.512 -7.486 16.38>, rot=<0.0491 -0.8955 0.4416 -0.0242>}
 fold_01_diag_ba_r {parent=fold_01_diag_ba_r_MSTR, loc=<0 0 0>, rot=<0 0 0 1>}
 fold_02_diag_ba_r {parent=fold_01_diag_ba_r, loc=<6.0078 0 0>, rot=<0 0 0 1>}
 fold_03_diag_ba_r {parent=fold_02_diag_ba_r, loc=<7.2452 0 0>, rot=<0 0 0 1>}
 fold_04_diag_ba_r {parent=fold_03_diag_ba_r, loc=<6.7772 0 0>, rot=<0 0 0 1>}
 fold_01_side_r_MSTR {parent=spine_03, loc=<-0.0072 0.9023 17.2844>, rot=<0.0387 -0.706 0.706 -0.0387>}
 fold_01_side_r {parent=fold_01_side_r_MSTR, loc=<0 0 0>, rot=<0 0 0 1>}
 fold_02_side_r {parent=fold_01_side_r, loc=<6.0078 0 0>, rot=<0 0 0 1>}
 fold_03_side_r {parent=fold_02_side_r, loc=<7.2452 0 0>, rot=<0 0 0 1>}
 fold_04_side_r {parent=fold_03_side_r, loc=<6.7772 0 0>, rot=<0 0 0 1>}
 fold_01_diag_fr_r_MSTR {parent=spine_03, loc=<-0.6589 9.0118 13.144>, rot=<0.0242 -0.4416 0.8955 -0.0491>}
 fold_01_diag_fr_r {parent=fold_01_diag_fr_r_MSTR, loc=<0 0 0>, rot=<0 0 0 1>}
 fold_02_diag_fr_r {parent=fold_01_diag_fr_r, loc=<6.0078 0 0>, rot=<0 0 0 1>}
 fold_03_diag_fr_r {parent=fold_02_diag_fr_r, loc=<7.2452 0 0>, rot=<0 0 0 1>}
 fold_04_diag_fr_r {parent=fold_03_diag_fr_r, loc=<6.7772 0 0>, rot=<0 0 0 1>}
 thigh_l {parent=pelvis, loc=<-2.5693 -0.0109 -7.7501>, rot=<0.0012 0.0572 0.0155 -0.9982>}
 calf_l {parent=thigh_l, loc=<-39.0484 0 0>, rot=<0.0006 -0.0161 -0.0391 -0.9991>}
 foot_l {parent=calf_l, loc=<-42.561 0.1077 0.4292>, rot=<0.0418 -0.0455 0.0343 -0.9975>}
 ball_l {parent=foot_l, loc=<-6.1599 13.8264 0.6806>, rot=<0.0001 -0.0001 -0.7063 -0.7079>}
 calf_twist_02_l {parent=calf_l, loc=<-18.8064 -0.0967 0.0397>, rot=<0.0237 0.0095 -0.0035 -0.9997>}
 calf_twist_01_l {parent=calf_l, loc=<-35.813 -0.0967 0.0397>, rot=<0.0237 0.0095 -0.0035 -0.9997>}
 hem_l {parent=calf_twist_01_l, loc=<-6 0 0>, rot=<0 0 0 1>}
 deform_knee_in_base_l {parent=calf_l, loc=<1.298 -5.2957 -0.0496>, rot=<0 0 0 -1>}
 deform_knee_in_l {parent=deform_knee_in_base_l, loc=<0 0 0>, rot=<0 0 0 1>}
 deform_knee_out_base_l {parent=calf_l, loc=<0 4 0>, rot=<0 0 0 -1>}
 deform_knee_out_l {parent=deform_knee_out_base_l, loc=<0 0 0>, rot=<0 0 0 1>}
 thigh_twist_01_l {parent=thigh_l, loc=<-21.6303 0 0>, rot=<0.0013 0 0 -1>}
 knee_l {parent=thigh_l, loc=<-39.0484 0 0>, rot=<0.0006 -0.0161 -0.0391 -0.9991>}
 knee_ba_l {parent=knee_l, loc=<-0.75 -7.5 0>, rot=<0 0 0 1>}
 knee_fr_tp_l {parent=knee_l, loc=<1.1 6.25 0>, rot=<0 0 0 1>}
 knee_fr_bt_l {parent=knee_l, loc=<-3.8 5.6 0>, rot=<0 0 0 1>}
 deform_groin_base_l {parent=thigh_l, loc=<-2.1908 8.1142 -1.7442>, rot=<0 0 0 -1>}
 deform_groin_l {parent=deform_groin_base_l, loc=<0 0 0>, rot=<0 0 0 1>}
 thigh_r {parent=pelvis, loc=<-2.5693 -0.0109 7.7501>, rot=<0.0572 -0.0012 0.9982 0.0155>}
 calf_r {parent=thigh_r, loc=<39.0484 0 0>, rot=<0.0006 -0.0161 -0.0391 -0.9991>}
 foot_r {parent=calf_r, loc=<42.561 -0.1077 -0.4292>, rot=<0.0418 -0.0455 0.0343 -0.9975>}
 ball_r {parent=foot_r, loc=<6.1599 -13.8264 -0.6806>, rot=<0.0001 -0.0001 -0.7063 -0.7079>}
 calf_twist_02_r {parent=calf_r, loc=<18.8064 0.0967 -0.0397>, rot=<0.0237 0.0095 -0.0035 -0.9997>}
 calf_twist_01_r {parent=calf_r, loc=<35.813 0.0967 -0.0397>, rot=<0.0237 0.0095 -0.0035 -0.9997>}
 hem_r {parent=calf_twist_01_r, loc=<6 0 0>, rot=<0 0 0 1>}
 deform_knee_in_base_r {parent=calf_r, loc=<-1.298 5.2957 0.0496>, rot=<0 0 0 -1>}
 deform_knee_in_r {parent=deform_knee_in_base_r, loc=<0 0 0>, rot=<0 0 0 1>}
 deform_knee_out_base_r {parent=calf_r, loc=<0 -4 0>, rot=<0 0 0 -1>}
 deform_knee_out_r {parent=deform_knee_out_base_r, loc=<0 0 0>, rot=<0 0 0 1>}
 thigh_twist_01_r {parent=thigh_r, loc=<21.6303 0 0>, rot=<0.0013 0 0 -1>}
 knee_r {parent=thigh_r, loc=<39.0484 0 0>, rot=<0.0006 -0.0161 -0.0391 -0.9991>}
 knee_ba_r {parent=knee_r, loc=<0.75 7.5 0>, rot=<0 0 0 1>}
 knee_fr_tp_r {parent=knee_r, loc=<-1.1 -6.25 0>, rot=<0 0 0 1>}
 knee_fr_bt_r {parent=knee_r, loc=<3.8 -5.6 0>, rot=<0 0 0 1>}
 deform_groin_base_r {parent=thigh_r, loc=<2.1908 -8.1142 1.7442>, rot=<0 0 0 -1>}
 deform_groin_r {parent=deform_groin_base_r, loc=<0 0 0>, rot=<0 0 0 1>}
 belt {parent=pelvis, loc=<0 0 0>, rot=<0 0 0 1>}
 belt_fr_l {parent=belt, loc=<-2.6206 9.1871 -8.9295>, rot=<0 0 0 1>}
 belt_fr_r {parent=belt, loc=<-2.6206 9.1871 8.93>, rot=<0 0 0 1>}
 belt_l {parent=belt, loc=<-1.5433 0.8525 -16.1077>, rot=<0 0 0 1>}
 belt_r {parent=belt, loc=<-1.5433 0.8525 16.108>, rot=<0 0 0 1>}
 belt_fr_ct {parent=belt, loc=<-2.7259 10.9431 0>, rot=<0 0 0 1>}
 belt_ba_l {parent=belt, loc=<-1.5565 -11.9027 -8.389>, rot=<0 0 0 1>}
 belt_ba_r {parent=belt, loc=<-1.5565 -11.9027 8.389>, rot=<0 0 0 1>}
 groin {parent=pelvis, loc=<-13.8038 4.0879 0>, rot=<0 0 0 1>}
 leg_crease_fr_l {parent=pelvis, loc=<-12.2219 8.8627 -9.441>, rot=<0 0 0 1>}
 leg_crease_fr_r {parent=pelvis, loc=<-12.2219 8.8627 9.441>, rot=<0 0 0 1>}
 glute_l {parent=pelvis, loc=<-7.8694 -11.8543 -8.2823>, rot=<0 0 0 1>}
 glute_r {parent=pelvis, loc=<-7.8694 -11.8543 8.282>, rot=<0 0 0 1>}
 leg_crease_l {parent=pelvis, loc=<-9.7842 0.7492 -16.9203>, rot=<0 0 0 1>}
 leg_crease_r {parent=pelvis, loc=<-9.7842 0.7492 16.92>, rot=<0 0 0 1>}
 deform_glute_base_l {parent=pelvis, loc=<-8.8064 -8.8222 -8.4287>, rot=<0.0275 0.0501 -0.4472 -0.8926>}
 deform_glute_l {parent=deform_glute_base_l, loc=<0 0 0>, rot=<0 0 0 1>}
 deform_glute_base_r {parent=pelvis, loc=<-8.8062 -8.8222 8.4287>, rot=<0.0501 -0.0275 0.8926 -0.4472>}
 deform_glute_r {parent=deform_glute_base_r, loc=<0 0 0>, rot=<0 0 0 1>}
]
variables [
 Debug Bones {type="TArray<FRigElementKey>", default=((Type=Bone,Name="calf_l"),(Type=Bone,Name="calf_r"))}
 Debug Controls On {type=bool, default=False}
 Debug Ctrls {type="TArray<FRigElementKey>", default=(())}
 ItemToModify_0_Transform_Translation {type=FVector, default=(X=0,Y=0,Z=0)}
 ItemToModify_4_Transform {type=FTransform, default=(Rotation=(X=0,Y=0,Z=0,W=1),Translation=(X=0,Y=0,Z=0),Scale3D=(X=1,Y=1,Z=1))}
 ItemToModify_4_Transform_0 {type=FTransform, default=(Rotation=(X=0,Y=0,Z=0,W=1),Translation=(X=0,Y=0,Z=0),Scale3D=(X=1,Y=1,Z=1))}
 ModifyTransforms_128_ItemToModify_1 {
  type=FRigUnit_ModifyTransforms_PerItem
  default=(Item=(Type=Bone,Name=""),Transform=(Rotation=(X=0,Y=0,Z=0,W=1),Translation=(X=0,Y=0,Z=0),Scale3D=(X=1,Y=1,Z=1)))
 }
]
